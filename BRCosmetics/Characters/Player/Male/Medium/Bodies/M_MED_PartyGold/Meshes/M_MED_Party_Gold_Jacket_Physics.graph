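# Physics asset for the male medium PartyGold jacket, giving its skirt panels dynamic bone motion on the character.
# Seven panel chains around front, sides and back each run from a tiny MSTR capsule on the pelvis through two box bodies.
# Panel links have limited swing and very tight twist, the front-left being stiffer.
# Capsules on pelvis, spine and thighs form the body shapes, linked, like the MSTR anchors, by fully free constraints.
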PhysicsAsset /BRCosmetics/Characters/Player/Male/Medium/Bodies/M_MED_PartyGold/Meshes/M_MED_Party_Gold_Jacket_Physics
bodies {
 dyn_jacket_bk_01 [{shapeType=Box, extentX=7.2737, extentY=14.6529, extentZ=2.6476, center=<3.4183 0 -0.0002>}]
 dyn_jacket_bk_02 [{shapeType=Box, extentX=6.453, extentY=14.6529, extentZ=2.6476, center=<3.3606 0 0>}]
 dyn_jacket_bk_MSTR [{shapeType=Capsule, length=0.505, radius=0.505, rotation=<rot 0 -90 0>}]
 dyn_jacket_bk_l_01 [{shapeType=Box, extentX=6.9979, extentY=12.5928, extentZ=1.452, center=<3.3139 0 -0.0001>}]
 dyn_jacket_bk_l_02 [{shapeType=Box, extentX=6.453, extentY=12.5928, extentZ=1.452, center=<3.2714 0 0>}]
 dyn_jacket_bk_l_MSTR [{shapeType=Capsule, length=0.505, radius=0.505, rotation=<rot 0 -90 0>}]
 dyn_jacket_bk_r_01 [{shapeType=Box, extentX=6.8038, extentY=14.6529, extentZ=2.0112, center=<3.244 0 -0.0001>}]
 dyn_jacket_bk_r_02 [{shapeType=Box, extentX=6.453, extentY=14.6529, extentZ=2.0112, center=<3.2714 0 0>}]
 dyn_jacket_bk_r_MSTR [{shapeType=Capsule, length=0.505, radius=0.505, rotation=<rot 0 -90 0>}]
 dyn_jacket_fr_l_01 [{shapeType=Box, extentX=9.778, extentY=9.1379, extentZ=3.2259, center=<5.0106 -0.0057 0.2633>}]
 dyn_jacket_fr_l_02 [{shapeType=Box, extentX=9.9037, extentY=10.0587, extentZ=1.7968, center=<4.6779 -0.0035 0.0989>, rotation=<rot 3.1558 0.0836 0.0033>}]
 dyn_jacket_fr_l_MSTR [{shapeType=Capsule, length=0.505, radius=0.505, rotation=<rot 0 -90 0>}]
 dyn_jacket_fr_r_01 [{shapeType=Box, extentX=10.1872, extentY=14.6529, extentZ=1.1627, center=<4.9133 0 0>}]
 dyn_jacket_fr_r_02 [{shapeType=Box, extentX=11.7244, extentY=14.6529, extentZ=1.1627, center=<5.7366 0 0>}]
 dyn_jacket_fr_r_MSTR [{shapeType=Capsule, length=0.505, radius=0.505, rotation=<rot 0 -90 0>}]
 dyn_jacket_l_01 [{shapeType=Box, extentX=6.5503, extentY=14.6529, extentZ=2.4493, center=<2.9558 0 0>}]
 dyn_jacket_l_02 [{shapeType=Box, extentX=5.9636, extentY=14.6529, extentZ=2.4493, center=<3.0002 0 0>}]
 dyn_jacket_l_MSTR [{shapeType=Capsule, length=0.505, radius=0.505, rotation=<rot 0 -90 0>}]
 dyn_jacket_r_01 [{shapeType=Box, extentX=6.4318, extentY=14.6529, extentZ=2.0973, center=<5.1612 -0.0012 0.281>, rotation=<rot -7.1854 -0.0479 0.6334>}]
 dyn_jacket_r_02 [{shapeType=Box, extentX=8.0712, extentY=14.6529, extentZ=2.1452, center=<4.0038 0.0039 0>}]
 dyn_jacket_r_MSTR [{shapeType=Capsule, length=0.505, radius=0.505, rotation=<rot 0 -90 0>}]
 pelvis [{shapeType=Capsule, length=6.7317, radius=11.9164, center=<-2.5684 -0.4702 0>}]
 spine_01 [{shapeType=Capsule, length=8.908, radius=8.9292}]
 spine_02 [{shapeType=Capsule, length=2.855, radius=8.8908}]
 spine_03 [{shapeType=Capsule, length=4.8029, radius=8.8099}]
 spine_04 [{shapeType=Capsule, length=5.4496, radius=11.1842}]
 thigh_l [{shapeType=Capsule, length=22.0623, radius=8.3939, center=<-19.2117 0.0698 0>, rotation=<rot 90 -90 -93.1699>}]
 thigh_r [{shapeType=Capsule, length=25.8296, radius=8.0842, center=<20.6563 0 0>, rotation=<rot 90 0 0>}]
}
constraints [
 {
  bone1Name=spine_01
  bone2Name=pelvis
  swing1LimitDegrees=45
  swing1Motion=Free
  swing2LimitDegrees=45
  swing2Motion=Free
  twistLimitDegrees=45
  twistMotion=Free
 }
 {
  bone1Name=spine_02
  bone2Name=spine_01
  swing1LimitDegrees=45
  swing1Motion=Free
  swing2LimitDegrees=45
  swing2Motion=Free
  twistLimitDegrees=45
  twistMotion=Free
 }
 {
  bone1Name=spine_03
  bone2Name=spine_02
  swing1LimitDegrees=45
  swing1Motion=Free
  swing2LimitDegrees=45
  swing2Motion=Free
  twistLimitDegrees=45
  twistMotion=Free
 }
 {
  bone1Name=spine_04
  bone2Name=spine_03
  swing1LimitDegrees=45
  swing1Motion=Free
  swing2LimitDegrees=45
  swing2Motion=Free
  twistLimitDegrees=45
  twistMotion=Free
 }
 {bone1Name=dyn_jacket_l_01, bone2Name=dyn_jacket_l_MSTR, swing1LimitDegrees=60, swing2LimitDegrees=60, twistLimitDegrees=5}
 {bone1Name=dyn_jacket_l_02, bone2Name=dyn_jacket_l_01, swing1LimitDegrees=60, swing2LimitDegrees=60, twistLimitDegrees=5}
 {bone1Name=dyn_jacket_fr_l_01, bone2Name=dyn_jacket_fr_l_MSTR, swing1LimitDegrees=40, swing2LimitDegrees=40, twistLimitDegrees=5}
 {bone1Name=dyn_jacket_fr_l_02, bone2Name=dyn_jacket_fr_l_01, swing1LimitDegrees=40, swing2LimitDegrees=40, twistLimitDegrees=5}
 {bone1Name=dyn_jacket_bk_l_01, bone2Name=dyn_jacket_bk_l_MSTR, swing1LimitDegrees=60, swing2LimitDegrees=60, twistLimitDegrees=5}
 {bone1Name=dyn_jacket_bk_l_02, bone2Name=dyn_jacket_bk_l_01, swing1LimitDegrees=60, swing2LimitDegrees=60, twistLimitDegrees=5}
 {bone1Name=dyn_jacket_bk_01, bone2Name=dyn_jacket_bk_MSTR, swing1LimitDegrees=60, swing2LimitDegrees=60, twistLimitDegrees=5}
 {bone1Name=dyn_jacket_bk_02, bone2Name=dyn_jacket_bk_01, swing1LimitDegrees=60, swing2LimitDegrees=60, twistLimitDegrees=5}
 {bone1Name=dyn_jacket_fr_r_01, bone2Name=dyn_jacket_fr_r_MSTR, swing1LimitDegrees=60, swing2LimitDegrees=60, twistLimitDegrees=5}
 {bone1Name=dyn_jacket_fr_r_02, bone2Name=dyn_jacket_fr_r_01, swing1LimitDegrees=60, swing2LimitDegrees=60, twistLimitDegrees=5}
 {bone1Name=dyn_jacket_bk_r_01, bone2Name=dyn_jacket_bk_r_MSTR, swing1LimitDegrees=60, swing2LimitDegrees=60, twistLimitDegrees=5}
 {bone1Name=dyn_jacket_bk_r_02, bone2Name=dyn_jacket_bk_r_01, swing1LimitDegrees=60, swing2LimitDegrees=60, twistLimitDegrees=5}
 {bone1Name=dyn_jacket_r_01, bone2Name=dyn_jacket_r_MSTR, swing1LimitDegrees=60, swing2LimitDegrees=60, twistLimitDegrees=5}
 {bone1Name=dyn_jacket_r_02, bone2Name=dyn_jacket_r_01, swing1LimitDegrees=60, swing2LimitDegrees=60, twistLimitDegrees=5}
 {
  bone1Name=dyn_jacket_l_MSTR
  bone2Name=pelvis
  swing1LimitDegrees=45
  swing1Motion=Free
  swing2LimitDegrees=45
  swing2Motion=Free
  twistLimitDegrees=45
  twistMotion=Free
 }
 {
  bone1Name=dyn_jacket_fr_l_MSTR
  bone2Name=pelvis
  swing1LimitDegrees=45
  swing1Motion=Free
  swing2LimitDegrees=45
  swing2Motion=Free
  twistLimitDegrees=45
  twistMotion=Free
 }
 {
  bone1Name=dyn_jacket_bk_l_MSTR
  bone2Name=pelvis
  swing1LimitDegrees=45
  swing1Motion=Free
  swing2LimitDegrees=45
  swing2Motion=Free
  twistLimitDegrees=45
  twistMotion=Free
 }
 {
  bone1Name=dyn_jacket_bk_MSTR
  bone2Name=pelvis
  swing1LimitDegrees=45
  swing1Motion=Free
  swing2LimitDegrees=45
  swing2Motion=Free
  twistLimitDegrees=45
  twistMotion=Free
 }
 {
  bone1Name=dyn_jacket_fr_r_MSTR
  bone2Name=pelvis
  swing1LimitDegrees=45
  swing1Motion=Free
  swing2LimitDegrees=45
  swing2Motion=Free
  twistLimitDegrees=45
  twistMotion=Free
 }
 {
  bone1Name=dyn_jacket_bk_r_MSTR
  bone2Name=pelvis
  swing1LimitDegrees=45
  swing1Motion=Free
  swing2LimitDegrees=45
  swing2Motion=Free
  twistLimitDegrees=45
  twistMotion=Free
 }
 {
  bone1Name=dyn_jacket_r_MSTR
  bone2Name=pelvis
  swing1LimitDegrees=45
  swing1Motion=Free
  swing2LimitDegrees=45
  swing2Motion=Free
  twistLimitDegrees=45
  twistMotion=Free
 }
 {
  bone1Name=thigh_l
  bone2Name=pelvis
  swing1LimitDegrees=45
  swing1Motion=Free
  swing2LimitDegrees=45
  swing2Motion=Free
  twistLimitDegrees=45
  twistMotion=Free
 }
 {
  bone1Name=thigh_r
  bone2Name=pelvis
  swing1LimitDegrees=45
  swing1Motion=Free
  swing2LimitDegrees=45
  swing2Motion=Free
  twistLimitDegrees=45
  twistMotion=Free
 }
]
