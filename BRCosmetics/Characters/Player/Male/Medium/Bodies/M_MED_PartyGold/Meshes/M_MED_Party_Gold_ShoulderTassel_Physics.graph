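# Physics asset for the shoulder tassels on the male medium PartyGold body, supporting secondary motion of the fray bones.
# Capsules on pelvis, spine, clavicles and upper arms form a body chain with angularly free joints.
# Each shoulder has twelve top and bottom strands, each chaining a small Offset anchor on the clavicle to two capsule segments.
# Strand joints limit twist; most first segments also limit swing, while most tips swing freely.
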PhysicsAsset /BRCosmetics/Characters/Player/Male/Medium/Bodies/M_MED_PartyGold/Meshes/M_MED_Party_Gold_ShoulderTassel_Physics
bodies {
 clavicle_l [{shapeType=Capsule, length=8.4599, radius=4.9346, center=<9.0789 0 0>, rotation=<rot 90 0 0>}]
 clavicle_r [{shapeType=Capsule, length=8.4599, radius=4.9346, center=<-7.5457 -0.0001 0>, rotation=<rot 90 0 0>}]
 dyn_l_shoulderFray_btm_01_a [{shapeType=Capsule, length=1.2836, radius=2.1196, center=<2.034 0.0637 -0.0484>, rotation=<rot 90 0 0>}]
 dyn_l_shoulderFray_btm_01_a_Offset [{shapeType=Capsule, length=0.505, radius=0.505}]
 dyn_l_shoulderFray_btm_01_b [{shapeType=Capsule, length=0.7883, radius=2.2394, center=<1.9417 0.0013 0.2144>, rotation=<rot 90 0 0>}]
 dyn_l_shoulderFray_btm_02_a [{shapeType=Capsule, length=2.4673, radius=1.7642, center=<2.0572 -0.2622 -0.2448>, rotation=<rot 90 0 0>}]
 dyn_l_shoulderFray_btm_02_a_Offset [{shapeType=Capsule, length=0.505, radius=0.505}]
 dyn_l_shoulderFray_btm_02_b [{shapeType=Capsule, length=0.4845, radius=1.829, center=<1.5441 -0.1256 -0.1266>, rotation=<rot 90 0 0>}]
 dyn_l_shoulderFray_btm_03_a [{shapeType=Capsule, length=2.9109, radius=1.0829, center=<2.3631 0.0446 0.16>, rotation=<rot 90 0 0>}]
 dyn_l_shoulderFray_btm_03_a_Offset [{shapeType=Capsule, length=0.505, radius=0.505}]
 dyn_l_shoulderFray_btm_03_b [{shapeType=Capsule, length=2.2148, radius=0.954, center=<1.8935 0 0>, rotation=<rot 90 0 0>}]
 dyn_l_shoulderFray_btm_04_a [{shapeType=Capsule, length=2.4688, radius=1.0713, center=<2.3004 0 0.0002>, rotation=<rot 90 0 0>}]
 dyn_l_shoulderFray_btm_04_a_Offset [{shapeType=Capsule, length=0.505, radius=0.505}]
 dyn_l_shoulderFray_btm_04_b [{shapeType=Capsule, length=1.2248, radius=1.0367, center=<1.5931 0 0.0001>, rotation=<rot 90 0 0>}]
 dyn_l_shoulderFray_btm_05_a [{shapeType=Capsule, length=3.8149, radius=1.0638, center=<2.3902 0 0.0001>, rotation=<rot 90 0 0>}]
 dyn_l_shoulderFray_btm_05_a_Offset [{shapeType=Capsule, length=0.505, radius=0.505}]
 dyn_l_shoulderFray_btm_05_b [{shapeType=Capsule, length=0.7822, radius=0.9102, center=<1.2167 0 0.0001>, rotation=<rot 90 0 0>}]
 dyn_l_shoulderFray_btm_06_a [{shapeType=Capsule, length=1.7616, radius=0.7466, center=<2.3048 0 0>, rotation=<rot 90 0 0>}]
 dyn_l_shoulderFray_btm_06_a_Offset [{shapeType=Capsule, length=0.505, radius=0.505}]
 dyn_l_shoulderFray_btm_06_b [{shapeType=Capsule, length=1.3089, radius=0.739, center=<1.3626 0 0>, rotation=<rot 90 0 0>}]
 dyn_l_shoulderFray_top_01_a [{shapeType=Capsule, length=1.7162, radius=1.3926, center=<1.8168 0 0>, rotation=<rot 90 0 0>}]
 dyn_l_shoulderFray_top_01_a_Offset [{shapeType=Capsule, length=0.505, radius=0.505}]
 dyn_l_shoulderFray_top_01_b [{shapeType=Capsule, length=0.4681, radius=1.3926, center=<1.4016 0 0.0001>, rotation=<rot 90 0 0>}]
 dyn_l_shoulderFray_top_02_a [{shapeType=Capsule, length=2.4225, radius=1.1099, center=<2.1924 0 0>, rotation=<rot 90 0 0>}]
 dyn_l_shoulderFray_top_02_a_Offset [{shapeType=Capsule, length=0.505, radius=0.505}]
 dyn_l_shoulderFray_top_02_b [{shapeType=Capsule, length=0.9663, radius=1.1462, center=<1.5119 0 0>, rotation=<rot 90 0 0>}]
 dyn_l_shoulderFray_top_03_a [{shapeType=Capsule, length=3.4926, radius=1.1359, center=<2.7442 0 0.0001>, rotation=<rot 90 0 0>}]
 dyn_l_shoulderFray_top_03_a_Offset [{shapeType=Capsule, length=0.505, radius=0.505}]
 dyn_l_shoulderFray_top_03_b [{shapeType=Capsule, length=1.2571, radius=1.2166, center=<1.6416 0 0>, rotation=<rot 85.4482 122.6734 122.0151>}]
 dyn_l_shoulderFray_top_04_a [{shapeType=Capsule, length=3.4454, radius=0.8972, center=<2.554 0 0.0002>, rotation=<rot 90 0 0>}]
 dyn_l_shoulderFray_top_04_a_Offset [{shapeType=Capsule, length=0.505, radius=0.505}]
 dyn_l_shoulderFray_top_04_b [{shapeType=Capsule, length=1.7894, radius=0.6665, center=<1.4386 -0.0001 0>, rotation=<rot 90 0 0>}]
 dyn_l_shoulderFray_top_05_a [{shapeType=Capsule, length=1.7234, radius=0.9102, center=<1.7205 0 0>, rotation=<rot 90 0 0>}]
 dyn_l_shoulderFray_top_05_a_Offset [{shapeType=Capsule, length=0.505, radius=0.505}]
 dyn_l_shoulderFray_top_05_b [{shapeType=Capsule, length=1.0426, radius=1.2277, center=<1.7537 0 0.0001>, rotation=<rot 90 0 0>}]
 dyn_l_shoulderFray_top_06_a [{shapeType=Capsule, length=2.4273, radius=0.7253, center=<1.7591 -0.0002 0>, rotation=<rot 90 0 0>}]
 dyn_l_shoulderFray_top_06_a_Offset [{shapeType=Capsule, length=0.505, radius=0.505}]
 dyn_l_shoulderFray_top_06_b [{shapeType=Capsule, length=1.4901, radius=0.7253, center=<1.516 -0.0001 0>, rotation=<rot 90 0 0>}]
 dyn_r_shoulderFray_btm_01_a [{shapeType=Capsule, length=2.0342, radius=1.0197, center=<-1.7102 0.0001 0>, rotation=<rot 90 0 0>}]
 dyn_r_shoulderFray_btm_01_a_Offset [{shapeType=Capsule, length=0.505, radius=0.505}]
 dyn_r_shoulderFray_btm_01_b [{shapeType=Capsule, length=2.9803, radius=1.0197, center=<-2.3504 -0.2011 0.1509>, rotation=<rot 83.4869 -32.2196 -37.6233>}]
 dyn_r_shoulderFray_btm_02_a [{shapeType=Capsule, length=3.5, radius=0.5, center=<-2.1529 0 0>, rotation=<rot 90 0 0>}]
 dyn_r_shoulderFray_btm_02_a_Offset [{shapeType=Capsule, length=0.505, radius=0.505}]
 dyn_r_shoulderFray_btm_02_b [{shapeType=Capsule, length=1.9976, radius=0.5, center=<-1.6193 0.0001 0>, rotation=<rot 90 0 0>}]
 dyn_r_shoulderFray_btm_03_a [{shapeType=Capsule, length=3.5, radius=0.5, center=<-2.26 0 0>, rotation=<rot 90 0 0>}]
 dyn_r_shoulderFray_btm_03_a_Offset [{shapeType=Capsule, length=0.505, radius=0.505}]
 dyn_r_shoulderFray_btm_03_b [{shapeType=Capsule, length=2.1408, radius=0.5, center=<-1.975 0 -0.0001>, rotation=<rot 90 0 0>}]
 dyn_r_shoulderFray_btm_04_a [{shapeType=Capsule, length=1.7474, radius=1.1734, center=<-1.8442 0 0>, rotation=<rot 90 0 0>}]
 dyn_r_shoulderFray_btm_04_a_Offset [{shapeType=Capsule, length=0.505, radius=0.505}]
 dyn_r_shoulderFray_btm_04_b [{shapeType=Capsule, length=2.209, radius=1.1336, center=<-2.4388 0 0>, rotation=<rot 90 0 0>}]
 dyn_r_shoulderFray_btm_05_a [{shapeType=Capsule, length=3.3096, radius=0.9818, center=<-2.6307 0.0001 0>, rotation=<rot 90 0 0>}]
 dyn_r_shoulderFray_btm_05_a_Offset [{shapeType=Capsule, length=0.505, radius=0.505}]
 dyn_r_shoulderFray_btm_05_b [{shapeType=Capsule, length=1.3744, radius=0.9538, center=<-1.5992 0 0>, rotation=<rot 90 0 0>}]
 dyn_r_shoulderFray_btm_06_a [{shapeType=Capsule, length=2.041, radius=0.7576, center=<-1.9357 0 0>, rotation=<rot 90 0 0>}]
 dyn_r_shoulderFray_btm_06_a_Offset [{shapeType=Capsule, length=0.505, radius=0.505}]
 dyn_r_shoulderFray_btm_06_b [{shapeType=Capsule, length=1.3871, radius=1.0869, center=<-1.8193 0 0>, rotation=<rot 90 0 0>}]
 dyn_r_shoulderFray_top_01_a [{shapeType=Capsule, length=1.3667, radius=1.2703, center=<-1.5323 0 0>, rotation=<rot 90 0 0>}]
 dyn_r_shoulderFray_top_01_a_Offset [{shapeType=Capsule, length=0.505, radius=0.505}]
 dyn_r_shoulderFray_top_01_b [{shapeType=Capsule, length=2.9604, radius=1.0396, center=<-2.3753 0 -0.0002>, rotation=<rot 90 0 0>}]
 dyn_r_shoulderFray_top_02_a [{shapeType=Capsule, length=2.973, radius=1, center=<-2.1343 0 0>, rotation=<rot 90 0 0>}]
 dyn_r_shoulderFray_top_02_a_Offset [{shapeType=Capsule, length=0.505, radius=0.505}]
 dyn_r_shoulderFray_top_02_b [{shapeType=Capsule, length=2, radius=1, center=<-1.8577 -0.0001 0>, rotation=<rot 90 0 0>}]
 dyn_r_shoulderFray_top_03_a [{shapeType=Capsule, length=2.7152, radius=1.2848, center=<-2.6845 0 0>, rotation=<rot 90 0 0>}]
 dyn_r_shoulderFray_top_03_a_Offset [{shapeType=Capsule, length=0.505, radius=0.505}]
 dyn_r_shoulderFray_top_03_b [{shapeType=Capsule, length=1.2026, radius=1.3482, center=<-1.5821 0 0>, rotation=<rot 90 0 0>}]
 dyn_r_shoulderFray_top_04_a [{shapeType=Capsule, length=3.2165, radius=1.3969, center=<-2.6488 0 0>, rotation=<rot 90 0 0>}]
 dyn_r_shoulderFray_top_04_a_Offset [{shapeType=Capsule, length=0.505, radius=0.505}]
 dyn_r_shoulderFray_top_04_b [{shapeType=Capsule, length=1.5055, radius=1.3969, center=<-1.7851 -0.0001 0>, rotation=<rot 90 0 0>}]
 dyn_r_shoulderFray_top_05_a [{shapeType=Capsule, length=0.7413, radius=1.323, center=<-1.7202 0 0>, rotation=<rot 90 0.3357 0.244>}]
 dyn_r_shoulderFray_top_05_a_Offset [{shapeType=Capsule, length=0.505, radius=0.505}]
 dyn_r_shoulderFray_top_05_b [{shapeType=Capsule, length=1.5036, radius=1.3044, center=<-1.752 0 0.0001>, rotation=<rot 90 0 -0.0916>}]
 dyn_r_shoulderFray_top_06_a [{shapeType=Capsule, length=1.2832, radius=0.7751, center=<-1.5711 0 0>, rotation=<rot 90 0.3357 0.244>}]
 dyn_r_shoulderFray_top_06_a_Offset [{shapeType=Capsule, length=0.505, radius=0.505}]
 dyn_r_shoulderFray_top_06_b [{shapeType=Capsule, length=1.4024, radius=0.9887, center=<-1.5927 0 0>, rotation=<rot 90 0.6155 0.5237>}]
 pelvis [{shapeType=Capsule, length=6.0858, radius=11.9839}]
 spine_01 [{shapeType=Capsule, length=14.1393, radius=6.1997}]
 spine_02 [{shapeType=Capsule, length=7.2211, radius=8.7852}]
 spine_03 [{shapeType=Capsule, length=7.1341, radius=8.8722}]
 spine_04 [{shapeType=Capsule, length=11.9019, radius=8.4371}]
 spine_05 [{shapeType=Capsule, length=21.431, radius=11.0584, center=<-0.2075 -1.9905 -1.4479>}]
 upperarm_l [{shapeType=Capsule, length=16.068, radius=5.9843, center=<10.0203 -0.5818 -0.8121>, rotation=<rot 90 0.2558 -1.9345>}]
 upperarm_r [{shapeType=Capsule, length=20.969, radius=6.1997, center=<-11.279 0.3855 0.7636>, rotation=<rot 90 0 0>}]
}
constraints [
 {
  bone1Name=spine_05
  bone2Name=spine_04
  swing1LimitDegrees=45
  swing1Motion=Free
  swing2LimitDegrees=45
  swing2Motion=Free
  twistLimitDegrees=45
  twistMotion=Free
 }
 {
  bone1Name=spine_04
  bone2Name=spine_03
  swing1LimitDegrees=45
  swing1Motion=Free
  swing2LimitDegrees=45
  swing2Motion=Free
  twistLimitDegrees=45
  twistMotion=Free
 }
 {
  bone1Name=spine_03
  bone2Name=spine_02
  swing1LimitDegrees=45
  swing1Motion=Free
  swing2LimitDegrees=45
  swing2Motion=Free
  twistLimitDegrees=45
  twistMotion=Free
 }
 {
  bone1Name=spine_02
  bone2Name=spine_01
  swing1LimitDegrees=45
  swing1Motion=Free
  swing2LimitDegrees=45
  swing2Motion=Free
  twistLimitDegrees=45
  twistMotion=Free
 }
 {
  bone1Name=spine_01
  bone2Name=pelvis
  swing1LimitDegrees=45
  swing1Motion=Free
  swing2LimitDegrees=45
  swing2Motion=Free
  twistLimitDegrees=45
  twistMotion=Free
 }
 {
  bone1Name=upperarm_l
  bone2Name=clavicle_l
  swing1LimitDegrees=45
  swing1Motion=Free
  swing2LimitDegrees=45
  swing2Motion=Free
  twistLimitDegrees=45
  twistMotion=Free
 }
 {
  bone1Name=clavicle_l
  bone2Name=spine_05
  swing1LimitDegrees=45
  swing1Motion=Free
  swing2LimitDegrees=45
  swing2Motion=Free
  twistLimitDegrees=45
  twistMotion=Free
 }
 {
  bone1Name=dyn_l_shoulderFray_top_01_a
  bone2Name=dyn_l_shoulderFray_top_01_a_Offset
  swing1LimitDegrees=60
  swing2LimitDegrees=60
  twistLimitDegrees=25
 }
 {
  bone1Name=dyn_l_shoulderFray_top_01_b
  bone2Name=dyn_l_shoulderFray_top_01_a
  swing1LimitDegrees=60
  swing2LimitDegrees=60
  twistLimitDegrees=25
 }
 {
  bone1Name=dyn_l_shoulderFray_top_02_a
  bone2Name=dyn_l_shoulderFray_top_02_a_Offset
  swing1LimitDegrees=60
  swing2LimitDegrees=60
  twistLimitDegrees=25
 }
 {
  bone1Name=dyn_l_shoulderFray_top_02_b
  bone2Name=dyn_l_shoulderFray_top_02_a
  swing1LimitDegrees=60
  swing2LimitDegrees=60
  twistLimitDegrees=25
 }
 {
  bone1Name=dyn_l_shoulderFray_top_03_a
  bone2Name=dyn_l_shoulderFray_top_03_a_Offset
  swing1LimitDegrees=60
  swing2LimitDegrees=60
  twistLimitDegrees=25
 }
 {
  bone1Name=dyn_l_shoulderFray_top_03_b
  bone2Name=dyn_l_shoulderFray_top_03_a
  swing1LimitDegrees=60
  swing1Motion=Free
  swing2LimitDegrees=60
  swing2Motion=Free
  twistLimitDegrees=25
 }
 {
  bone1Name=dyn_l_shoulderFray_top_04_a
  bone2Name=dyn_l_shoulderFray_top_04_a_Offset
  swing1LimitDegrees=60
  swing2LimitDegrees=60
  twistLimitDegrees=25
 }
 {
  bone1Name=dyn_l_shoulderFray_top_04_b
  bone2Name=dyn_l_shoulderFray_top_04_a
  swing1LimitDegrees=60
  swing1Motion=Free
  swing2LimitDegrees=60
  swing2Motion=Free
  twistLimitDegrees=25
 }
 {
  bone1Name=dyn_l_shoulderFray_top_05_a
  bone2Name=dyn_l_shoulderFray_top_05_a_Offset
  swing1LimitDegrees=60
  swing2LimitDegrees=60
  twistLimitDegrees=25
 }
 {
  bone1Name=dyn_l_shoulderFray_top_05_b
  bone2Name=dyn_l_shoulderFray_top_05_a
  swing1LimitDegrees=60
  swing1Motion=Free
  swing2LimitDegrees=60
  swing2Motion=Free
  twistLimitDegrees=25
 }
 {
  bone1Name=dyn_l_shoulderFray_top_06_a
  bone2Name=dyn_l_shoulderFray_top_06_a_Offset
  swing1LimitDegrees=60
  swing2LimitDegrees=60
  twistLimitDegrees=25
 }
 {
  bone1Name=dyn_l_shoulderFray_top_06_b
  bone2Name=dyn_l_shoulderFray_top_06_a
  swing1LimitDegrees=60
  swing1Motion=Free
  swing2LimitDegrees=60
  swing2Motion=Free
  twistLimitDegrees=25
 }
 {
  bone1Name=dyn_l_shoulderFray_btm_06_a
  bone2Name=dyn_l_shoulderFray_btm_06_a_Offset
  swing1LimitDegrees=60
  swing2LimitDegrees=60
  twistLimitDegrees=25
 }
 {
  bone1Name=dyn_l_shoulderFray_btm_06_b
  bone2Name=dyn_l_shoulderFray_btm_06_a
  swing1LimitDegrees=60
  swing1Motion=Free
  swing2LimitDegrees=60
  swing2Motion=Free
  twistLimitDegrees=25
 }
 {
  bone1Name=dyn_l_shoulderFray_btm_05_a
  bone2Name=dyn_l_shoulderFray_btm_05_a_Offset
  swing1LimitDegrees=60
  swing2LimitDegrees=60
  twistLimitDegrees=25
 }
 {
  bone1Name=dyn_l_shoulderFray_btm_05_b
  bone2Name=dyn_l_shoulderFray_btm_05_a
  swing1LimitDegrees=60
  swing1Motion=Free
  swing2LimitDegrees=60
  swing2Motion=Free
  twistLimitDegrees=25
 }
 {
  bone1Name=dyn_l_shoulderFray_btm_04_a
  bone2Name=dyn_l_shoulderFray_btm_04_a_Offset
  swing1LimitDegrees=60
  swing2LimitDegrees=60
  twistLimitDegrees=25
 }
 {
  bone1Name=dyn_l_shoulderFray_btm_04_b
  bone2Name=dyn_l_shoulderFray_btm_04_a
  swing1LimitDegrees=60
  swing1Motion=Free
  swing2LimitDegrees=60
  swing2Motion=Free
  twistLimitDegrees=25
 }
 {
  bone1Name=dyn_l_shoulderFray_btm_03_a
  bone2Name=dyn_l_shoulderFray_btm_03_a_Offset
  swing1LimitDegrees=60
  swing2LimitDegrees=60
  twistLimitDegrees=25
 }
 {
  bone1Name=dyn_l_shoulderFray_btm_03_b
  bone2Name=dyn_l_shoulderFray_btm_03_a
  swing1LimitDegrees=60
  swing1Motion=Free
  swing2LimitDegrees=60
  swing2Motion=Free
  twistLimitDegrees=25
 }
 {
  bone1Name=dyn_l_shoulderFray_btm_02_a
  bone2Name=dyn_l_shoulderFray_btm_02_a_Offset
  swing1LimitDegrees=60
  swing2LimitDegrees=60
  twistLimitDegrees=25
 }
 {
  bone1Name=dyn_l_shoulderFray_btm_02_b
  bone2Name=dyn_l_shoulderFray_btm_02_a
  swing1LimitDegrees=60
  swing2LimitDegrees=60
  twistLimitDegrees=25
 }
 {
  bone1Name=dyn_l_shoulderFray_btm_01_a
  bone2Name=dyn_l_shoulderFray_btm_01_a_Offset
  swing1LimitDegrees=60
  swing2LimitDegrees=60
  twistLimitDegrees=25
 }
 {
  bone1Name=dyn_l_shoulderFray_btm_01_b
  bone2Name=dyn_l_shoulderFray_btm_01_a
  swing1LimitDegrees=60
  swing2LimitDegrees=60
  twistLimitDegrees=25
 }
 {
  bone1Name=dyn_r_shoulderFray_top_01_a
  bone2Name=dyn_r_shoulderFray_top_01_a_Offset
  swing1LimitDegrees=60
  swing2LimitDegrees=60
  twistLimitDegrees=25
 }
 {
  bone1Name=dyn_r_shoulderFray_top_01_b
  bone2Name=dyn_r_shoulderFray_top_01_a
  swing1LimitDegrees=60
  swing1Motion=Free
  swing2LimitDegrees=60
  swing2Motion=Free
  twistLimitDegrees=25
 }
 {
  bone1Name=dyn_r_shoulderFray_top_03_a
  bone2Name=dyn_r_shoulderFray_top_03_a_Offset
  swing1LimitDegrees=60
  swing2LimitDegrees=60
  twistLimitDegrees=25
 }
 {
  bone1Name=dyn_r_shoulderFray_top_03_b
  bone2Name=dyn_r_shoulderFray_top_03_a
  swing1LimitDegrees=60
  swing1Motion=Free
  swing2LimitDegrees=60
  swing2Motion=Free
  twistLimitDegrees=25
 }
 {
  bone1Name=dyn_r_shoulderFray_top_05_a
  bone2Name=dyn_r_shoulderFray_top_05_a_Offset
  swing1LimitDegrees=60
  swing2LimitDegrees=60
  twistLimitDegrees=25
 }
 {
  bone1Name=dyn_r_shoulderFray_top_05_b
  bone2Name=dyn_r_shoulderFray_top_05_a
  swing1LimitDegrees=60
  swing2LimitDegrees=60
  twistLimitDegrees=25
 }
 {
  bone1Name=dyn_r_shoulderFray_top_06_a
  bone2Name=dyn_r_shoulderFray_top_06_a_Offset
  swing1LimitDegrees=60
  swing2LimitDegrees=60
  twistLimitDegrees=25
 }
 {
  bone1Name=dyn_r_shoulderFray_top_06_b
  bone2Name=dyn_r_shoulderFray_top_06_a
  swing1LimitDegrees=60
  swing2LimitDegrees=60
  twistLimitDegrees=25
 }
 {
  bone1Name=dyn_r_shoulderFray_btm_06_a
  bone2Name=dyn_r_shoulderFray_btm_06_a_Offset
  swing1LimitDegrees=60
  swing2LimitDegrees=60
  twistLimitDegrees=25
 }
 {
  bone1Name=dyn_r_shoulderFray_btm_06_b
  bone2Name=dyn_r_shoulderFray_btm_06_a
  swing1LimitDegrees=60
  swing1Motion=Free
  swing2LimitDegrees=60
  swing2Motion=Free
  twistLimitDegrees=25
 }
 {
  bone1Name=dyn_r_shoulderFray_btm_05_a
  bone2Name=dyn_r_shoulderFray_btm_05_a_Offset
  swing1LimitDegrees=60
  swing2LimitDegrees=60
  twistLimitDegrees=25
 }
 {
  bone1Name=dyn_r_shoulderFray_btm_05_b
  bone2Name=dyn_r_shoulderFray_btm_05_a
  swing1LimitDegrees=60
  swing1Motion=Free
  swing2LimitDegrees=60
  swing2Motion=Free
  twistLimitDegrees=25
 }
 {
  bone1Name=dyn_r_shoulderFray_btm_04_a
  bone2Name=dyn_r_shoulderFray_btm_04_a_Offset
  swing1LimitDegrees=60
  swing2LimitDegrees=60
  twistLimitDegrees=25
 }
 {
  bone1Name=dyn_r_shoulderFray_btm_04_b
  bone2Name=dyn_r_shoulderFray_btm_04_a
  swing1LimitDegrees=60
  swing1Motion=Free
  swing2LimitDegrees=60
  swing2Motion=Free
  twistLimitDegrees=25
 }
 {
  bone1Name=dyn_r_shoulderFray_btm_03_a
  bone2Name=dyn_r_shoulderFray_btm_03_a_Offset
  swing1LimitDegrees=60
  swing2LimitDegrees=60
  twistLimitDegrees=25
 }
 {
  bone1Name=dyn_r_shoulderFray_btm_03_b
  bone2Name=dyn_r_shoulderFray_btm_03_a
  swing1LimitDegrees=60
  swing1Motion=Free
  swing2LimitDegrees=60
  swing2Motion=Free
  twistLimitDegrees=25
 }
 {
  bone1Name=dyn_r_shoulderFray_btm_02_a
  bone2Name=dyn_r_shoulderFray_btm_02_a_Offset
  swing1LimitDegrees=60
  swing2LimitDegrees=60
  twistLimitDegrees=25
 }
 {
  bone1Name=dyn_r_shoulderFray_btm_02_b
  bone2Name=dyn_r_shoulderFray_btm_02_a
  swing1LimitDegrees=60
  swing1Motion=Free
  swing2LimitDegrees=60
  swing2Motion=Free
  twistLimitDegrees=25
 }
 {
  bone1Name=dyn_r_shoulderFray_btm_01_a
  bone2Name=dyn_r_shoulderFray_btm_01_a_Offset
  swing1LimitDegrees=60
  swing2LimitDegrees=60
  twistLimitDegrees=25
 }
 {
  bone1Name=dyn_r_shoulderFray_btm_01_b
  bone2Name=dyn_r_shoulderFray_btm_01_a
  swing1LimitDegrees=60
  swing2LimitDegrees=60
  twistLimitDegrees=25
 }
 {
  bone1Name=upperarm_r
  bone2Name=clavicle_r
  swing1LimitDegrees=45
  swing1Motion=Free
  swing2LimitDegrees=45
  swing2Motion=Free
  twistLimitDegrees=45
  twistMotion=Free
 }
 {
  bone1Name=clavicle_r
  bone2Name=spine_05
  swing1LimitDegrees=45
  swing1Motion=Free
  swing2LimitDegrees=45
  swing2Motion=Free
  twistLimitDegrees=45
  twistMotion=Free
 }
 {
  bone1Name=dyn_l_shoulderFray_btm_06_a_Offset
  bone2Name=clavicle_l
  swing1LimitDegrees=45
  swing1Motion=Free
  swing2LimitDegrees=45
  swing2Motion=Free
  twistLimitDegrees=45
  twistMotion=Free
 }
 {
  bone1Name=dyn_l_shoulderFray_top_06_a_Offset
  bone2Name=clavicle_l
  swing1LimitDegrees=45
  swing1Motion=Free
  swing2LimitDegrees=45
  swing2Motion=Free
  twistLimitDegrees=45
  twistMotion=Free
 }
 {
  bone1Name=dyn_l_shoulderFray_top_05_a_Offset
  bone2Name=clavicle_l
  swing1LimitDegrees=45
  swing1Motion=Free
  swing2LimitDegrees=45
  swing2Motion=Free
  twistLimitDegrees=45
  twistMotion=Free
 }
 {
  bone1Name=dyn_l_shoulderFray_btm_05_a_Offset
  bone2Name=clavicle_l
  swing1LimitDegrees=45
  swing1Motion=Free
  swing2LimitDegrees=45
  swing2Motion=Free
  twistLimitDegrees=45
  twistMotion=Free
 }
 {
  bone1Name=dyn_l_shoulderFray_top_04_a_Offset
  bone2Name=clavicle_l
  swing1LimitDegrees=45
  swing1Motion=Free
  swing2LimitDegrees=45
  swing2Motion=Free
  twistLimitDegrees=45
  twistMotion=Free
 }
 {
  bone1Name=dyn_l_shoulderFray_btm_04_a_Offset
  bone2Name=clavicle_l
  swing1LimitDegrees=45
  swing1Motion=Free
  swing2LimitDegrees=45
  swing2Motion=Free
  twistLimitDegrees=45
  twistMotion=Free
 }
 {
  bone1Name=dyn_l_shoulderFray_top_03_a_Offset
  bone2Name=clavicle_l
  swing1LimitDegrees=45
  swing1Motion=Free
  swing2LimitDegrees=45
  swing2Motion=Free
  twistLimitDegrees=45
  twistMotion=Free
 }
 {
  bone1Name=dyn_l_shoulderFray_btm_03_a_Offset
  bone2Name=clavicle_l
  swing1LimitDegrees=45
  swing1Motion=Free
  swing2LimitDegrees=45
  swing2Motion=Free
  twistLimitDegrees=45
  twistMotion=Free
 }
 {
  bone1Name=dyn_l_shoulderFray_top_02_a_Offset
  bone2Name=clavicle_l
  swing1LimitDegrees=45
  swing1Motion=Free
  swing2LimitDegrees=45
  swing2Motion=Free
  twistLimitDegrees=45
  twistMotion=Free
 }
 {
  bone1Name=dyn_l_shoulderFray_btm_02_a_Offset
  bone2Name=clavicle_l
  swing1LimitDegrees=45
  swing1Motion=Free
  swing2LimitDegrees=45
  swing2Motion=Free
  twistLimitDegrees=45
  twistMotion=Free
 }
 {
  bone1Name=dyn_l_shoulderFray_top_01_a_Offset
  bone2Name=clavicle_l
  swing1LimitDegrees=45
  swing1Motion=Free
  swing2LimitDegrees=45
  swing2Motion=Free
  twistLimitDegrees=45
  twistMotion=Free
 }
 {
  bone1Name=dyn_l_shoulderFray_btm_01_a_Offset
  bone2Name=clavicle_l
  swing1LimitDegrees=45
  swing1Motion=Free
  swing2LimitDegrees=45
  swing2Motion=Free
  twistLimitDegrees=45
  twistMotion=Free
 }
 {
  bone1Name=dyn_r_shoulderFray_btm_06_a_Offset
  bone2Name=clavicle_r
  swing1LimitDegrees=45
  swing1Motion=Free
  swing2LimitDegrees=45
  swing2Motion=Free
  twistLimitDegrees=45
  twistMotion=Free
 }
 {
  bone1Name=dyn_r_shoulderFray_top_05_a_Offset
  bone2Name=clavicle_r
  swing1LimitDegrees=45
  swing1Motion=Free
  swing2LimitDegrees=45
  swing2Motion=Free
  twistLimitDegrees=45
  twistMotion=Free
 }
 {
  bone1Name=dyn_r_shoulderFray_btm_05_a_Offset
  bone2Name=clavicle_r
  swing1LimitDegrees=45
  swing1Motion=Free
  swing2LimitDegrees=45
  swing2Motion=Free
  twistLimitDegrees=45
  twistMotion=Free
 }
 {
  bone1Name=dyn_r_shoulderFray_top_06_a_Offset
  bone2Name=clavicle_r
  swing1LimitDegrees=45
  swing1Motion=Free
  swing2LimitDegrees=45
  swing2Motion=Free
  twistLimitDegrees=45
  twistMotion=Free
 }
 {
  bone1Name=dyn_r_shoulderFray_top_04_a_Offset
  bone2Name=clavicle_r
  swing1LimitDegrees=45
  swing1Motion=Free
  swing2LimitDegrees=45
  swing2Motion=Free
  twistLimitDegrees=45
 }
 {
  bone1Name=dyn_r_shoulderFray_btm_04_a_Offset
  bone2Name=clavicle_r
  swing1LimitDegrees=45
  swing1Motion=Free
  swing2LimitDegrees=45
  swing2Motion=Free
  twistLimitDegrees=45
  twistMotion=Free
 }
 {
  bone1Name=dyn_r_shoulderFray_top_03_a_Offset
  bone2Name=clavicle_r
  swing1LimitDegrees=45
  swing1Motion=Free
  swing2LimitDegrees=45
  swing2Motion=Free
  twistLimitDegrees=45
  twistMotion=Free
 }
 {
  bone1Name=dyn_r_shoulderFray_btm_03_a_Offset
  bone2Name=clavicle_r
  swing1LimitDegrees=45
  swing1Motion=Free
  swing2LimitDegrees=45
  swing2Motion=Free
  twistLimitDegrees=45
  twistMotion=Free
 }
 {
  bone1Name=dyn_r_shoulderFray_top_02_a_Offset
  bone2Name=clavicle_r
  swing1LimitDegrees=45
  swing1Motion=Free
  swing2LimitDegrees=45
  swing2Motion=Free
  twistLimitDegrees=45
  twistMotion=Free
 }
 {
  bone1Name=dyn_r_shoulderFray_btm_02_a_Offset
  bone2Name=clavicle_r
  swing1LimitDegrees=45
  swing1Motion=Free
  swing2LimitDegrees=45
  swing2Motion=Free
  twistLimitDegrees=45
  twistMotion=Free
 }
 {
  bone1Name=dyn_r_shoulderFray_top_01_a_Offset
  bone2Name=clavicle_r
  swing1LimitDegrees=45
  swing1Motion=Free
  swing2LimitDegrees=45
  swing2Motion=Free
  twistLimitDegrees=45
  twistMotion=Free
 }
 {
  bone1Name=dyn_r_shoulderFray_btm_01_a_Offset
  bone2Name=clavicle_r
  swing1LimitDegrees=45
  swing1Motion=Free
  swing2LimitDegrees=45
  swing2Motion=Free
  twistLimitDegrees=45
  twistMotion=Free
 }
 {
  bone1Name=dyn_r_shoulderFray_top_02_a
  bone2Name=dyn_r_shoulderFray_top_02_a_Offset
  swing1LimitDegrees=60
  swing2LimitDegrees=60
  twistLimitDegrees=25
 }
 {
  bone1Name=dyn_r_shoulderFray_top_02_b
  bone2Name=dyn_r_shoulderFray_top_02_a
  swing1LimitDegrees=60
  swing1Motion=Free
  swing2LimitDegrees=60
  swing2Motion=Free
  twistLimitDegrees=25
 }
 {
  bone1Name=dyn_r_shoulderFray_top_04_a
  bone2Name=dyn_r_shoulderFray_top_04_a_Offset
  swing1LimitDegrees=45
  swing1Motion=Free
  swing2LimitDegrees=45
  swing2Motion=Free
  twistLimitDegrees=5
 }
 {
  bone1Name=dyn_r_shoulderFray_top_04_b
  bone2Name=dyn_r_shoulderFray_top_04_a
  swing1LimitDegrees=45
  swing1Motion=Free
  swing2LimitDegrees=45
  swing2Motion=Free
  twistLimitDegrees=5
 }
]
