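# Physics asset for the M_MED Party Gold player body's zipper bones: two chest pairs and one leg pair, apparently for secondary dangle motion.
# Each zipper is a box constrained to a small Offset capsule; chest Offsets link to the pelvis, leg Offsets to their thighs.
# Capsules on pelvis, spine and thighs form the body shapes, with angularly free constraints.
# Leg twist is asymmetric: free on the right zipper, locked on the left.
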PhysicsAsset /BRCosmetics/Characters/Player/Male/Medium/Bodies/M_MED_PartyGold/Meshes/M_MED_Party_Gold_Zipper_Physics
bodies {
 dyn_l_zipper_chest_01 [{shapeType=Box, extentX=1.6524, extentY=1.0442, extentZ=4.3954, center=<1.7548 0.1146 -0.0435>, rotation=<rot 0 91.4822 -89.5556>}]
 dyn_l_zipper_chest_01_Offset [{shapeType=Capsule, length=0.505, radius=0.505, rotation=<rot 0 -90 0>}]
 dyn_l_zipper_chest_02 [{shapeType=Box, extentX=1.6331, extentY=1.0981, extentZ=4.3954, center=<1.7895 0.0982 -0.1153>, rotation=<rot 0 -87.343 -87.4815>}]
 dyn_l_zipper_chest_02_Offset [{shapeType=Capsule, length=0.505, radius=0.505, rotation=<rot 0 -90 0>}]
 dyn_l_zipper_leg [{shapeType=Box, extentX=1.4102, extentY=0.5004, extentZ=4.4844, center=<1.8408 -0.2097 -0.122>, rotation=<rot 0 -97.5538 -86.4772>}]
 dyn_l_zipper_leg_Offset [{shapeType=Capsule, length=0.505, radius=0.505, rotation=<rot 0 -90 0>}]
 dyn_r_zipper_chest_01 [{shapeType=Box, extentX=1.653, extentY=1.0425, extentZ=4.3954, center=<1.7429 -0.1573 -0.0517>, rotation=<rot 0 -97.5925 -86.872>}]
 dyn_r_zipper_chest_01_Offset [{shapeType=Capsule, length=0.505, radius=0.505, rotation=<rot 0 -90 0>}]
 dyn_r_zipper_chest_02 [{shapeType=Box, extentX=1.6321, extentY=1.1004, extentZ=4.3954, center=<1.8797 -0.1117 -0.1195>, rotation=<rot 0 -95.3572 -85.8226>}]
 dyn_r_zipper_chest_02_Offset [{shapeType=Capsule, length=0.505, radius=0.505, rotation=<rot 0 -90 0>}]
 dyn_r_zipper_leg [{shapeType=Box, extentX=1.538, extentY=0.5821, extentZ=4.4843, center=<1.6663 -0.0763 0.1464>, rotation=<rot 0.0057 85.6546 -84.3673>}]
 dyn_r_zipper_leg_Offset [{shapeType=Capsule, length=0.505, radius=0.505, rotation=<rot 0 -90 0>}]
 pelvis [{shapeType=Capsule, length=8.1829, radius=11.0539}]
 spine_01 [{shapeType=Capsule, length=8.2602, radius=10.2251}]
 spine_02 [{shapeType=Capsule, length=5.1119, radius=10.6158}]
 spine_03 [{shapeType=Capsule, length=4.8794, radius=10.6158}]
 spine_04 [{shapeType=Capsule, length=6.9406, radius=10.6158, center=<-2.1843 0 0>}]
 spine_05 [{shapeType=Capsule, length=8.1829, radius=11.0539}]
 thigh_l [{shapeType=Capsule, length=25.1196, radius=6.8192, center=<-19.7813 0.0023 -0.4426>, rotation=<rot 83.2393 -2.3062 -2.3704>}]
 thigh_r [{shapeType=Capsule, length=25.0521, radius=6.8867, center=<18.5594 0.0318 0.2129>, rotation=<rot 85.8125 -3.7642 -3.8233>}]
}
constraints [
 {
  bone1Name=spine_01
  bone2Name=pelvis
  swing1LimitDegrees=45
  swing1Motion=Free
  swing2LimitDegrees=45
  swing2Motion=Free
  twistLimitDegrees=45
  twistMotion=Free
 }
 {
  bone1Name=spine_02
  bone2Name=spine_01
  swing1LimitDegrees=45
  swing1Motion=Free
  swing2LimitDegrees=45
  swing2Motion=Free
  twistLimitDegrees=45
  twistMotion=Free
 }
 {
  bone1Name=spine_03
  bone2Name=spine_02
  swing1LimitDegrees=45
  swing1Motion=Free
  swing2LimitDegrees=45
  swing2Motion=Free
  twistLimitDegrees=45
  twistMotion=Free
 }
 {
  bone1Name=spine_04
  bone2Name=spine_03
  swing1LimitDegrees=45
  swing1Motion=Free
  swing2LimitDegrees=45
  swing2Motion=Free
  twistLimitDegrees=45
  twistMotion=Free
 }
 {
  bone1Name=spine_05
  bone2Name=spine_04
  swing1LimitDegrees=45
  swing1Motion=Free
  swing2LimitDegrees=45
  swing2Motion=Free
  twistLimitDegrees=45
  twistMotion=Free
 }
 {
  bone1Name=thigh_l
  bone2Name=pelvis
  swing1LimitDegrees=45
  swing1Motion=Free
  swing2LimitDegrees=45
  swing2Motion=Free
  twistLimitDegrees=45
  twistMotion=Free
 }
 {
  bone1Name=thigh_r
  bone2Name=pelvis
  swing1LimitDegrees=45
  swing1Motion=Free
  swing2LimitDegrees=45
  swing2Motion=Free
  twistLimitDegrees=45
  twistMotion=Free
 }
 {
  bone1Name=dyn_l_zipper_chest_02_Offset
  bone2Name=pelvis
  swing1LimitDegrees=45
  swing1Motion=Free
  swing2LimitDegrees=45
  swing2Motion=Free
  twistLimitDegrees=45
  twistMotion=Free
 }
 {
  bone1Name=dyn_l_zipper_chest_01_Offset
  bone2Name=pelvis
  swing1LimitDegrees=45
  swing1Motion=Free
  swing2LimitDegrees=45
  swing2Motion=Free
  twistLimitDegrees=45
  twistMotion=Free
 }
 {
  bone1Name=dyn_r_zipper_chest_02_Offset
  bone2Name=pelvis
  swing1LimitDegrees=45
  swing1Motion=Free
  swing2LimitDegrees=45
  swing2Motion=Free
  twistLimitDegrees=45
  twistMotion=Free
 }
 {
  bone1Name=dyn_r_zipper_chest_01_Offset
  bone2Name=pelvis
  swing1LimitDegrees=45
  swing1Motion=Free
  swing2LimitDegrees=45
  swing2Motion=Free
  twistLimitDegrees=45
  twistMotion=Free
 }
 {
  bone1Name=dyn_r_zipper_leg_Offset
  bone2Name=thigh_r
  swing1LimitDegrees=45
  swing1Motion=Free
  swing2LimitDegrees=45
  swing2Motion=Free
  twistLimitDegrees=45
  twistMotion=Free
 }
 {
  bone1Name=dyn_l_zipper_leg_Offset
  bone2Name=thigh_l
  swing1LimitDegrees=45
  swing1Motion=Free
  swing2LimitDegrees=45
  swing2Motion=Free
  twistLimitDegrees=45
  twistMotion=Free
 }
 {
  bone1Name=dyn_r_zipper_leg
  bone2Name=dyn_r_zipper_leg_Offset
  swing1LimitDegrees=45
  swing2LimitDegrees=45
  twistLimitDegrees=45
  twistMotion=Free
 }
 {
  bone1Name=dyn_l_zipper_leg
  bone2Name=dyn_l_zipper_leg_Offset
  swing1LimitDegrees=45
  swing2LimitDegrees=45
  twistLimitDegrees=45
  twistMotion=Locked
 }
 {bone1Name=dyn_r_zipper_chest_01, bone2Name=dyn_r_zipper_chest_01_Offset, swing1LimitDegrees=45, swing2LimitDegrees=45, twistLimitDegrees=45}
 {bone1Name=dyn_r_zipper_chest_02, bone2Name=dyn_r_zipper_chest_02_Offset, swing1LimitDegrees=45, swing2LimitDegrees=45, twistLimitDegrees=45}
 {bone1Name=dyn_l_zipper_chest_02, bone2Name=dyn_l_zipper_chest_02_Offset, swing1LimitDegrees=45, swing2LimitDegrees=45, twistLimitDegrees=45}
 {bone1Name=dyn_l_zipper_chest_01, bone2Name=dyn_l_zipper_chest_01_Offset, swing1LimitDegrees=45, swing2LimitDegrees=45, twistLimitDegrees=45}
]
