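# Post-process style AnimBlueprint for the PartyGold medium male body cosmetic, animating its jacket, capelet, shoulder tassels and zippers.
# It copies the attached parent mesh's pose, then adds AnimDynamics spring motion to capelet and tassel bones, some only while targeting.
# The output path applies look-driven tassel offsets outside emotes, weighted zipper constraints to spine and thighs, a Control Rig, and movement-state jacket and tassel nudges.
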AnimBlueprint /BRCosmetics/Characters/Player/Male/Medium/Bodies/M_MED_PartyGold/Meshes/M_MED_Party_Gold_AnimBP
target_skeleton=&/M_MED_Party_Gold_legs_LOD0_Skeleton
graphs [
 AnimGraph {nodes [
   n1 = Get Tassels_Rifle_LookUp_Left() -> (Tassels_Rifle_LookUp_Left:double)
   n2 = "Transform (Modify) Bone - Bone: dyn_l_shoulderFray_top_01_a_Offset"(
    ComponentPose=Local To Component(LocalPose=Use cached pose 'Pre_RifleCorrection'())
    Alpha=@n1
    Translation="0,0,3"
    Rotation="-60,-90,0"
    Scale="1,1,1"
   ) {
    rotationMode=BMM_Additive
    scaleMode=BMM_Additive
    translationMode=BMM_Additive
    alphaBoolBlend {blendOption=Linear}
    boneToModify {boneName=dyn_l_shoulderFray_top_01_a_Offset}
    rotation {pitch=-60, yaw=-90}
    translation {z=3}
   }
   n3 = "Transform (Modify) Bone - Bone: dyn_l_shoulderFray_btm_01_a_Offset"(ComponentPose=@n2, Alpha=@n1, Translation="0,0,3", Rotation="-70,-90,0", Scale="1,1,1") {
    rotationMode=BMM_Additive
    scaleMode=BMM_Additive
    translationMode=BMM_Additive
    alphaBoolBlend {blendOption=Linear}
    boneToModify {boneName=dyn_l_shoulderFray_btm_01_a_Offset}
    rotation {pitch=-70, yaw=-90}
    translation {z=3}
   }
   n4 = "Transform (Modify) Bone - Bone: dyn_l_shoulderFray_top_02_a_Offset"(ComponentPose=@n3, Alpha=@n1, Translation="0,0,3", Rotation="-15,-60,0", Scale="1,1,1") {
    rotationMode=BMM_Additive
    scaleMode=BMM_Additive
    translationMode=BMM_Additive
    alphaBoolBlend {blendOption=Linear}
    boneToModify {boneName=dyn_l_shoulderFray_top_02_a_Offset}
    rotation {pitch=-15, yaw=-60}
    translation {z=3}
   }
   n5 = "Transform (Modify) Bone - Bone: dyn_l_shoulderFray_btm_02_a_Offset"(ComponentPose=@n4, Alpha=@n1, Translation="0,0,3", Rotation="-90,-60,0", Scale="1,1,1") {
    rotationMode=BMM_Additive
    scaleMode=BMM_Additive
    translationMode=BMM_Additive
    alphaBoolBlend {blendOption=Linear}
    boneToModify {boneName=dyn_l_shoulderFray_btm_02_a_Offset}
    rotation {pitch=-90, yaw=-60}
    translation {z=3}
   }
   n6 = "Transform (Modify) Bone - Bone: dyn_l_shoulderFray_top_03_a_Offset"(ComponentPose=@n5, Alpha=@n1, Translation="0,0,0", Rotation="0,0,0", Scale="1,1,1") {
    rotationMode=BMM_Additive
    scaleMode=BMM_Additive
    translationMode=BMM_Additive
    alphaBoolBlend {blendOption=Linear}
    boneToModify {boneName=dyn_l_shoulderFray_top_03_a_Offset}
   }
   n7 = "Transform (Modify) Bone - Bone: dyn_l_shoulderFray_btm_03_a_Offset"(ComponentPose=@n6, Alpha=@n1, Translation="0,0,0", Rotation="0,0,0", Scale="1,1,1") {
    rotationMode=BMM_Additive
    scaleMode=BMM_Additive
    translationMode=BMM_Additive
    alphaBoolBlend {blendOption=Linear}
    boneToModify {boneName=dyn_l_shoulderFray_btm_03_a_Offset}
   }
   n8 = "Transform (Modify) Bone - Bone: dyn_l_shoulderFray_top_04_a_Offset"(ComponentPose=@n7, Alpha=@n1, Translation="0,0,0", Rotation="10,0,0", Scale="1,1,1") {
    rotationMode=BMM_Additive
    scaleMode=BMM_Additive
    translationMode=BMM_Additive
    alphaBoolBlend {blendOption=Linear}
    boneToModify {boneName=dyn_l_shoulderFray_top_04_a_Offset}
    rotation {pitch=10}
   }
   n9 = "Transform (Modify) Bone - Bone: dyn_l_shoulderFray_btm_04_a_Offset"(ComponentPose=@n8, Alpha=@n1, Translation="0,0,0", Rotation="10,0,0", Scale="1,1,1") {
    rotationMode=BMM_Additive
    scaleMode=BMM_Additive
    translationMode=BMM_Additive
    alphaBoolBlend {blendOption=Linear}
    boneToModify {boneName=dyn_l_shoulderFray_btm_04_a_Offset}
    rotation {pitch=10}
   }
   n10 = "Transform (Modify) Bone - Bone: dyn_l_shoulderFray_top_05_a_Offset"(ComponentPose=@n9, Alpha=@n1, Translation="0,0,0", Rotation="20,0,0", Scale="1,1,1") {
    rotationMode=BMM_Additive
    scaleMode=BMM_Additive
    translationMode=BMM_Additive
    alphaBoolBlend {blendOption=Linear}
    boneToModify {boneName=dyn_l_shoulderFray_top_05_a_Offset}
    rotation {pitch=20}
   }
   n11 = "Transform (Modify) Bone - Bone: dyn_l_shoulderFray_btm_05_a_Offset"(ComponentPose=@n10, Alpha=@n1, Translation="0,0,0", Rotation="20,0,0", Scale="1,1,1") {
    rotationMode=BMM_Additive
    scaleMode=BMM_Additive
    translationMode=BMM_Additive
    alphaBoolBlend {blendOption=Linear}
    boneToModify {boneName=dyn_l_shoulderFray_btm_05_a_Offset}
    rotation {pitch=20}
   }
   n12 = "Transform (Modify) Bone - Bone: dyn_l_shoulderFray_top_06_a_Offset"(ComponentPose=@n11, Alpha=@n1, Translation="0,0,0", Rotation="25,0,0", Scale="1,1,1") {
    rotationMode=BMM_Additive
    scaleMode=BMM_Additive
    translationMode=BMM_Additive
    alphaBoolBlend {blendOption=Linear}
    boneToModify {boneName=dyn_l_shoulderFray_top_06_a_Offset}
    rotation {pitch=25}
   }
   n13 = "Transform (Modify) Bone - Bone: dyn_l_shoulderFray_btm_06_a_Offset"(ComponentPose=@n12, Alpha=@n1, Translation="0,0,0", Rotation="25,0,0", Scale="1,1,1") {
    rotationMode=BMM_Additive
    scaleMode=BMM_Additive
    translationMode=BMM_Additive
    alphaBoolBlend {blendOption=Linear}
    boneToModify {boneName=dyn_l_shoulderFray_btm_06_a_Offset}
    rotation {pitch=25}
   }
   Blend Poses by bool(
    BlendPose_0=Component To Local(ComponentPose=@n13)
    BlendPose_1=Use cached pose 'Pre_RifleCorrection'()
    bActiveValue=Get isAimingRifle() -> (isAimingRifle:bool)
   ) {blendType=Linear}
   Component To Local(ComponentPose=Local To Component())
   "Constraint - dyn_l_zipper_chest_01_Offset (PPPP)"!(
    ComponentPose
    Alpha=1
    ConstraintWeights_0=0.0362
    ConstraintWeights_1:float=0.2869
    ConstraintWeights_2:float=0.4155
    ConstraintWeights_3:float=0.2614
   ) -> (Pose) {
    alphaBoolBlend {blendOption=Linear}
    boneToModify {boneName=dyn_l_zipper_chest_01_Offset}
    constraintSetup [{targetBone {boneName=spine_01}}, {targetBone {boneName=spine_02}}, {targetBone {boneName=spine_03}}, {targetBone {boneName=spine_04}}]
    constraintWeights [1,1,1,1]
   }
   "Constraint - dyn_r_zipper_chest_01_Offset (PPPPP)"!(
    ComponentPose
    Alpha=1
    ConstraintWeights_0=0.0362
    ConstraintWeights_1:float=0.2412
    ConstraintWeights_2:float=0.4093
    ConstraintWeights_3:float=0.3032
    ConstraintWeights_4:float=0.0102
   ) -> (Pose) {
    alphaBoolBlend {blendOption=Linear}
    boneToModify {boneName=dyn_r_zipper_chest_01_Offset}
    constraintSetup [
     {targetBone {boneName=spine_01}}
     {targetBone {boneName=spine_02}}
     {targetBone {boneName=spine_03}}
     {targetBone {boneName=spine_04}}
     {targetBone {boneName=pelvis}}
    ]
    constraintWeights [1,1,1,1,1]
   }
   "Constraint - dyn_r_zipper_chest_02_Offset (PPPPP)"!(
    ComponentPose
    Alpha=1
    ConstraintWeights_0=0.1258
    ConstraintWeights_1:float=0.2962
    ConstraintWeights_2:float=0.3833
    ConstraintWeights_3:float=0.2218
    ConstraintWeights_4:float=0.0274
   ) -> (Pose) {
    alphaBoolBlend {blendOption=Linear}
    boneToModify {boneName=dyn_r_zipper_chest_02_Offset}
    constraintSetup [
     {targetBone {boneName=spine_01}}
     {targetBone {boneName=spine_02}}
     {targetBone {boneName=spine_03}}
     {targetBone {boneName=spine_04}}
     {targetBone {boneName=pelvis}}
    ]
    constraintWeights [1,1,1,1,1]
   }
   "Transform (Modify) Bone - Bone: dyn_jacket_bk_r_01"!(ComponentPose, bAlphaBoolEnabled:bool=Get bIsKneeSliding(), Translation="0,0,0", Rotation="0,0,33", Scale="1,1,1") -> (Pose) {
    alphaInputType=Bool
    rotationMode=BMM_Additive
    scaleMode=BMM_Additive
    translationMode=BMM_Additive
    alphaBoolBlend {blendInTime=0.1, blendOption=ExpOut, blendOutTime=0.1}
    boneToModify {boneName=dyn_jacket_bk_r_01}
    rotation {roll=33}
   }
   RigidBody() {
    bClampLinearTranslationLimitToRefPose=true
    bOverrideWorldGravity=true
    overridePhysicsAsset=&/M_MED_Party_Gold_Zipper_Physics
    simulationSpace=BaseBoneSpace
    alphaBoolBlend {blendOption=Linear}
    baseBoneRef {boneName=pelvis}
    componentAppliedLinearAccClamp=<5000 5000 5000>
    componentLinearAccScale=<0.5 0.5 0.5>
    componentLinearVelScale=<0.5 0.5 0.5>
    overrideWorldGravity {z=-980}
   }
   RigidBody(Alpha=SelectFloat(A=0.6, B=0.6, bPickA=Get bIsTargeting())) {
    bClampLinearTranslationLimitToRefPose=true
    bOverrideWorldGravity=true
    overridePhysicsAsset=&/M_MED_Party_Gold_ShoulderTassel_Physics
    simulationSpace=BaseBoneSpace
    alphaBoolBlend {blendOption=Linear}
    baseBoneRef {boneName=root}
    componentLinearAccScale=<0.1 0.1 0.5>
    overrideWorldGravity {z=-980}
   }
   RigidBody(Alpha=Get rigids alpha() -> (rigids alpha:double)) {
    bClampLinearTranslationLimitToRefPose=true
    bOverrideWorldGravity=true
    overridePhysicsAsset=&/M_MED_Party_Gold_Jacket_Physics
    simulationSpace=BaseBoneSpace
    alphaBoolBlend {blendOption=Linear}
    baseBoneRef {boneName=root}
    componentAppliedLinearAccClamp=<3000 3000 3000>
    componentLinearAccScale=<0.4 0.4 0.4>
    componentLinearVelScale=<0.4 0.3 0.3>
    overrideWorldGravity {z=-980}
   }
   n14 = Get Tassels_LookLevel() -> (Tassels_LookLevel:double)
   n15 = "Transform (Modify) Bone - Bone: dyn_r_shoulderFray_top_01_a_Offset"(
    ComponentPose=Local To Component(LocalPose=Use cached pose 'Anim_branch'())
    Alpha=@n14
    Translation="0,-2,0"
    Rotation="0,0,0"
    Scale="1,1,1"
   ) {
    rotationMode=BMM_Additive
    scaleMode=BMM_Additive
    translationMode=BMM_Additive
    alphaBoolBlend {blendOption=Linear}
    boneToModify {boneName=dyn_r_shoulderFray_top_01_a_Offset}
    translation {y=-2}
   }
   n16 = "Transform (Modify) Bone - Bone: dyn_r_shoulderFray_btm_01_a_Offset"(ComponentPose=@n15, Alpha=@n14, Translation="0,-2,0", Rotation="0,0,0", Scale="1,1,1") {
    rotationMode=BMM_Additive
    scaleMode=BMM_Additive
    translationMode=BMM_Additive
    alphaBoolBlend {blendOption=Linear}
    boneToModify {boneName=dyn_r_shoulderFray_btm_01_a_Offset}
    translation {y=-2}
   }
   n17 = Get Tassels_LookUp_Right() -> (Tassels_LookUp_Right:double)
   n18 = "Transform (Modify) Bone - Bone: dyn_r_shoulderFray_top_01_a_Offset"(ComponentPose=@n16, Alpha=@n17, Translation="0,-4,4", Rotation="20,-15,-25", Scale="1,1,1") {
    rotationMode=BMM_Additive
    scaleMode=BMM_Additive
    translationMode=BMM_Additive
    alphaBoolBlend {blendOption=Linear}
    boneToModify {boneName=dyn_r_shoulderFray_top_01_a_Offset}
    rotation=<rot 20 -15 -25>
    translation {y=-4, z=4}
   }
   n19 = "Transform (Modify) Bone - Bone: dyn_r_shoulderFray_btm_01_a_Offset"(ComponentPose=@n18, Alpha=@n17, Translation="0,-5,4", Rotation="50,-20,0", Scale="1,1,1") {
    rotationMode=BMM_Additive
    scaleMode=BMM_Additive
    translationMode=BMM_Additive
    alphaBoolBlend {blendOption=Linear}
    boneToModify {boneName=dyn_r_shoulderFray_btm_01_a_Offset}
    rotation {pitch=50, yaw=-20}
    translation {y=-5, z=4}
   }
   n20 = "Transform (Modify) Bone - Bone: dyn_r_shoulderFray_top_02_a_Offset"(ComponentPose=@n19, Alpha=@n17, Translation="0,-2,2", Rotation="0,-10,0", Scale="1,1,1") {
    rotationMode=BMM_Additive
    scaleMode=BMM_Additive
    translationMode=BMM_Additive
    alphaBoolBlend {blendOption=Linear}
    boneToModify {boneName=dyn_r_shoulderFray_top_02_a_Offset}
    rotation {yaw=-10}
    translation {y=-2, z=2}
   }
   n21 = "Transform (Modify) Bone - Bone: dyn_r_shoulderFray_btm_02_a_Offset"(ComponentPose=@n20, Alpha=@n17, Translation="0,-2,1", Rotation="0,-10,10", Scale="1,1,1") {
    rotationMode=BMM_Additive
    scaleMode=BMM_Additive
    translationMode=BMM_Additive
    alphaBoolBlend {blendOption=Linear}
    boneToModify {boneName=dyn_r_shoulderFray_btm_02_a_Offset}
    rotation {roll=10, yaw=-10}
    translation {y=-2, z=1}
   }
   n22 = "Transform (Modify) Bone - Bone: dyn_r_shoulderFray_top_03_a_Offset"(ComponentPose=@n21, Alpha=@n17, Translation="0,-2,1", Rotation="0,0,0", Scale="1,1,0.71") {
    rotationMode=BMM_Additive
    scaleMode=BMM_Additive
    translationMode=BMM_Additive
    alphaBoolBlend {blendOption=Linear}
    boneToModify {boneName=dyn_r_shoulderFray_top_03_a_Offset}
    scale {z=0.71}
    translation {y=-2, z=1}
   }
   n23 = "Transform (Modify) Bone - Bone: dyn_r_shoulderFray_btm_03_a_Offset"(ComponentPose=@n22, Alpha=@n17, Translation="0,-2,1", Rotation="0,0,0", Scale="1,1,1") {
    rotationMode=BMM_Additive
    scaleMode=BMM_Additive
    translationMode=BMM_Additive
    alphaBoolBlend {blendOption=Linear}
    boneToModify {boneName=dyn_r_shoulderFray_btm_03_a_Offset}
    translation {y=-2, z=1}
   }
   n24 = "Transform (Modify) Bone - Bone: dyn_r_shoulderFray_btm_02_a_Offset"(
    ComponentPose=@n23
    Alpha=Get Tassels_LookUp_Right_TwoCorrect() -> (Tassels_LookUp_Right_TwoCorrect:double)
    Translation="0,-1,0.5"
    Rotation="-25,0.000,0.000"
    Scale="1,1,1"
   ) {
    rotationMode=BMM_Additive
    scaleMode=BMM_Additive
    translationMode=BMM_Additive
    alphaBoolBlend {blendOption=Linear}
    boneToModify {boneName=dyn_r_shoulderFray_btm_02_a_Offset}
    rotation {pitch=-25}
    translation {y=-1, z=0.5}
   }
   n25 = Get Tassels_LookUp_Left() -> (Tassels_LookUp_Left:double)
   n26 = "Transform (Modify) Bone - Bone: dyn_l_shoulderFray_top_01_a_Offset"(ComponentPose=@n24, Alpha=@n25, Translation="0,0,2", Rotation="0,0,0", Scale="1,1,1") {
    rotationMode=BMM_Additive
    scaleMode=BMM_Additive
    translationMode=BMM_Additive
    alphaBoolBlend {blendOption=Linear}
    boneToModify {boneName=dyn_l_shoulderFray_top_01_a_Offset}
    translation {z=2}
   }
   n27 = "Transform (Modify) Bone - Bone: dyn_l_shoulderFray_btm_01_a_Offset"(ComponentPose=@n26, Alpha=@n25, Translation="0,0,2", Rotation="0,0,0", Scale="1,1,1") {
    rotationMode=BMM_Additive
    scaleMode=BMM_Additive
    translationMode=BMM_Additive
    alphaBoolBlend {blendOption=Linear}
    boneToModify {boneName=dyn_l_shoulderFray_btm_01_a_Offset}
    translation {z=2}
   }
   n28 = "Transform (Modify) Bone - Bone: dyn_l_shoulderFray_top_02_a_Offset"(ComponentPose=@n27, Alpha=@n25, Translation="0,0,1", Rotation="0,0,0", Scale="1,1,1") {
    rotationMode=BMM_Additive
    scaleMode=BMM_Additive
    translationMode=BMM_Additive
    alphaBoolBlend {blendOption=Linear}
    boneToModify {boneName=dyn_l_shoulderFray_top_02_a_Offset}
    translation {z=1}
   }
   n29 = "Transform (Modify) Bone - Bone: dyn_l_shoulderFray_btm_02_a_Offset"(ComponentPose=@n28, Alpha=@n25, Translation="0,0,1", Rotation="0,0,0", Scale="1,1,1") {
    rotationMode=BMM_Additive
    scaleMode=BMM_Additive
    translationMode=BMM_Additive
    alphaBoolBlend {blendOption=Linear}
    boneToModify {boneName=dyn_l_shoulderFray_btm_02_a_Offset}
    translation {z=1}
   }
   n30 = "Transform (Modify) Bone - Bone: dyn_l_shoulderFray_top_03_a_Offset"(ComponentPose=@n29, Alpha=@n25, Translation="0,0,0", Rotation="0,0,0", Scale="1,1,1") {
    rotationMode=BMM_Additive
    scaleMode=BMM_Additive
    translationMode=BMM_Additive
    alphaBoolBlend {blendOption=Linear}
    boneToModify {boneName=dyn_l_shoulderFray_top_03_a_Offset}
   }
   n31 = "Transform (Modify) Bone - Bone: dyn_l_shoulderFray_btm_03_a_Offset"(ComponentPose=@n30, Alpha=@n25, Translation="0,0,0", Rotation="0,0,0", Scale="1,1,1") {
    rotationMode=BMM_Additive
    scaleMode=BMM_Additive
    translationMode=BMM_Additive
    alphaBoolBlend {blendOption=Linear}
    boneToModify {boneName=dyn_l_shoulderFray_btm_03_a_Offset}
   }
   n32 = "Transform (Modify) Bone - Bone: dyn_r_shoulderFray_btm_01_a_Offset"!(ComponentPose=@n31, bAlphaBoolEnabled:bool=Get bIsCrouching(), Translation="0,0,2", Rotation="0,0,0", Scale="1,1,1") -> (Pose) {
    alphaInputType=Bool
    rotationMode=BMM_Additive
    scaleMode=BMM_Additive
    translationMode=BMM_Additive
    alphaBoolBlend {blendOption=Linear}
    boneToModify {boneName=dyn_r_shoulderFray_btm_01_a_Offset}
    translation {z=2}
   }
   n33 = Component To Local(ComponentPose=@n32)
   n34 = Use cached pose 'Anim_branch'()
   n35 = Get bIsPlayingEmote()
   n36 = "Constraint - dyn_l_zipper_chest_01_Offset (PPPP)"(
    ComponentPose=Local To Component(LocalPose=Blend Poses by bool(BlendPose_0=@n34, BlendPose_1=@n33, bActiveValue=@n35) {blendType=Linear})
    Alpha=1
    ConstraintWeights_0=0.0362
    ConstraintWeights_1=0.2869
    ConstraintWeights_2=0.4155
    ConstraintWeights_3=0.2614
   ) {
    alphaBoolBlend {blendOption=Linear}
    boneToModify {boneName=dyn_l_zipper_chest_01_Offset}
    constraintSetup [{targetBone {boneName=spine_01}}, {targetBone {boneName=spine_02}}, {targetBone {boneName=spine_03}}, {targetBone {boneName=spine_04}}]
    constraintWeights [1,1,1,1]
   }
   n37 = "Constraint - dyn_l_zipper_chest_02_Offset (PPPPP)"!(
    ComponentPose=@n36
    Alpha=1
    ConstraintWeights_0=0.1258
    ConstraintWeights_1:float=0.3342
    ConstraintWeights_2:float=0.3487
    ConstraintWeights_3:float=0.1594
    ConstraintWeights_4:float=0.0319
   ) -> (Pose) {
    alphaBoolBlend {blendOption=Linear}
    boneToModify {boneName=dyn_l_zipper_chest_02_Offset}
    constraintSetup [
     {targetBone {boneName=spine_01}}
     {targetBone {boneName=spine_02}}
     {targetBone {boneName=spine_03}}
     {targetBone {boneName=spine_04}}
     {targetBone {boneName=pelvis}}
    ]
    constraintWeights [1,1,1,1,1]
   }
   n38 = "Constraint - dyn_r_zipper_chest_01_Offset (PPPPP)"(
    ComponentPose=@n37
    Alpha=1
    ConstraintWeights_0=0.0102
    ConstraintWeights_1=0.0361
    ConstraintWeights_2=0.2412
    ConstraintWeights_3=0.4093
    ConstraintWeights_4=0.3032
   ) {
    alphaBoolBlend {blendOption=Linear}
    boneToModify {boneName=dyn_r_zipper_chest_01_Offset}
    constraintSetup [
     {targetBone {boneName=pelvis}}
     {targetBone {boneName=spine_01}}
     {targetBone {boneName=spine_02}}
     {targetBone {boneName=spine_03}}
     {targetBone {boneName=spine_04}}
    ]
    constraintWeights [1,1,1,1,1]
   }
   n39 = "Constraint - dyn_r_zipper_chest_02_Offset (PPPPP)"(
    ComponentPose=@n38
    Alpha=1
    ConstraintWeights_0=0.0274
    ConstraintWeights_1=0.0714
    ConstraintWeights_2=0.2962
    ConstraintWeights_3=0.3833
    ConstraintWeights_4=0.2218
   ) {
    alphaBoolBlend {blendOption=Linear}
    boneToModify {boneName=dyn_r_zipper_chest_02_Offset}
    constraintSetup [
     {targetBone {boneName=pelvis}}
     {targetBone {boneName=spine_01}}
     {targetBone {boneName=spine_02}}
     {targetBone {boneName=spine_03}}
     {targetBone {boneName=spine_04}}
    ]
    constraintWeights [1,1,1,1,1]
   }
   n40 = "Constraint - dyn_r_zipper_leg_Offset (PP)"!(ComponentPose=@n39, Alpha=1, ConstraintWeights_0=0.5574, ConstraintWeights_1:float=0.4426) -> (Pose) {
    alphaBoolBlend {blendOption=Linear}
    boneToModify {boneName=dyn_r_zipper_leg_Offset}
    constraintSetup [{targetBone {boneName=thigh_r}}, {targetBone {boneName=thigh_twist_01_r}}]
    constraintWeights [1,1]
   }
   n41 = "Constraint - dyn_l_zipper_leg_Offset (PP)"!(ComponentPose=@n40, Alpha=1, ConstraintWeights_0=0.5222, ConstraintWeights_1:float=0.4778) -> (Pose) {
    alphaBoolBlend {blendOption=Linear}
    boneToModify {boneName=dyn_l_zipper_leg_Offset}
    constraintSetup [{targetBone {boneName=thigh_l}}, {targetBone {boneName=thigh_twist_01_l}}]
    constraintWeights [1,1]
   }
   n42 = Control Rig!(
    Source=Component To Local(ComponentPose=@n41)
    Alpha
    ParachuteOffset:double=Get Parachute Offset() -> (Parachute Offset:double)
    CrouchingOffset:double=Get Crouching Offset() -> (Crouching Offset:double)
   ) -> (Pose) {
    alphaBoolBlend {blendOption=Linear}
    controlRigAssetReference {blueprintRigClass=&/M_MED_Party_Gold_CtrlRig.M_MED_Party_Gold_CtrlRig_C}
   }
   n43 = "Transform (Modify) Bone - Bone: dyn_jacket_fr_l_01"!(
    ComponentPose=Local To Component(LocalPose=@n42)
    bAlphaBoolEnabled:bool=Get bIsBaseballSliding()
    Translation="0,-10,0"
    Rotation="0,-15,-10"
    Scale="1,1,1"
   ) -> (Pose) {
    alphaInputType=Bool
    rotationMode=BMM_Additive
    rotationSpace=BCS_BoneSpace
    scaleMode=BMM_Additive
    scaleSpace=BCS_BoneSpace
    translationMode=BMM_Additive
    translationSpace=BCS_BoneSpace
    alphaBoolBlend {blendInTime=0.1, blendOption=Linear, blendOutTime=0.1}
    boneToModify {boneName=dyn_jacket_fr_l_01}
    rotation {roll=-10, yaw=-15}
    translation {y=-10}
   }
   n44 = "Transform (Modify) Bone - Bone: dyn_jacket_fr_l_01"(
    ComponentPose=@n43
    bAlphaBoolEnabled=Get CrouchingAndMoving() -> (CrouchingAndMoving:bool)
    Translation="4,0,0"
    Rotation="0,0,0"
    Scale="1,1,1"
   ) {
    alphaInputType=Bool
    rotationMode=BMM_Additive
    rotationSpace=BCS_BoneSpace
    scaleMode=BMM_Additive
    scaleSpace=BCS_BoneSpace
    translationMode=BMM_Additive
    translationSpace=BCS_BoneSpace
    alphaBoolBlend {blendInTime=2, blendOption=Custom, blendOutTime=1}
    boneToModify {boneName=dyn_jacket_fr_l_01}
    translation {x=4}
   }
   n45 = "Transform (Modify) Bone - Bone: dyn_jacket_l_01"!(ComponentPose=@n44, bAlphaBoolEnabled:bool=Get bIsMoving2D(), Translation="0,0,2", Rotation="0,0,0", Scale="1,1,1") -> (Pose) {
    alphaInputType=Bool
    rotationMode=BMM_Additive
    rotationSpace=BCS_BoneSpace
    scaleMode=BMM_Additive
    scaleSpace=BCS_BoneSpace
    translationMode=BMM_Additive
    translationSpace=BCS_BoneSpace
    alphaBoolBlend {blendInTime=0.1, blendOption=Linear, blendOutTime=0.3}
    boneToModify {boneName=dyn_jacket_l_01}
    translation {z=2}
   }
   n46 = "Transform (Modify) Bone - Bone: dyn_jacket_bk_r_01"(ComponentPose=@n45, bAlphaBoolEnabled=Get bIsKneeSliding(), Translation="0,0,0", Rotation="30,0,0", Scale="1,1,1") {
    alphaInputType=Bool
    rotationMode=BMM_Additive
    rotationSpace=BCS_BoneSpace
    scaleMode=BMM_Additive
    scaleSpace=BCS_BoneSpace
    translationMode=BMM_Additive
    translationSpace=BCS_BoneSpace
    alphaBoolBlend {blendInTime=0.1, blendOption=Linear, blendOutTime=0.3}
    boneToModify {boneName=dyn_jacket_bk_r_01}
    rotation {pitch=30}
   }
   n47 = "Transform (Modify) Bone - Bone: dyn_jacket_fr_r_01"!(ComponentPose=@n46, bAlphaBoolEnabled:bool=Get Idle() -> (Idle:bool), Translation="0,0,0", Rotation="-25,0,0", Scale="1,1,1") -> (Pose) {
    alphaInputType=Bool
    rotationMode=BMM_Additive
    rotationSpace=BCS_BoneSpace
    scaleMode=BMM_Additive
    scaleSpace=BCS_BoneSpace
    translationMode=BMM_Additive
    translationSpace=BCS_BoneSpace
    alphaBoolBlend {blendInTime=0.1, blendOption=Linear, blendOutTime=0.5}
    boneToModify {boneName=dyn_jacket_fr_r_01}
    rotation {pitch=-25}
   }
   n48 = Get TargetingCrouchMoving() -> (TargetingCrouchMoving:bool)
   n49 = "Transform (Modify) Bone - Bone: dyn_r_shoulderFray_btm_01_a_Offset"(ComponentPose=@n47, bAlphaBoolEnabled=NOT Boolean(A=@n48), Translation="0,0,0", Rotation="-25,0,0", Scale="1,1,1") {
    alphaInputType=Bool
    rotationMode=BMM_Additive
    rotationSpace=BCS_BoneSpace
    scaleMode=BMM_Additive
    scaleSpace=BCS_BoneSpace
    translationMode=BMM_Additive
    translationSpace=BCS_BoneSpace
    alphaBoolBlend {blendInTime=0.5, blendOption=Linear, blendOutTime=0.2}
    boneToModify {boneName=dyn_r_shoulderFray_btm_01_a_Offset}
    rotation {pitch=-25}
   }
   n50 = "Transform (Modify) Bone - Bone: dyn_jacket_fr_r_01"(ComponentPose=@n49, bAlphaBoolEnabled=@n48, Translation="0,0,0", Rotation="0,0,0", Scale="1,1,1") {
    alphaInputType=Bool
    rotationMode=BMM_Additive
    rotationSpace=BCS_BoneSpace
    scaleMode=BMM_Additive
    scaleSpace=BCS_BoneSpace
    translationMode=BMM_Additive
    translationSpace=BCS_BoneSpace
    alphaBoolBlend {blendInTime=0.1, blendOption=Linear, blendOutTime=0.1}
    boneToModify {boneName=dyn_jacket_fr_r_01}
   }
   Output Pose(Result=Component To Local(ComponentPose=@n50)) {layerGroup=None}
   Save cached pose 'Pre_RifleCorrection'() {cacheName=Pre_RifleCorrection}
   n51 = Get NotTargeting() -> (NotTargeting:bool)
   n52 = Anim Dynamics - Bone: dyn_capelet_b_01(
    ComponentPose=Local To Component(LocalPose=Copy Pose From Mesh() {bCopyCurves=true, bUseAttachedParent=true})
    Alpha=0.5
    ExternalForce="0,0,0"
   ) {
    angularDampingOverride=0.9999
    angularSpringConstant=35
    bAngularSpring=true
    bGravityOverrideInSimSpace=true
    bOverrideAngularDamping=true
    gravityScale=0.3
    lODThreshold=1
    alphaBoolBlend {blendOption=Linear}
    boundBone {boneName=dyn_capelet_b_01}
    componentLinearAccScale=<0.25 0.25 0.6>
    physicsBodyDefinitions [{
      sphereCollisionRadius=0
      boundBone {boneName=dyn_capelet_b_01}
      boxExtents=<20 20 5>
      constraintSetup {twistAxis=AxisZ, angularLimitsMax=<5 45 45>, angularLimitsMin=<-5 -45 -45>, angularTarget=<0 0 0>}
      localJointOffset=<0 0 -5>
     }]
    relativeSpaceBone {boneName=root}
   }
   n53 = Anim Dynamics - Bone: dyn_capelet_b_02(ComponentPose=@n52, Alpha=0.5, ExternalForce="0,0,0") {
    angularDampingOverride=0.9999
    angularSpringConstant=35
    bAngularSpring=true
    bGravityOverrideInSimSpace=true
    bOverrideAngularDamping=true
    gravityScale=0.3
    lODThreshold=1
    alphaBoolBlend {blendOption=Linear}
    boundBone {boneName=dyn_capelet_b_02}
    componentLinearAccScale=<0.25 0.25 0.6>
    physicsBodyDefinitions [{
      sphereCollisionRadius=0
      boundBone {boneName=dyn_capelet_b_02}
      boxExtents=<25 25 10>
      constraintSetup {twistAxis=AxisZ, angularLimitsMax=<5 45 45>, angularLimitsMin=<-5 -45 -45>, angularTarget=<0 0 0>}
      localJointOffset=<0 0 -5>
     }]
    relativeSpaceBone {boneName=root}
   }
   n54 = Anim Dynamics - Bone: dyn_capelet_l(ComponentPose=@n53, Alpha=0.35, ExternalForce="0,0,0") {
    angularDampingOverride=0.9999
    angularSpringConstant=30
    bAngularSpring=true
    bGravityOverrideInSimSpace=true
    bOverrideAngularDamping=true
    gravityScale=0.3
    lODThreshold=1
    alphaBoolBlend {blendOption=Linear}
    boundBone {boneName=dyn_capelet_l}
    componentLinearAccScale=<0.25 0.25 0.6>
    physicsBodyDefinitions [{
      sphereCollisionRadius=0
      boundBone {boneName=dyn_capelet_l}
      boxExtents=<30 30 5>
      constraintSetup {twistAxis=AxisZ, angularLimitsMax=<45 45 45>, angularLimitsMin=<-45 -45 -45>, angularTarget=<0 0 0>}
      localJointOffset=<0 0 -5>
     }]
    relativeSpaceBone {boneName=root}
   }
   n55 = Anim Dynamics - Bone: dyn_capelet_r(ComponentPose=@n54, Alpha=0.09, ExternalForce="0,0,0") {
    angularDampingOverride=0.9999
    angularSpringConstant=25
    bAngularSpring=true
    bGravityOverrideInSimSpace=true
    bOverrideAngularDamping=true
    gravityScale=0.3
    lODThreshold=1
    alphaBoolBlend {blendOption=Linear}
    boundBone {boneName=dyn_capelet_r}
    componentLinearAccScale=<0.25 0.25 0.6>
    physicsBodyDefinitions [{
      sphereCollisionRadius=0
      boundBone {boneName=dyn_capelet_r}
      boxExtents=<10 10 5>
      constraintSetup {twistAxis=AxisZ, angularLimitsMax=<45 45 45>, angularLimitsMin=<-45 -45 -45>, angularTarget=<0 0 0>}
      localJointOffset=<0 0 -5>
     }]
    relativeSpaceBone {boneName=root}
   }
   n56 = Get bIsTargeting()
   n57 = Anim Dynamics - Bone: dyn_l_shoulderFray_btm_06_a!(ComponentPose=@n55, bAlphaBoolEnabled:bool=@n56, ExternalForce="0,0,0") -> (Pose) {
    alphaInputType=Bool
    angularDampingOverride=0.9999
    angularSpringConstant=75
    bAngularSpring=true
    bGravityOverrideInSimSpace=true
    bOverrideAngularDamping=true
    bUseGravityOverride=true
    alphaBoolBlend {blendInTime=0.3, blendOption=Linear, blendOutTime=0.3}
    boundBone {boneName=dyn_l_shoulderFray_btm_06_a}
    componentAppliedLinearAccClamp=<3000 3000 3000>
    componentLinearAccScale=<0.5 0.5 0.5>
    componentLinearVelScale=<0.6 0.6 0.6>
    gravityOverride {z=-500}
    physicsBodyDefinitions [{
      sphereCollisionRadius=0
      boundBone {boneName=dyn_l_shoulderFray_btm_06_a}
      boxExtents=<20 15 15>
      constraintSetup {angularLimitsMax=<25 10 25>, angularLimitsMin=<-25 -25 -25>, angularTarget=<0 0 0>}
      localJointOffset=<2 0 0>
     }]
   }
   n58 = Anim Dynamics - Bone: dyn_l_shoulderFray_btm_06_b!(ComponentPose=@n57, bAlphaBoolEnabled:bool=@n56, ExternalForce="0,0,0") -> (Pose) {
    alphaInputType=Bool
    angularDampingOverride=0.9999
    angularSpringConstant=75
    bAngularSpring=true
    bGravityOverrideInSimSpace=true
    bOverrideAngularDamping=true
    bUseGravityOverride=true
    alphaBoolBlend {blendInTime=0.3, blendOption=Linear, blendOutTime=0.3}
    boundBone {boneName=dyn_l_shoulderFray_btm_06_b}
    componentAppliedLinearAccClamp=<3000 3000 3000>
    componentLinearAccScale=<0.5 0.5 0.5>
    componentLinearVelScale=<0.6 0.6 0.6>
    gravityOverride {z=-500}
    physicsBodyDefinitions [{
      sphereCollisionRadius=0
      boundBone {boneName=dyn_l_shoulderFray_btm_06_b}
      boxExtents=<20 15 15>
      constraintSetup {angularLimitsMax=<25 10 25>, angularLimitsMin=<-25 -25 -25>, angularTarget=<0 0 0>}
      localJointOffset=<2 0 0>
     }]
   }
   n59 = Anim Dynamics - Bone: dyn_l_shoulderFray_top_06_a!(ComponentPose=@n58, bAlphaBoolEnabled:bool=@n56, ExternalForce="0,0,0") -> (Pose) {
    alphaInputType=Bool
    angularDampingOverride=0.9999
    angularSpringConstant=75
    bAngularSpring=true
    bGravityOverrideInSimSpace=true
    bOverrideAngularDamping=true
    bUseGravityOverride=true
    alphaBoolBlend {blendInTime=0.3, blendOption=Linear, blendOutTime=0.3}
    boundBone {boneName=dyn_l_shoulderFray_top_06_a}
    componentAppliedLinearAccClamp=<3000 3000 3000>
    componentLinearAccScale=<0.5 0.5 0.5>
    componentLinearVelScale=<0.6 0.6 0.6>
    gravityOverride {z=-500}
    physicsBodyDefinitions [{
      sphereCollisionRadius=0
      boundBone {boneName=dyn_l_shoulderFray_top_06_a}
      boxExtents=<20 15 15>
      constraintSetup {angularLimitsMax=<25 10 25>, angularLimitsMin=<-25 -25 -25>, angularTarget=<0 0 0>}
      localJointOffset=<2 0 0>
     }]
   }
   n60 = Anim Dynamics - Bone: dyn_l_shoulderFray_top_06_b!(ComponentPose=@n59, bAlphaBoolEnabled:bool=@n56, ExternalForce="0,0,0") -> (Pose) {
    alphaInputType=Bool
    angularDampingOverride=0.9999
    angularSpringConstant=75
    bAngularSpring=true
    bGravityOverrideInSimSpace=true
    bOverrideAngularDamping=true
    bUseGravityOverride=true
    alphaBoolBlend {blendInTime=0.3, blendOption=Linear, blendOutTime=0.3}
    boundBone {boneName=dyn_l_shoulderFray_top_06_b}
    componentAppliedLinearAccClamp=<3000 3000 3000>
    componentLinearAccScale=<0.5 0.5 0.5>
    componentLinearVelScale=<0.6 0.6 0.6>
    gravityOverride {z=-500}
    physicsBodyDefinitions [{
      sphereCollisionRadius=0
      boundBone {boneName=dyn_l_shoulderFray_top_06_b}
      boxExtents=<20 15 15>
      constraintSetup {angularLimitsMax=<25 10 25>, angularLimitsMin=<-25 -25 -25>, angularTarget=<0 0 0>}
      localJointOffset=<2 0 0>
     }]
   }
   n61 = Anim Dynamics - Bone: dyn_l_shoulderFray_top_05_a!(ComponentPose=@n60, bAlphaBoolEnabled:bool=@n56, ExternalForce="0,0,0") -> (Pose) {
    alphaInputType=Bool
    angularDampingOverride=0.9999
    angularSpringConstant=75
    bAngularSpring=true
    bGravityOverrideInSimSpace=true
    bOverrideAngularDamping=true
    bUseGravityOverride=true
    alphaBoolBlend {blendInTime=0.3, blendOption=Linear, blendOutTime=0.3}
    boundBone {boneName=dyn_l_shoulderFray_top_05_a}
    componentAppliedLinearAccClamp=<3000 3000 3000>
    componentLinearAccScale=<0.5 0.5 0.5>
    componentLinearVelScale=<0.6 0.6 0.6>
    gravityOverride {z=-500}
    physicsBodyDefinitions [{
      sphereCollisionRadius=0
      boundBone {boneName=dyn_l_shoulderFray_top_05_a}
      boxExtents=<20 15 15>
      constraintSetup {angularLimitsMax=<25 10 25>, angularLimitsMin=<-25 -25 -25>, angularTarget=<0 0 0>}
      localJointOffset=<2 0 0>
     }]
   }
   n62 = Anim Dynamics - Bone: dyn_l_shoulderFray_top_05_b!(ComponentPose=@n61, bAlphaBoolEnabled:bool=@n56, ExternalForce="0,0,0") -> (Pose) {
    alphaInputType=Bool
    angularDampingOverride=0.9999
    angularSpringConstant=75
    bAngularSpring=true
    bGravityOverrideInSimSpace=true
    bOverrideAngularDamping=true
    bUseGravityOverride=true
    alphaBoolBlend {blendInTime=0.3, blendOption=Linear, blendOutTime=0.3}
    boundBone {boneName=dyn_l_shoulderFray_top_05_b}
    componentAppliedLinearAccClamp=<3000 3000 3000>
    componentLinearAccScale=<0.5 0.5 0.5>
    componentLinearVelScale=<0.6 0.6 0.6>
    gravityOverride {z=-500}
    physicsBodyDefinitions [{
      sphereCollisionRadius=0
      boundBone {boneName=dyn_l_shoulderFray_top_05_b}
      boxExtents=<20 15 15>
      constraintSetup {angularLimitsMax=<25 10 25>, angularLimitsMin=<-25 -25 -25>, angularTarget=<0 0 0>}
      localJointOffset=<2 0 0>
     }]
   }
   n63 = Anim Dynamics - Bone: dyn_l_shoulderFray_btm_05_a!(ComponentPose=@n62, bAlphaBoolEnabled:bool=@n56, ExternalForce="0,0,0") -> (Pose) {
    alphaInputType=Bool
    angularDampingOverride=0.9999
    angularSpringConstant=75
    bAngularSpring=true
    bGravityOverrideInSimSpace=true
    bOverrideAngularDamping=true
    bUseGravityOverride=true
    alphaBoolBlend {blendInTime=0.3, blendOption=Linear, blendOutTime=0.3}
    boundBone {boneName=dyn_l_shoulderFray_btm_05_a}
    componentAppliedLinearAccClamp=<3000 3000 3000>
    componentLinearAccScale=<0.5 0.5 0.5>
    componentLinearVelScale=<0.6 0.6 0.6>
    gravityOverride {z=-500}
    physicsBodyDefinitions [{
      sphereCollisionRadius=0
      boundBone {boneName=dyn_l_shoulderFray_btm_05_a}
      boxExtents=<20 15 15>
      constraintSetup {angularLimitsMax=<25 10 25>, angularLimitsMin=<-25 -25 -25>, angularTarget=<0 0 0>}
      localJointOffset=<2 0 0>
     }]
   }
   n64 = Anim Dynamics - Bone: dyn_l_shoulderFray_btm_05_b!(ComponentPose=@n63, bAlphaBoolEnabled:bool=@n56, ExternalForce="0,0,0") -> (Pose) {
    alphaInputType=Bool
    angularDampingOverride=0.9999
    angularSpringConstant=75
    bAngularSpring=true
    bGravityOverrideInSimSpace=true
    bOverrideAngularDamping=true
    bUseGravityOverride=true
    alphaBoolBlend {blendInTime=0.3, blendOption=Linear, blendOutTime=0.3}
    boundBone {boneName=dyn_l_shoulderFray_btm_05_b}
    componentAppliedLinearAccClamp=<3000 3000 3000>
    componentLinearAccScale=<0.5 0.5 0.5>
    componentLinearVelScale=<0.6 0.6 0.6>
    gravityOverride {z=-500}
    physicsBodyDefinitions [{
      sphereCollisionRadius=0
      boundBone {boneName=dyn_l_shoulderFray_btm_05_b}
      boxExtents=<20 15 15>
      constraintSetup {angularLimitsMax=<25 10 25>, angularLimitsMin=<-25 -25 -25>, angularTarget=<0 0 0>}
      localJointOffset=<2 0 0>
     }]
   }
   n65 = Anim Dynamics - Bone: dyn_l_shoulderFray_top_04_a!(ComponentPose=@n64, bAlphaBoolEnabled:bool=@n56, ExternalForce="0,0,0") -> (Pose) {
    alphaInputType=Bool
    angularDampingOverride=0.9999
    angularSpringConstant=75
    bAngularSpring=true
    bGravityOverrideInSimSpace=true
    bOverrideAngularDamping=true
    bUseGravityOverride=true
    alphaBoolBlend {blendInTime=0.3, blendOption=Linear, blendOutTime=0.3}
    boundBone {boneName=dyn_l_shoulderFray_top_04_a}
    componentAppliedLinearAccClamp=<3000 3000 3000>
    componentLinearAccScale=<0.5 0.5 0.5>
    componentLinearVelScale=<0.6 0.6 0.6>
    gravityOverride {z=-500}
    physicsBodyDefinitions [{
      sphereCollisionRadius=0
      boundBone {boneName=dyn_l_shoulderFray_top_04_a}
      boxExtents=<20 15 15>
      constraintSetup {angularLimitsMax=<25 10 25>, angularLimitsMin=<-25 -25 -25>, angularTarget=<0 0 0>}
      localJointOffset=<2 0 0>
     }]
   }
   n66 = Anim Dynamics - Bone: dyn_l_shoulderFray_top_04_b!(ComponentPose=@n65, bAlphaBoolEnabled:bool=@n56, ExternalForce="0,0,0") -> (Pose) {
    alphaInputType=Bool
    angularDampingOverride=0.9999
    angularSpringConstant=75
    bAngularSpring=true
    bGravityOverrideInSimSpace=true
    bOverrideAngularDamping=true
    bUseGravityOverride=true
    alphaBoolBlend {blendInTime=0.3, blendOption=Linear, blendOutTime=0.3}
    boundBone {boneName=dyn_l_shoulderFray_top_04_b}
    componentAppliedLinearAccClamp=<3000 3000 3000>
    componentLinearAccScale=<0.5 0.5 0.5>
    componentLinearVelScale=<0.6 0.6 0.6>
    gravityOverride {z=-500}
    physicsBodyDefinitions [{
      sphereCollisionRadius=0
      boundBone {boneName=dyn_l_shoulderFray_top_04_b}
      boxExtents=<20 15 15>
      constraintSetup {angularLimitsMax=<25 10 25>, angularLimitsMin=<-25 -25 -25>, angularTarget=<0 0 0>}
      localJointOffset=<2 0 0>
     }]
   }
   n67 = Anim Dynamics - Bone: dyn_l_shoulderFray_btm_04_a!(ComponentPose=@n66, bAlphaBoolEnabled:bool=@n56, ExternalForce="0,0,0") -> (Pose) {
    alphaInputType=Bool
    angularDampingOverride=0.9999
    angularSpringConstant=75
    bAngularSpring=true
    bGravityOverrideInSimSpace=true
    bOverrideAngularDamping=true
    bUseGravityOverride=true
    alphaBoolBlend {blendInTime=0.3, blendOption=Linear, blendOutTime=0.3}
    boundBone {boneName=dyn_l_shoulderFray_btm_04_a}
    componentAppliedLinearAccClamp=<3000 3000 3000>
    componentLinearAccScale=<0.5 0.5 0.5>
    componentLinearVelScale=<0.6 0.6 0.6>
    gravityOverride {z=-500}
    physicsBodyDefinitions [{
      sphereCollisionRadius=0
      boundBone {boneName=dyn_l_shoulderFray_btm_04_a}
      boxExtents=<20 15 15>
      constraintSetup {angularLimitsMax=<25 10 25>, angularLimitsMin=<-25 -25 -25>, angularTarget=<0 0 0>}
      localJointOffset=<2 0 0>
     }]
   }
   n68 = Anim Dynamics - Bone: dyn_l_shoulderFray_btm_04_b!(ComponentPose=@n67, bAlphaBoolEnabled:bool=@n56, ExternalForce="0,0,0") -> (Pose) {
    alphaInputType=Bool
    angularDampingOverride=0.9999
    angularSpringConstant=75
    bAngularSpring=true
    bGravityOverrideInSimSpace=true
    bOverrideAngularDamping=true
    bUseGravityOverride=true
    alphaBoolBlend {blendInTime=0.3, blendOption=Linear, blendOutTime=0.3}
    boundBone {boneName=dyn_l_shoulderFray_btm_04_b}
    componentAppliedLinearAccClamp=<3000 3000 3000>
    componentLinearAccScale=<0.5 0.5 0.5>
    componentLinearVelScale=<0.6 0.6 0.6>
    gravityOverride {z=-500}
    physicsBodyDefinitions [{
      sphereCollisionRadius=0
      boundBone {boneName=dyn_l_shoulderFray_btm_04_b}
      boxExtents=<20 15 15>
      constraintSetup {angularLimitsMax=<25 10 25>, angularLimitsMin=<-25 -25 -25>, angularTarget=<0 0 0>}
      localJointOffset=<2 0 0>
     }]
   }
   n69 = Anim Dynamics - Bone: dyn_l_shoulderFray_top_03_a!(ComponentPose=@n68, bAlphaBoolEnabled:bool=@n56, ExternalForce="0,0,0") -> (Pose) {
    alphaInputType=Bool
    angularDampingOverride=0.9999
    angularSpringConstant=75
    bAngularSpring=true
    bGravityOverrideInSimSpace=true
    bOverrideAngularDamping=true
    bUseGravityOverride=true
    alphaBoolBlend {blendInTime=0.3, blendOption=Linear, blendOutTime=0.3}
    boundBone {boneName=dyn_l_shoulderFray_top_03_a}
    componentAppliedLinearAccClamp=<3000 3000 3000>
    componentLinearAccScale=<0.5 0.5 0.5>
    componentLinearVelScale=<0.6 0.6 0.6>
    gravityOverride {z=-500}
    physicsBodyDefinitions [{
      sphereCollisionRadius=0
      boundBone {boneName=dyn_l_shoulderFray_top_03_a}
      boxExtents=<20 15 15>
      constraintSetup {angularLimitsMax=<25 10 25>, angularLimitsMin=<-25 -25 -25>, angularTarget=<0 0 0>}
      localJointOffset=<2 0 0>
     }]
   }
   n70 = Anim Dynamics - Bone: dyn_l_shoulderFray_top_03_b!(ComponentPose=@n69, bAlphaBoolEnabled:bool=@n56, ExternalForce="0,0,0") -> (Pose) {
    alphaInputType=Bool
    angularDampingOverride=0.9999
    angularSpringConstant=75
    bAngularSpring=true
    bGravityOverrideInSimSpace=true
    bOverrideAngularDamping=true
    bUseGravityOverride=true
    alphaBoolBlend {blendInTime=0.3, blendOption=Linear, blendOutTime=0.3}
    boundBone {boneName=dyn_l_shoulderFray_top_03_b}
    componentAppliedLinearAccClamp=<3000 3000 3000>
    componentLinearAccScale=<0.5 0.5 0.5>
    componentLinearVelScale=<0.6 0.6 0.6>
    gravityOverride {z=-500}
    physicsBodyDefinitions [{
      sphereCollisionRadius=0
      boundBone {boneName=dyn_l_shoulderFray_top_03_b}
      boxExtents=<20 15 15>
      constraintSetup {angularLimitsMax=<25 10 25>, angularLimitsMin=<-25 -25 -25>, angularTarget=<0 0 0>}
      localJointOffset=<2 0 0>
     }]
   }
   n71 = Anim Dynamics - Bone: dyn_l_shoulderFray_btm_03_a!(ComponentPose=@n70, bAlphaBoolEnabled:bool=@n56, ExternalForce="0,0,0") -> (Pose) {
    alphaInputType=Bool
    angularDampingOverride=0.9999
    angularSpringConstant=75
    bAngularSpring=true
    bGravityOverrideInSimSpace=true
    bOverrideAngularDamping=true
    bUseGravityOverride=true
    alphaBoolBlend {blendInTime=0.3, blendOption=Linear, blendOutTime=0.3}
    boundBone {boneName=dyn_l_shoulderFray_btm_03_a}
    componentAppliedLinearAccClamp=<3000 3000 3000>
    componentLinearAccScale=<0.5 0.5 0.5>
    componentLinearVelScale=<0.6 0.6 0.6>
    gravityOverride {z=-500}
    physicsBodyDefinitions [{
      sphereCollisionRadius=0
      boundBone {boneName=dyn_l_shoulderFray_btm_03_a}
      boxExtents=<20 15 15>
      constraintSetup {angularLimitsMax=<25 10 25>, angularLimitsMin=<-25 -25 -25>, angularTarget=<0 0 0>}
      localJointOffset=<2 0 0>
     }]
   }
   n72 = Anim Dynamics - Bone: dyn_l_shoulderFray_btm_03_b!(ComponentPose=@n71, bAlphaBoolEnabled:bool=@n56, ExternalForce="0,0,0") -> (Pose) {
    alphaInputType=Bool
    angularDampingOverride=0.9999
    angularSpringConstant=75
    bAngularSpring=true
    bGravityOverrideInSimSpace=true
    bOverrideAngularDamping=true
    bUseGravityOverride=true
    alphaBoolBlend {blendInTime=0.3, blendOption=Linear, blendOutTime=0.3}
    boundBone {boneName=dyn_l_shoulderFray_btm_03_b}
    componentAppliedLinearAccClamp=<3000 3000 3000>
    componentLinearAccScale=<0.5 0.5 0.5>
    componentLinearVelScale=<0.6 0.6 0.6>
    gravityOverride {z=-500}
    physicsBodyDefinitions [{
      sphereCollisionRadius=0
      boundBone {boneName=dyn_l_shoulderFray_btm_03_b}
      boxExtents=<20 15 15>
      constraintSetup {angularLimitsMax=<25 10 25>, angularLimitsMin=<-25 -25 -25>, angularTarget=<0 0 0>}
      localJointOffset=<2 0 0>
     }]
   }
   n73 = Anim Dynamics - Bone: dyn_l_shoulderFray_top_02_a!(ComponentPose=@n72, bAlphaBoolEnabled:bool=@n56, ExternalForce="0,0,0") -> (Pose) {
    alphaInputType=Bool
    angularDampingOverride=0.9999
    angularSpringConstant=75
    bAngularSpring=true
    bGravityOverrideInSimSpace=true
    bOverrideAngularDamping=true
    bUseGravityOverride=true
    alphaBoolBlend {blendInTime=0.3, blendOption=Linear, blendOutTime=0.3}
    boundBone {boneName=dyn_l_shoulderFray_top_02_a}
    componentAppliedLinearAccClamp=<3000 3000 3000>
    componentLinearAccScale=<0.5 0.5 0.5>
    componentLinearVelScale=<0.6 0.6 0.6>
    gravityOverride {z=-500}
    physicsBodyDefinitions [{
      sphereCollisionRadius=0
      boundBone {boneName=dyn_l_shoulderFray_top_02_a}
      boxExtents=<20 15 15>
      constraintSetup {angularLimitsMax=<25 10 25>, angularLimitsMin=<-25 -25 -25>, angularTarget=<0 0 0>}
      localJointOffset=<2 0 0>
     }]
   }
   n74 = Anim Dynamics - Bone: dyn_l_shoulderFray_top_02_b!(ComponentPose=@n73, bAlphaBoolEnabled:bool=@n56, ExternalForce="0,0,0") -> (Pose) {
    alphaInputType=Bool
    angularDampingOverride=0.9999
    angularSpringConstant=75
    bAngularSpring=true
    bGravityOverrideInSimSpace=true
    bOverrideAngularDamping=true
    bUseGravityOverride=true
    alphaBoolBlend {blendInTime=0.3, blendOption=Linear, blendOutTime=0.3}
    boundBone {boneName=dyn_l_shoulderFray_top_02_b}
    componentAppliedLinearAccClamp=<3000 3000 3000>
    componentLinearAccScale=<0.5 0.5 0.5>
    componentLinearVelScale=<0.6 0.6 0.6>
    gravityOverride {z=-500}
    physicsBodyDefinitions [{
      sphereCollisionRadius=0
      boundBone {boneName=dyn_l_shoulderFray_top_02_b}
      boxExtents=<20 15 15>
      constraintSetup {angularLimitsMax=<25 10 25>, angularLimitsMin=<-25 -25 -25>, angularTarget=<0 0 0>}
      localJointOffset=<2 0 0>
     }]
   }
   n75 = Anim Dynamics - Bone: dyn_l_shoulderFray_btm_02_a!(ComponentPose=@n74, bAlphaBoolEnabled:bool=@n56, ExternalForce="0,0,0") -> (Pose) {
    alphaInputType=Bool
    angularDampingOverride=0.9999
    angularSpringConstant=75
    bAngularSpring=true
    bGravityOverrideInSimSpace=true
    bOverrideAngularDamping=true
    bUseGravityOverride=true
    alphaBoolBlend {blendInTime=0.3, blendOption=Linear, blendOutTime=0.3}
    boundBone {boneName=dyn_l_shoulderFray_btm_02_a}
    componentAppliedLinearAccClamp=<3000 3000 3000>
    componentLinearAccScale=<0.5 0.5 0.5>
    componentLinearVelScale=<0.6 0.6 0.6>
    gravityOverride {z=-500}
    physicsBodyDefinitions [{
      sphereCollisionRadius=0
      boundBone {boneName=dyn_l_shoulderFray_btm_02_a}
      boxExtents=<20 15 15>
      constraintSetup {angularLimitsMax=<25 10 25>, angularLimitsMin=<-25 -25 -25>, angularTarget=<0 0 0>}
      localJointOffset=<2 0 0>
     }]
   }
   n76 = Anim Dynamics - Bone: dyn_l_shoulderFray_btm_02_b!(ComponentPose=@n75, bAlphaBoolEnabled:bool=@n56, ExternalForce="0,0,0") -> (Pose) {
    alphaInputType=Bool
    angularDampingOverride=0.9999
    angularSpringConstant=75
    bAngularSpring=true
    bGravityOverrideInSimSpace=true
    bOverrideAngularDamping=true
    bUseGravityOverride=true
    alphaBoolBlend {blendInTime=0.3, blendOption=Linear, blendOutTime=0.3}
    boundBone {boneName=dyn_l_shoulderFray_btm_02_b}
    componentAppliedLinearAccClamp=<3000 3000 3000>
    componentLinearAccScale=<0.5 0.5 0.5>
    componentLinearVelScale=<0.6 0.6 0.6>
    gravityOverride {z=-500}
    physicsBodyDefinitions [{
      sphereCollisionRadius=0
      boundBone {boneName=dyn_l_shoulderFray_btm_02_b}
      boxExtents=<20 15 15>
      constraintSetup {angularLimitsMax=<25 10 25>, angularLimitsMin=<-25 -25 -25>, angularTarget=<0 0 0>}
      localJointOffset=<2 0 0>
     }]
   }
   n77 = Anim Dynamics - Bone: dyn_l_shoulderFray_top_01_a!(ComponentPose=@n76, bAlphaBoolEnabled:bool=True, ExternalForce="0,0,0") -> (Pose) {
    alphaInputType=Bool
    angularDampingOverride=0.9999
    angularSpringConstant=75
    bAngularSpring=true
    bGravityOverrideInSimSpace=true
    bOverrideAngularDamping=true
    bUseGravityOverride=true
    alphaBoolBlend {blendInTime=0.3, blendOption=Linear, blendOutTime=0.3}
    boundBone {boneName=dyn_l_shoulderFray_top_01_a}
    componentAppliedLinearAccClamp=<3000 3000 3000>
    componentLinearAccScale=<0.5 0.5 0.5>
    componentLinearVelScale=<0.6 0.6 0.6>
    gravityOverride {z=-500}
    physicsBodyDefinitions [{
      sphereCollisionRadius=0
      boundBone {boneName=dyn_l_shoulderFray_top_01_a}
      boxExtents=<20 15 15>
      constraintSetup {angularLimitsMax=<25 10 25>, angularLimitsMin=<-25 -25 -25>, angularTarget=<0 0 0>}
      localJointOffset=<2 0 0>
     }]
   }
   n78 = Anim Dynamics - Bone: dyn_l_shoulderFray_top_01_b!(ComponentPose=@n77, bAlphaBoolEnabled:bool=True, ExternalForce="0,0,0") -> (Pose) {
    alphaInputType=Bool
    angularDampingOverride=0.9999
    angularSpringConstant=75
    bAngularSpring=true
    bGravityOverrideInSimSpace=true
    bOverrideAngularDamping=true
    bUseGravityOverride=true
    alphaBoolBlend {blendInTime=0.3, blendOption=Linear, blendOutTime=0.3}
    boundBone {boneName=dyn_l_shoulderFray_top_01_b}
    componentAppliedLinearAccClamp=<3000 3000 3000>
    componentLinearAccScale=<0.5 0.5 0.5>
    componentLinearVelScale=<0.6 0.6 0.6>
    gravityOverride {z=-500}
    physicsBodyDefinitions [{
      sphereCollisionRadius=0
      boundBone {boneName=dyn_l_shoulderFray_top_01_b}
      boxExtents=<20 15 15>
      constraintSetup {angularLimitsMax=<25 10 25>, angularLimitsMin=<-25 -25 -25>, angularTarget=<0 0 0>}
      localJointOffset=<2 0 0>
     }]
   }
   n79 = Anim Dynamics - Bone: dyn_l_shoulderFray_btm_01_a!(ComponentPose=@n78, bAlphaBoolEnabled:bool=@n56, ExternalForce="0,0,0") -> (Pose) {
    alphaInputType=Bool
    angularDampingOverride=0.9999
    angularSpringConstant=75
    bAngularSpring=true
    bGravityOverrideInSimSpace=true
    bOverrideAngularDamping=true
    bUseGravityOverride=true
    alphaBoolBlend {blendInTime=0.3, blendOption=Linear, blendOutTime=0.3}
    boundBone {boneName=dyn_l_shoulderFray_btm_01_a}
    componentAppliedLinearAccClamp=<3000 3000 3000>
    componentLinearAccScale=<0.5 0.5 0.5>
    componentLinearVelScale=<0.6 0.6 0.6>
    gravityOverride {z=-500}
    physicsBodyDefinitions [{
      sphereCollisionRadius=0
      boundBone {boneName=dyn_l_shoulderFray_btm_01_a}
      boxExtents=<20 15 15>
      constraintSetup {angularLimitsMax=<25 10 25>, angularLimitsMin=<-25 -25 -25>, angularTarget=<0 0 0>}
      localJointOffset=<2 0 0>
     }]
   }
   n80 = Anim Dynamics - Bone: dyn_l_shoulderFray_btm_01_b!(ComponentPose=@n79, bAlphaBoolEnabled:bool=@n56, ExternalForce="0,0,0") -> (Pose) {
    alphaInputType=Bool
    angularDampingOverride=0.9999
    angularSpringConstant=75
    bAngularSpring=true
    bGravityOverrideInSimSpace=true
    bOverrideAngularDamping=true
    bUseGravityOverride=true
    alphaBoolBlend {blendInTime=0.3, blendOption=Linear, blendOutTime=0.3}
    boundBone {boneName=dyn_l_shoulderFray_btm_01_b}
    componentAppliedLinearAccClamp=<3000 3000 3000>
    componentLinearAccScale=<0.5 0.5 0.5>
    componentLinearVelScale=<0.6 0.6 0.6>
    gravityOverride {z=-500}
    physicsBodyDefinitions [{
      sphereCollisionRadius=0
      boundBone {boneName=dyn_l_shoulderFray_btm_01_b}
      boxExtents=<20 15 15>
      constraintSetup {angularLimitsMax=<25 10 25>, angularLimitsMin=<-25 -25 -25>, angularTarget=<0 0 0>}
      localJointOffset=<2 0 0>
     }]
   }
   n81 = Anim Dynamics - Bone: dyn_r_shoulderFray_btm_06_a!(ComponentPose=@n80, bAlphaBoolEnabled:bool=@n56, ExternalForce="0,0,0") -> (Pose) {
    alphaInputType=Bool
    angularDampingOverride=0.9999
    angularSpringConstant=75
    bAngularSpring=true
    bGravityOverrideInSimSpace=true
    bOverrideAngularDamping=true
    bUseGravityOverride=true
    alphaBoolBlend {blendInTime=0.3, blendOption=Linear, blendOutTime=0.3}
    boundBone {boneName=dyn_r_shoulderFray_btm_06_a}
    componentAppliedLinearAccClamp=<3000 3000 3000>
    componentLinearAccScale=<0.5 0.5 0.5>
    componentLinearVelScale=<0.6 0.6 0.6>
    gravityOverride {z=-500}
    physicsBodyDefinitions [{
      sphereCollisionRadius=0
      boundBone {boneName=dyn_r_shoulderFray_btm_06_a}
      boxExtents=<20 15 15>
      constraintSetup {angularLimitsMax=<25 10 25>, angularLimitsMin=<-25 -25 -25>, angularTarget=<0 0 0>}
      localJointOffset=<-2 0 0>
     }]
   }
   n82 = Anim Dynamics - Bone: dyn_r_shoulderFray_btm_06_b!(ComponentPose=@n81, bAlphaBoolEnabled:bool=@n56, ExternalForce="0,0,0") -> (Pose) {
    alphaInputType=Bool
    angularDampingOverride=0.9999
    angularSpringConstant=75
    bAngularSpring=true
    bGravityOverrideInSimSpace=true
    bOverrideAngularDamping=true
    bUseGravityOverride=true
    alphaBoolBlend {blendInTime=0.3, blendOption=Linear, blendOutTime=0.3}
    boundBone {boneName=dyn_r_shoulderFray_btm_06_b}
    componentAppliedLinearAccClamp=<3000 3000 3000>
    componentLinearAccScale=<0.5 0.5 0.5>
    componentLinearVelScale=<0.6 0.6 0.6>
    gravityOverride {z=-500}
    physicsBodyDefinitions [{
      sphereCollisionRadius=0
      boundBone {boneName=dyn_r_shoulderFray_btm_06_b}
      boxExtents=<20 15 15>
      constraintSetup {angularLimitsMax=<25 10 25>, angularLimitsMin=<-25 -25 -25>, angularTarget=<0 0 0>}
      localJointOffset=<-2 0 0>
     }]
   }
   n83 = Anim Dynamics - Bone: dyn_r_shoulderFray_top_05_a!(ComponentPose=@n82, bAlphaBoolEnabled:bool=@n56, ExternalForce="0,0,0") -> (Pose) {
    alphaInputType=Bool
    angularDampingOverride=0.9999
    angularSpringConstant=75
    bAngularSpring=true
    bGravityOverrideInSimSpace=true
    bOverrideAngularDamping=true
    bUseGravityOverride=true
    alphaBoolBlend {blendInTime=0.3, blendOption=Linear, blendOutTime=0.3}
    boundBone {boneName=dyn_r_shoulderFray_top_05_a}
    componentAppliedLinearAccClamp=<3000 3000 3000>
    componentLinearAccScale=<0.5 0.5 0.5>
    componentLinearVelScale=<0.6 0.6 0.6>
    gravityOverride {z=-500}
    physicsBodyDefinitions [{
      sphereCollisionRadius=0
      boundBone {boneName=dyn_r_shoulderFray_top_05_a}
      boxExtents=<20 15 15>
      constraintSetup {angularLimitsMax=<10 10 10>, angularLimitsMin=<-10 -25 -10>, angularTarget=<0 0 0>}
      localJointOffset=<-2 0 0>
     }]
   }
   n84 = Anim Dynamics - Bone: dyn_r_shoulderFray_top_05_b!(ComponentPose=@n83, bAlphaBoolEnabled:bool=@n56, ExternalForce="0,0,0") -> (Pose) {
    alphaInputType=Bool
    angularDampingOverride=0.9999
    angularSpringConstant=75
    bAngularSpring=true
    bGravityOverrideInSimSpace=true
    bOverrideAngularDamping=true
    bUseGravityOverride=true
    alphaBoolBlend {blendInTime=0.3, blendOption=Linear, blendOutTime=0.3}
    boundBone {boneName=dyn_r_shoulderFray_top_05_b}
    componentAppliedLinearAccClamp=<3000 3000 3000>
    componentLinearAccScale=<0.5 0.5 0.5>
    componentLinearVelScale=<0.6 0.6 0.6>
    gravityOverride {z=-500}
    physicsBodyDefinitions [{
      sphereCollisionRadius=0
      boundBone {boneName=dyn_r_shoulderFray_top_05_b}
      boxExtents=<20 15 15>
      constraintSetup {angularLimitsMax=<10 10 10>, angularLimitsMin=<-10 -25 -10>, angularTarget=<0 0 0>}
      localJointOffset=<-2 0 0>
     }]
   }
   n85 = Anim Dynamics - Bone: dyn_r_shoulderFray_btm_05_a!(ComponentPose=@n84, bAlphaBoolEnabled:bool=@n56, ExternalForce="0,0,0") -> (Pose) {
    alphaInputType=Bool
    angularDampingOverride=0.9999
    angularSpringConstant=75
    bAngularSpring=true
    bGravityOverrideInSimSpace=true
    bOverrideAngularDamping=true
    bUseGravityOverride=true
    alphaBoolBlend {blendInTime=0.3, blendOption=Linear, blendOutTime=0.3}
    boundBone {boneName=dyn_r_shoulderFray_btm_05_a}
    componentAppliedLinearAccClamp=<3000 3000 3000>
    componentLinearAccScale=<0.5 0.5 0.5>
    componentLinearVelScale=<0.6 0.6 0.6>
    gravityOverride {z=-500}
    physicsBodyDefinitions [{
      sphereCollisionRadius=0
      boundBone {boneName=dyn_r_shoulderFray_btm_05_a}
      boxExtents=<20 15 15>
      constraintSetup {angularLimitsMax=<10 10 10>, angularLimitsMin=<-10 -25 -10>, angularTarget=<0 0 0>}
      localJointOffset=<-2 0 0>
     }]
   }
   n86 = Anim Dynamics - Bone: dyn_r_shoulderFray_btm_05_b!(ComponentPose=@n85, bAlphaBoolEnabled:bool=@n56, ExternalForce="0,0,0") -> (Pose) {
    alphaInputType=Bool
    angularDampingOverride=0.9999
    angularSpringConstant=75
    bAngularSpring=true
    bGravityOverrideInSimSpace=true
    bOverrideAngularDamping=true
    bUseGravityOverride=true
    alphaBoolBlend {blendInTime=0.3, blendOption=Linear, blendOutTime=0.3}
    boundBone {boneName=dyn_r_shoulderFray_btm_05_b}
    componentAppliedLinearAccClamp=<3000 3000 3000>
    componentLinearAccScale=<0.5 0.5 0.5>
    componentLinearVelScale=<0.6 0.6 0.6>
    gravityOverride {z=-500}
    physicsBodyDefinitions [{
      sphereCollisionRadius=0
      boundBone {boneName=dyn_r_shoulderFray_btm_05_b}
      boxExtents=<20 15 15>
      constraintSetup {angularLimitsMax=<10 10 10>, angularLimitsMin=<-10 -25 -10>, angularTarget=<0 0 0>}
      localJointOffset=<-2 0 0>
     }]
   }
   n87 = Anim Dynamics - Bone: dyn_r_shoulderFray_top_06_a!(ComponentPose=@n86, bAlphaBoolEnabled:bool=@n56, ExternalForce="0,0,0") -> (Pose) {
    alphaInputType=Bool
    angularDampingOverride=0.9999
    angularSpringConstant=75
    bAngularSpring=true
    bGravityOverrideInSimSpace=true
    bOverrideAngularDamping=true
    bUseGravityOverride=true
    alphaBoolBlend {blendInTime=0.3, blendOption=Linear, blendOutTime=0.3}
    boundBone {boneName=dyn_r_shoulderFray_top_06_a}
    componentAppliedLinearAccClamp=<3000 3000 3000>
    componentLinearAccScale=<0.5 0.5 0.5>
    componentLinearVelScale=<0.6 0.6 0.6>
    gravityOverride {z=-500}
    physicsBodyDefinitions [{
      sphereCollisionRadius=0
      boundBone {boneName=dyn_r_shoulderFray_top_06_a}
      boxExtents=<20 15 15>
      constraintSetup {angularLimitsMax=<25 10 25>, angularLimitsMin=<-25 -25 -25>, angularTarget=<0 0 0>}
      localJointOffset=<-2 0 0>
     }]
   }
   n88 = Anim Dynamics - Bone: dyn_r_shoulderFray_top_06_b!(ComponentPose=@n87, bAlphaBoolEnabled:bool=@n56, ExternalForce="0,0,0") -> (Pose) {
    alphaInputType=Bool
    angularDampingOverride=0.9999
    angularSpringConstant=75
    bAngularSpring=true
    bGravityOverrideInSimSpace=true
    bOverrideAngularDamping=true
    bUseGravityOverride=true
    alphaBoolBlend {blendInTime=0.3, blendOption=Linear, blendOutTime=0.3}
    boundBone {boneName=dyn_r_shoulderFray_top_06_b}
    componentAppliedLinearAccClamp=<3000 3000 3000>
    componentLinearAccScale=<0.5 0.5 0.5>
    componentLinearVelScale=<0.6 0.6 0.6>
    gravityOverride {z=-500}
    physicsBodyDefinitions [{
      sphereCollisionRadius=0
      boundBone {boneName=dyn_r_shoulderFray_top_06_b}
      boxExtents=<20 15 15>
      constraintSetup {angularLimitsMax=<25 10 25>, angularLimitsMin=<-25 -25 -25>, angularTarget=<0 0 0>}
      localJointOffset=<-2 0 0>
     }]
   }
   n89 = Anim Dynamics - Bone: dyn_r_shoulderFray_top_04_a!(ComponentPose=@n88, bAlphaBoolEnabled:bool=@n56, ExternalForce="0,0,0") -> (Pose) {
    alphaInputType=Bool
    angularDampingOverride=0.9999
    angularSpringConstant=75
    bAngularSpring=true
    bGravityOverrideInSimSpace=true
    bOverrideAngularDamping=true
    bUseGravityOverride=true
    alphaBoolBlend {blendInTime=0.3, blendOption=Linear, blendOutTime=0.3}
    boundBone {boneName=dyn_r_shoulderFray_top_04_a}
    componentAppliedLinearAccClamp=<3000 3000 3000>
    componentLinearAccScale=<0.5 0.5 0.5>
    componentLinearVelScale=<0.6 0.6 0.6>
    gravityOverride {z=-500}
    physicsBodyDefinitions [{
      sphereCollisionRadius=0
      boundBone {boneName=dyn_r_shoulderFray_top_04_a}
      boxExtents=<20 15 15>
      constraintSetup {angularLimitsMax=<25 10 25>, angularLimitsMin=<-25 -25 -25>, angularTarget=<0 0 0>}
      localJointOffset=<-2 0 0>
     }]
   }
   n90 = Anim Dynamics - Bone: dyn_r_shoulderFray_top_04_b!(ComponentPose=@n89, bAlphaBoolEnabled:bool=@n56, ExternalForce="0,0,0") -> (Pose) {
    alphaInputType=Bool
    angularDampingOverride=0.9999
    angularSpringConstant=75
    bAngularSpring=true
    bGravityOverrideInSimSpace=true
    bOverrideAngularDamping=true
    bUseGravityOverride=true
    alphaBoolBlend {blendInTime=0.3, blendOption=Linear, blendOutTime=0.3}
    boundBone {boneName=dyn_r_shoulderFray_top_04_b}
    componentAppliedLinearAccClamp=<3000 3000 3000>
    componentLinearAccScale=<0.5 0.5 0.5>
    componentLinearVelScale=<0.6 0.6 0.6>
    gravityOverride {z=-500}
    physicsBodyDefinitions [{
      sphereCollisionRadius=0
      boundBone {boneName=dyn_r_shoulderFray_top_04_b}
      boxExtents=<20 15 15>
      constraintSetup {angularLimitsMax=<25 10 25>, angularLimitsMin=<-25 -25 -25>, angularTarget=<0 0 0>}
      localJointOffset=<-2 0 0>
     }]
   }
   n91 = Anim Dynamics - Bone: dyn_r_shoulderFray_btm_04_a!(ComponentPose=@n90, bAlphaBoolEnabled:bool=@n56, ExternalForce="0,0,0") -> (Pose) {
    alphaInputType=Bool
    angularDampingOverride=0.9999
    angularSpringConstant=75
    bAngularSpring=true
    bGravityOverrideInSimSpace=true
    bOverrideAngularDamping=true
    bUseGravityOverride=true
    alphaBoolBlend {blendInTime=0.3, blendOption=Linear, blendOutTime=0.3}
    boundBone {boneName=dyn_r_shoulderFray_btm_04_a}
    componentAppliedLinearAccClamp=<3000 3000 3000>
    componentLinearAccScale=<0.5 0.5 0.5>
    componentLinearVelScale=<0.6 0.6 0.6>
    gravityOverride {z=-500}
    physicsBodyDefinitions [{
      sphereCollisionRadius=0
      boundBone {boneName=dyn_r_shoulderFray_btm_04_a}
      boxExtents=<20 15 15>
      constraintSetup {angularLimitsMax=<25 10 25>, angularLimitsMin=<-25 -25 -25>, angularTarget=<0 0 0>}
      localJointOffset=<-2 0 0>
     }]
   }
   n92 = Anim Dynamics - Bone: dyn_r_shoulderFray_btm_04_b!(ComponentPose=@n91, bAlphaBoolEnabled:bool=@n56, ExternalForce="0,0,0") -> (Pose) {
    alphaInputType=Bool
    angularDampingOverride=0.9999
    angularSpringConstant=75
    bAngularSpring=true
    bGravityOverrideInSimSpace=true
    bOverrideAngularDamping=true
    bUseGravityOverride=true
    alphaBoolBlend {blendInTime=0.3, blendOption=Linear, blendOutTime=0.3}
    boundBone {boneName=dyn_r_shoulderFray_btm_04_b}
    componentAppliedLinearAccClamp=<3000 3000 3000>
    componentLinearAccScale=<0.5 0.5 0.5>
    componentLinearVelScale=<0.6 0.6 0.6>
    gravityOverride {z=-500}
    physicsBodyDefinitions [{
      sphereCollisionRadius=0
      boundBone {boneName=dyn_r_shoulderFray_btm_04_b}
      boxExtents=<20 15 15>
      constraintSetup {angularLimitsMax=<25 10 25>, angularLimitsMin=<-25 -25 -25>, angularTarget=<0 0 0>}
      localJointOffset=<-2 0 0>
     }]
   }
   n93 = Anim Dynamics - Bone: dyn_r_shoulderFray_top_03_a!(ComponentPose=@n92, bAlphaBoolEnabled:bool=@n56, ExternalForce="0,0,0") -> (Pose) {
    alphaInputType=Bool
    angularDampingOverride=0.9999
    angularSpringConstant=75
    bAngularSpring=true
    bGravityOverrideInSimSpace=true
    bOverrideAngularDamping=true
    bUseGravityOverride=true
    alphaBoolBlend {blendInTime=0.3, blendOption=Linear, blendOutTime=0.3}
    boundBone {boneName=dyn_r_shoulderFray_top_03_a}
    componentAppliedLinearAccClamp=<3000 3000 3000>
    componentLinearAccScale=<0.5 0.5 0.5>
    componentLinearVelScale=<0.6 0.6 0.6>
    gravityOverride {z=-500}
    physicsBodyDefinitions [{
      sphereCollisionRadius=0
      boundBone {boneName=dyn_r_shoulderFray_top_03_a}
      boxExtents=<20 15 15>
      constraintSetup {angularLimitsMax=<25 10 25>, angularLimitsMin=<-25 -25 -25>, angularTarget=<0 0 0>}
      localJointOffset=<-2 0 0>
     }]
   }
   n94 = Anim Dynamics - Bone: dyn_r_shoulderFray_top_03_b!(ComponentPose=@n93, bAlphaBoolEnabled:bool=@n56, ExternalForce="0,0,0") -> (Pose) {
    alphaInputType=Bool
    angularDampingOverride=0.9999
    angularSpringConstant=75
    bAngularSpring=true
    bGravityOverrideInSimSpace=true
    bOverrideAngularDamping=true
    bUseGravityOverride=true
    alphaBoolBlend {blendInTime=0.3, blendOption=Linear, blendOutTime=0.3}
    boundBone {boneName=dyn_r_shoulderFray_top_03_b}
    componentAppliedLinearAccClamp=<3000 3000 3000>
    componentLinearAccScale=<0.5 0.5 0.5>
    componentLinearVelScale=<0.6 0.6 0.6>
    gravityOverride {z=-500}
    physicsBodyDefinitions [{
      sphereCollisionRadius=0
      boundBone {boneName=dyn_r_shoulderFray_top_03_b}
      boxExtents=<20 15 15>
      constraintSetup {angularLimitsMax=<25 10 25>, angularLimitsMin=<-25 -25 -25>, angularTarget=<0 0 0>}
      localJointOffset=<-2 0 0>
     }]
   }
   n95 = Anim Dynamics - Bone: dyn_r_shoulderFray_btm_03_a!(ComponentPose=@n94, bAlphaBoolEnabled:bool=@n56, ExternalForce="0,0,0") -> (Pose) {
    alphaInputType=Bool
    angularDampingOverride=0.9999
    angularSpringConstant=75
    bAngularSpring=true
    bGravityOverrideInSimSpace=true
    bOverrideAngularDamping=true
    bUseGravityOverride=true
    alphaBoolBlend {blendInTime=0.3, blendOption=Linear, blendOutTime=0.3}
    boundBone {boneName=dyn_r_shoulderFray_btm_03_a}
    componentAppliedLinearAccClamp=<3000 3000 3000>
    componentLinearAccScale=<0.5 0.5 0.5>
    componentLinearVelScale=<0.6 0.6 0.6>
    gravityOverride {z=-500}
    physicsBodyDefinitions [{
      sphereCollisionRadius=0
      boundBone {boneName=dyn_r_shoulderFray_btm_03_a}
      boxExtents=<20 15 15>
      constraintSetup {angularLimitsMax=<25 10 25>, angularLimitsMin=<-25 -25 -25>, angularTarget=<0 0 0>}
      localJointOffset=<-2 0 0>
     }]
   }
   n96 = Anim Dynamics - Bone: dyn_r_shoulderFray_btm_03_b!(ComponentPose=@n95, bAlphaBoolEnabled:bool=@n56, ExternalForce="0,0,0") -> (Pose) {
    alphaInputType=Bool
    angularDampingOverride=0.9999
    angularSpringConstant=75
    bAngularSpring=true
    bGravityOverrideInSimSpace=true
    bOverrideAngularDamping=true
    bUseGravityOverride=true
    alphaBoolBlend {blendInTime=0.3, blendOption=Linear, blendOutTime=0.3}
    boundBone {boneName=dyn_r_shoulderFray_btm_03_b}
    componentAppliedLinearAccClamp=<3000 3000 3000>
    componentLinearAccScale=<0.5 0.5 0.5>
    componentLinearVelScale=<0.6 0.6 0.6>
    gravityOverride {z=-500}
    physicsBodyDefinitions [{
      sphereCollisionRadius=0
      boundBone {boneName=dyn_r_shoulderFray_btm_03_b}
      boxExtents=<20 15 15>
      constraintSetup {angularLimitsMax=<25 10 25>, angularLimitsMin=<-25 -25 -25>, angularTarget=<0 0 0>}
      localJointOffset=<-2 0 0>
     }]
   }
   n97 = Anim Dynamics - Bone: dyn_r_shoulderFray_top_02_a!(ComponentPose=@n96, bAlphaBoolEnabled:bool=@n56, ExternalForce="0,0,0") -> (Pose) {
    alphaInputType=Bool
    angularDampingOverride=0.9999
    angularSpringConstant=75
    bAngularSpring=true
    bGravityOverrideInSimSpace=true
    bOverrideAngularDamping=true
    bUseGravityOverride=true
    alphaBoolBlend {blendInTime=0.3, blendOption=Linear, blendOutTime=0.3}
    boundBone {boneName=dyn_r_shoulderFray_top_02_a}
    componentAppliedLinearAccClamp=<3000 3000 3000>
    componentLinearAccScale=<0.5 0.5 0.5>
    componentLinearVelScale=<0.6 0.6 0.6>
    gravityOverride {z=-500}
    physicsBodyDefinitions [{
      sphereCollisionRadius=0
      boundBone {boneName=dyn_r_shoulderFray_top_02_a}
      boxExtents=<20 15 15>
      constraintSetup {angularLimitsMax=<25 10 25>, angularLimitsMin=<-25 -25 -25>, angularTarget=<0 0 0>}
      localJointOffset=<-2 0 0>
     }]
   }
   n98 = Anim Dynamics - Bone: dyn_r_shoulderFray_top_02_b!(ComponentPose=@n97, bAlphaBoolEnabled:bool=@n56, ExternalForce="0,0,0") -> (Pose) {
    alphaInputType=Bool
    angularDampingOverride=0.9999
    angularSpringConstant=75
    bAngularSpring=true
    bGravityOverrideInSimSpace=true
    bOverrideAngularDamping=true
    bUseGravityOverride=true
    alphaBoolBlend {blendInTime=0.3, blendOption=Linear, blendOutTime=0.3}
    boundBone {boneName=dyn_r_shoulderFray_top_02_b}
    componentAppliedLinearAccClamp=<3000 3000 3000>
    componentLinearAccScale=<0.5 0.5 0.5>
    componentLinearVelScale=<0.6 0.6 0.6>
    gravityOverride {z=-500}
    physicsBodyDefinitions [{
      sphereCollisionRadius=0
      boundBone {boneName=dyn_r_shoulderFray_top_02_b}
      boxExtents=<20 15 15>
      constraintSetup {angularLimitsMax=<25 10 25>, angularLimitsMin=<-25 -25 -25>, angularTarget=<0 0 0>}
      localJointOffset=<-2 0 0>
     }]
   }
   n99 = Anim Dynamics - Bone: dyn_r_shoulderFray_btm_02_a!(ComponentPose=@n98, bAlphaBoolEnabled:bool=@n56, ExternalForce="0,0,0") -> (Pose) {
    alphaInputType=Bool
    angularDampingOverride=0.9999
    angularSpringConstant=75
    bAngularSpring=true
    bGravityOverrideInSimSpace=true
    bOverrideAngularDamping=true
    bUseGravityOverride=true
    alphaBoolBlend {blendInTime=0.3, blendOption=Linear, blendOutTime=0.3}
    boundBone {boneName=dyn_r_shoulderFray_btm_02_a}
    componentAppliedLinearAccClamp=<3000 3000 3000>
    componentLinearAccScale=<0.5 0.5 0.5>
    componentLinearVelScale=<0.6 0.6 0.6>
    gravityOverride {z=-500}
    physicsBodyDefinitions [{
      sphereCollisionRadius=0
      boundBone {boneName=dyn_r_shoulderFray_btm_02_a}
      boxExtents=<20 15 15>
      constraintSetup {angularLimitsMax=<25 10 25>, angularLimitsMin=<-25 -25 -25>, angularTarget=<0 0 0>}
      localJointOffset=<-2 0 0>
     }]
   }
   n100 = Anim Dynamics - Bone: dyn_r_shoulderFray_btm_02_b!(ComponentPose=@n99, bAlphaBoolEnabled:bool=@n56, ExternalForce="0,0,0") -> (Pose) {
    alphaInputType=Bool
    angularDampingOverride=0.9999
    angularSpringConstant=75
    bAngularSpring=true
    bGravityOverrideInSimSpace=true
    bOverrideAngularDamping=true
    bUseGravityOverride=true
    alphaBoolBlend {blendInTime=0.3, blendOption=Linear, blendOutTime=0.3}
    boundBone {boneName=dyn_r_shoulderFray_btm_02_b}
    componentAppliedLinearAccClamp=<3000 3000 3000>
    componentLinearAccScale=<0.5 0.5 0.5>
    componentLinearVelScale=<0.6 0.6 0.6>
    gravityOverride {z=-500}
    physicsBodyDefinitions [{
      sphereCollisionRadius=0
      boundBone {boneName=dyn_r_shoulderFray_btm_02_b}
      boxExtents=<20 15 15>
      constraintSetup {angularLimitsMax=<25 10 25>, angularLimitsMin=<-25 -25 -25>, angularTarget=<0 0 0>}
      localJointOffset=<-2 0 0>
     }]
   }
   n101 = Anim Dynamics - Bone: dyn_r_shoulderFray_top_01_a!(ComponentPose=@n100, bAlphaBoolEnabled:bool=@n56, ExternalForce="0,0,0") -> (Pose) {
    alphaInputType=Bool
    angularDampingOverride=0.9999
    angularSpringConstant=75
    bAngularSpring=true
    bGravityOverrideInSimSpace=true
    bOverrideAngularDamping=true
    bUseGravityOverride=true
    alphaBoolBlend {blendInTime=0.3, blendOption=Linear, blendOutTime=0.3}
    boundBone {boneName=dyn_r_shoulderFray_top_01_a}
    componentAppliedLinearAccClamp=<3000 3000 3000>
    componentLinearAccScale=<0.5 0.5 0.5>
    componentLinearVelScale=<0.6 0.6 0.6>
    gravityOverride {z=-500}
    physicsBodyDefinitions [{
      sphereCollisionRadius=0
      boundBone {boneName=dyn_r_shoulderFray_top_01_a}
      boxExtents=<20 15 15>
      constraintSetup {angularLimitsMax=<25 10 25>, angularLimitsMin=<-25 -25 -25>, angularTarget=<0 0 0>}
      localJointOffset=<-2 0 0>
     }]
   }
   n102 = Anim Dynamics - Bone: dyn_r_shoulderFray_top_01_b!(ComponentPose=@n101, bAlphaBoolEnabled:bool=@n56, ExternalForce="0,0,0") -> (Pose) {
    alphaInputType=Bool
    angularDampingOverride=0.9999
    angularSpringConstant=75
    bAngularSpring=true
    bGravityOverrideInSimSpace=true
    bOverrideAngularDamping=true
    bUseGravityOverride=true
    alphaBoolBlend {blendInTime=0.3, blendOption=Linear, blendOutTime=0.3}
    boundBone {boneName=dyn_r_shoulderFray_top_01_b}
    componentAppliedLinearAccClamp=<3000 3000 3000>
    componentLinearAccScale=<0.5 0.5 0.5>
    componentLinearVelScale=<0.6 0.6 0.6>
    gravityOverride {z=-500}
    physicsBodyDefinitions [{
      sphereCollisionRadius=0
      boundBone {boneName=dyn_r_shoulderFray_top_01_b}
      boxExtents=<20 15 15>
      constraintSetup {angularLimitsMax=<25 10 25>, angularLimitsMin=<-25 -25 -25>, angularTarget=<0 0 0>}
      localJointOffset=<-2 0 0>
     }]
   }
   n103 = Anim Dynamics - Bone: dyn_r_shoulderFray_btm_01_a!(ComponentPose=@n102, bAlphaBoolEnabled:bool=@n56, ExternalForce="0,0,0") -> (Pose) {
    alphaInputType=Bool
    angularDampingOverride=0.9999
    angularSpringConstant=75
    bAngularSpring=true
    bGravityOverrideInSimSpace=true
    bOverrideAngularDamping=true
    bUseGravityOverride=true
    alphaBoolBlend {blendInTime=0.3, blendOption=Linear, blendOutTime=0.3}
    boundBone {boneName=dyn_r_shoulderFray_btm_01_a}
    componentAppliedLinearAccClamp=<3000 3000 3000>
    componentLinearAccScale=<0.5 0.5 0.5>
    componentLinearVelScale=<0.6 0.6 0.6>
    gravityOverride {z=-500}
    physicsBodyDefinitions [{
      sphereCollisionRadius=0
      boundBone {boneName=dyn_r_shoulderFray_btm_01_a}
      boxExtents=<20 15 15>
      constraintSetup {angularLimitsMax=<25 10 25>, angularLimitsMin=<-25 -25 -25>, angularTarget=<0 0 0>}
      localJointOffset=<-2 0 0>
     }]
   }
   n104 = Anim Dynamics - Bone: dyn_r_shoulderFray_btm_01_b!(ComponentPose=@n103, bAlphaBoolEnabled:bool=@n56, ExternalForce="0,0,0") -> (Pose) {
    alphaInputType=Bool
    angularDampingOverride=0.9999
    angularSpringConstant=75
    bAngularSpring=true
    bGravityOverrideInSimSpace=true
    bOverrideAngularDamping=true
    bUseGravityOverride=true
    alphaBoolBlend {blendInTime=0.3, blendOption=Linear, blendOutTime=0.3}
    boundBone {boneName=dyn_r_shoulderFray_btm_01_b}
    componentAppliedLinearAccClamp=<3000 3000 3000>
    componentLinearAccScale=<0.5 0.5 0.5>
    componentLinearVelScale=<0.6 0.6 0.6>
    gravityOverride {z=-500}
    physicsBodyDefinitions [{
      sphereCollisionRadius=0
      boundBone {boneName=dyn_r_shoulderFray_btm_01_b}
      boxExtents=<20 15 15>
      constraintSetup {angularLimitsMax=<25 10 25>, angularLimitsMin=<-25 -25 -25>, angularTarget=<0 0 0>}
      localJointOffset=<-2 0 0>
     }]
   }
   n105 = Anim Dynamics - Bone: dyn_l_shoulderFray_btm_06_a(ComponentPose=@n104, bAlphaBoolEnabled=@n51, ExternalForce="0,0,0") {
    alphaInputType=Bool
    angularDampingOverride=0.9999
    angularSpringConstant=75
    bAngularSpring=true
    bGravityOverrideInSimSpace=true
    bOverrideAngularDamping=true
    bUseGravityOverride=true
    simulationSpace=RootRelative
    alphaBoolBlend {blendInTime=0.3, blendOption=Linear, blendOutTime=0.3}
    boundBone {boneName=dyn_l_shoulderFray_btm_06_a}
    componentAppliedLinearAccClamp=<3000 3000 3000>
    componentLinearAccScale=<0.5 0.5 0.5>
    componentLinearVelScale=<0.6 0.6 0.6>
    gravityOverride {z=-900}
    physicsBodyDefinitions [{
      sphereCollisionRadius=0
      boundBone {boneName=dyn_l_shoulderFray_btm_06_a}
      boxExtents=<20 10 10>
      constraintSetup {angularLimitsMax=<25 10 25>, angularLimitsMin=<-25 -25 -25>, angularTarget=<0 0 0>}
      localJointOffset=<2 0 0>
     }]
   }
   n106 = Anim Dynamics - Bone: dyn_l_shoulderFray_btm_06_b(ComponentPose=@n105, bAlphaBoolEnabled=@n51, ExternalForce="0,0,0") {
    alphaInputType=Bool
    angularDampingOverride=0.9999
    angularSpringConstant=75
    bAngularSpring=true
    bGravityOverrideInSimSpace=true
    bOverrideAngularDamping=true
    bUseGravityOverride=true
    simulationSpace=RootRelative
    alphaBoolBlend {blendInTime=0.3, blendOption=Linear, blendOutTime=0.3}
    boundBone {boneName=dyn_l_shoulderFray_btm_06_b}
    componentAppliedLinearAccClamp=<3000 3000 3000>
    componentLinearAccScale=<0.5 0.5 0.5>
    componentLinearVelScale=<0.6 0.6 0.6>
    gravityOverride {z=-900}
    physicsBodyDefinitions [{
      sphereCollisionRadius=0
      boundBone {boneName=dyn_l_shoulderFray_btm_06_b}
      boxExtents=<20 10 10>
      constraintSetup {angularLimitsMax=<25 10 25>, angularLimitsMin=<-25 -25 -25>, angularTarget=<0 0 0>}
      localJointOffset=<2 0 0>
     }]
   }
   n107 = Anim Dynamics - Bone: dyn_l_shoulderFray_top_06_a(ComponentPose=@n106, bAlphaBoolEnabled=@n51, ExternalForce="0,0,0") {
    alphaInputType=Bool
    angularDampingOverride=0.9999
    angularSpringConstant=75
    bAngularSpring=true
    bGravityOverrideInSimSpace=true
    bOverrideAngularDamping=true
    bUseGravityOverride=true
    simulationSpace=RootRelative
    alphaBoolBlend {blendInTime=0.3, blendOption=Linear, blendOutTime=0.3}
    boundBone {boneName=dyn_l_shoulderFray_top_06_a}
    componentAppliedLinearAccClamp=<3000 3000 3000>
    componentLinearAccScale=<0.5 0.5 0.5>
    componentLinearVelScale=<0.6 0.6 0.6>
    gravityOverride {z=-900}
    physicsBodyDefinitions [{
      sphereCollisionRadius=0
      boundBone {boneName=dyn_l_shoulderFray_top_06_a}
      boxExtents=<20 10 10>
      constraintSetup {angularLimitsMax=<25 10 25>, angularLimitsMin=<-25 -25 -25>, angularTarget=<0 0 0>}
      localJointOffset=<2 0 0>
     }]
   }
   n108 = Anim Dynamics - Bone: dyn_l_shoulderFray_top_06_b(ComponentPose=@n107, bAlphaBoolEnabled=@n51, ExternalForce="0,0,0") {
    alphaInputType=Bool
    angularDampingOverride=0.9999
    angularSpringConstant=75
    bAngularSpring=true
    bGravityOverrideInSimSpace=true
    bOverrideAngularDamping=true
    bUseGravityOverride=true
    simulationSpace=RootRelative
    alphaBoolBlend {blendInTime=0.3, blendOption=Linear, blendOutTime=0.3}
    boundBone {boneName=dyn_l_shoulderFray_top_06_b}
    componentAppliedLinearAccClamp=<3000 3000 3000>
    componentLinearAccScale=<0.5 0.5 0.5>
    componentLinearVelScale=<0.6 0.6 0.6>
    gravityOverride {z=-900}
    physicsBodyDefinitions [{
      sphereCollisionRadius=0
      boundBone {boneName=dyn_l_shoulderFray_top_06_b}
      boxExtents=<20 10 10>
      constraintSetup {angularLimitsMax=<25 10 25>, angularLimitsMin=<-25 -25 -25>, angularTarget=<0 0 0>}
      localJointOffset=<2 0 0>
     }]
   }
   n109 = Anim Dynamics - Bone: dyn_l_shoulderFray_top_05_a(ComponentPose=@n108, bAlphaBoolEnabled=@n51, ExternalForce="0,0,0") {
    alphaInputType=Bool
    angularDampingOverride=0.9999
    angularSpringConstant=75
    bAngularSpring=true
    bGravityOverrideInSimSpace=true
    bOverrideAngularDamping=true
    bUseGravityOverride=true
    simulationSpace=RootRelative
    alphaBoolBlend {blendInTime=0.3, blendOption=Linear, blendOutTime=0.3}
    boundBone {boneName=dyn_l_shoulderFray_top_05_a}
    componentAppliedLinearAccClamp=<3000 3000 3000>
    componentLinearAccScale=<0.5 0.5 0.5>
    componentLinearVelScale=<0.6 0.6 0.6>
    gravityOverride {z=-900}
    physicsBodyDefinitions [{
      sphereCollisionRadius=0
      boundBone {boneName=dyn_l_shoulderFray_top_05_a}
      boxExtents=<20 10 10>
      constraintSetup {angularLimitsMax=<25 10 25>, angularLimitsMin=<-25 -25 -25>, angularTarget=<0 0 0>}
      localJointOffset=<2 0 0>
     }]
   }
   n110 = Anim Dynamics - Bone: dyn_l_shoulderFray_top_05_b(ComponentPose=@n109, bAlphaBoolEnabled=@n51, ExternalForce="0,0,0") {
    alphaInputType=Bool
    angularDampingOverride=0.9999
    angularSpringConstant=75
    bAngularSpring=true
    bGravityOverrideInSimSpace=true
    bOverrideAngularDamping=true
    bUseGravityOverride=true
    simulationSpace=RootRelative
    alphaBoolBlend {blendInTime=0.3, blendOption=Linear, blendOutTime=0.3}
    boundBone {boneName=dyn_l_shoulderFray_top_05_b}
    componentAppliedLinearAccClamp=<3000 3000 3000>
    componentLinearAccScale=<0.5 0.5 0.5>
    componentLinearVelScale=<0.6 0.6 0.6>
    gravityOverride {z=-900}
    physicsBodyDefinitions [{
      sphereCollisionRadius=0
      boundBone {boneName=dyn_l_shoulderFray_top_05_b}
      boxExtents=<20 10 10>
      constraintSetup {angularLimitsMax=<25 10 25>, angularLimitsMin=<-25 -25 -25>, angularTarget=<0 0 0>}
      localJointOffset=<2 0 0>
     }]
   }
   n111 = Anim Dynamics - Bone: dyn_l_shoulderFray_btm_05_a(ComponentPose=@n110, bAlphaBoolEnabled=@n51, ExternalForce="0,0,0") {
    alphaInputType=Bool
    angularDampingOverride=0.9999
    angularSpringConstant=75
    bAngularSpring=true
    bGravityOverrideInSimSpace=true
    bOverrideAngularDamping=true
    bUseGravityOverride=true
    simulationSpace=RootRelative
    alphaBoolBlend {blendInTime=0.3, blendOption=Linear, blendOutTime=0.3}
    boundBone {boneName=dyn_l_shoulderFray_btm_05_a}
    componentAppliedLinearAccClamp=<3000 3000 3000>
    componentLinearAccScale=<0.5 0.5 0.5>
    componentLinearVelScale=<0.6 0.6 0.6>
    gravityOverride {z=-900}
    physicsBodyDefinitions [{
      sphereCollisionRadius=0
      boundBone {boneName=dyn_l_shoulderFray_btm_05_a}
      boxExtents=<20 10 10>
      constraintSetup {angularLimitsMax=<25 10 25>, angularLimitsMin=<-25 -25 -25>, angularTarget=<0 0 0>}
      localJointOffset=<2 0 0>
     }]
   }
   n112 = Anim Dynamics - Bone: dyn_l_shoulderFray_btm_05_b(ComponentPose=@n111, bAlphaBoolEnabled=@n51, ExternalForce="0,0,0") {
    alphaInputType=Bool
    angularDampingOverride=0.9999
    angularSpringConstant=75
    bAngularSpring=true
    bGravityOverrideInSimSpace=true
    bOverrideAngularDamping=true
    bUseGravityOverride=true
    simulationSpace=RootRelative
    alphaBoolBlend {blendInTime=0.3, blendOption=Linear, blendOutTime=0.3}
    boundBone {boneName=dyn_l_shoulderFray_btm_05_b}
    componentAppliedLinearAccClamp=<3000 3000 3000>
    componentLinearAccScale=<0.5 0.5 0.5>
    componentLinearVelScale=<0.6 0.6 0.6>
    gravityOverride {z=-900}
    physicsBodyDefinitions [{
      sphereCollisionRadius=0
      boundBone {boneName=dyn_l_shoulderFray_btm_05_b}
      boxExtents=<20 10 10>
      constraintSetup {angularLimitsMax=<25 10 25>, angularLimitsMin=<-25 -25 -25>, angularTarget=<0 0 0>}
      localJointOffset=<2 0 0>
     }]
   }
   n113 = Anim Dynamics - Bone: dyn_l_shoulderFray_top_04_a(ComponentPose=@n112, bAlphaBoolEnabled=@n51, ExternalForce="0,0,0") {
    alphaInputType=Bool
    angularDampingOverride=0.9999
    angularSpringConstant=75
    bAngularSpring=true
    bGravityOverrideInSimSpace=true
    bOverrideAngularDamping=true
    bUseGravityOverride=true
    simulationSpace=RootRelative
    alphaBoolBlend {blendInTime=0.3, blendOption=Linear, blendOutTime=0.3}
    boundBone {boneName=dyn_l_shoulderFray_top_04_a}
    componentAppliedLinearAccClamp=<3000 3000 3000>
    componentLinearAccScale=<0.5 0.5 0.5>
    componentLinearVelScale=<0.6 0.6 0.6>
    gravityOverride {z=-900}
    physicsBodyDefinitions [{
      sphereCollisionRadius=0
      boundBone {boneName=dyn_l_shoulderFray_top_04_a}
      boxExtents=<20 10 10>
      constraintSetup {angularLimitsMax=<25 10 25>, angularLimitsMin=<-25 -25 -25>, angularTarget=<0 0 0>}
      localJointOffset=<2 0 0>
     }]
   }
   n114 = Anim Dynamics - Bone: dyn_l_shoulderFray_top_04_b(ComponentPose=@n113, bAlphaBoolEnabled=@n51, ExternalForce="0,0,0") {
    alphaInputType=Bool
    angularDampingOverride=0.9999
    angularSpringConstant=75
    bAngularSpring=true
    bGravityOverrideInSimSpace=true
    bOverrideAngularDamping=true
    bUseGravityOverride=true
    simulationSpace=RootRelative
    alphaBoolBlend {blendInTime=0.3, blendOption=Linear, blendOutTime=0.3}
    boundBone {boneName=dyn_l_shoulderFray_top_04_b}
    componentAppliedLinearAccClamp=<3000 3000 3000>
    componentLinearAccScale=<0.5 0.5 0.5>
    componentLinearVelScale=<0.6 0.6 0.6>
    gravityOverride {z=-900}
    physicsBodyDefinitions [{
      sphereCollisionRadius=0
      boundBone {boneName=dyn_l_shoulderFray_top_04_b}
      boxExtents=<20 10 10>
      constraintSetup {angularLimitsMax=<25 10 25>, angularLimitsMin=<-25 -25 -25>, angularTarget=<0 0 0>}
      localJointOffset=<2 0 0>
     }]
   }
   n115 = Anim Dynamics - Bone: dyn_l_shoulderFray_btm_04_a(ComponentPose=@n114, bAlphaBoolEnabled=@n51, ExternalForce="0,0,0") {
    alphaInputType=Bool
    angularDampingOverride=0.9999
    angularSpringConstant=75
    bAngularSpring=true
    bGravityOverrideInSimSpace=true
    bOverrideAngularDamping=true
    bUseGravityOverride=true
    simulationSpace=RootRelative
    alphaBoolBlend {blendInTime=0.3, blendOption=Linear, blendOutTime=0.3}
    boundBone {boneName=dyn_l_shoulderFray_btm_04_a}
    componentAppliedLinearAccClamp=<3000 3000 3000>
    componentLinearAccScale=<0.5 0.5 0.5>
    componentLinearVelScale=<0.6 0.6 0.6>
    gravityOverride {z=-900}
    physicsBodyDefinitions [{
      sphereCollisionRadius=0
      boundBone {boneName=dyn_l_shoulderFray_btm_04_a}
      boxExtents=<20 10 10>
      constraintSetup {angularLimitsMax=<25 10 25>, angularLimitsMin=<-25 -25 -25>, angularTarget=<0 0 0>}
      localJointOffset=<2 0 0>
     }]
   }
   n116 = Anim Dynamics - Bone: dyn_l_shoulderFray_btm_04_b(ComponentPose=@n115, bAlphaBoolEnabled=@n51, ExternalForce="0,0,0") {
    alphaInputType=Bool
    angularDampingOverride=0.9999
    angularSpringConstant=75
    bAngularSpring=true
    bGravityOverrideInSimSpace=true
    bOverrideAngularDamping=true
    bUseGravityOverride=true
    simulationSpace=RootRelative
    alphaBoolBlend {blendInTime=0.3, blendOption=Linear, blendOutTime=0.3}
    boundBone {boneName=dyn_l_shoulderFray_btm_04_b}
    componentAppliedLinearAccClamp=<3000 3000 3000>
    componentLinearAccScale=<0.5 0.5 0.5>
    componentLinearVelScale=<0.6 0.6 0.6>
    gravityOverride {z=-900}
    physicsBodyDefinitions [{
      sphereCollisionRadius=0
      boundBone {boneName=dyn_l_shoulderFray_btm_04_b}
      boxExtents=<20 10 10>
      constraintSetup {angularLimitsMax=<25 10 25>, angularLimitsMin=<-25 -25 -25>, angularTarget=<0 0 0>}
      localJointOffset=<2 0 0>
     }]
   }
   n117 = Anim Dynamics - Bone: dyn_l_shoulderFray_top_03_a(ComponentPose=@n116, bAlphaBoolEnabled=@n51, ExternalForce="0,0,0") {
    alphaInputType=Bool
    angularDampingOverride=0.9999
    angularSpringConstant=75
    bAngularSpring=true
    bGravityOverrideInSimSpace=true
    bOverrideAngularDamping=true
    bUseGravityOverride=true
    simulationSpace=RootRelative
    alphaBoolBlend {blendInTime=0.3, blendOption=Linear, blendOutTime=0.3}
    boundBone {boneName=dyn_l_shoulderFray_top_03_a}
    componentAppliedLinearAccClamp=<3000 3000 3000>
    componentLinearAccScale=<0.5 0.5 0.5>
    componentLinearVelScale=<0.6 0.6 0.6>
    gravityOverride {z=-900}
    physicsBodyDefinitions [{
      sphereCollisionRadius=0
      boundBone {boneName=dyn_l_shoulderFray_top_03_a}
      boxExtents=<20 10 10>
      constraintSetup {angularLimitsMax=<25 10 25>, angularLimitsMin=<-25 -25 -25>, angularTarget=<0 0 0>}
      localJointOffset=<2 0 0>
     }]
   }
   n118 = Anim Dynamics - Bone: dyn_l_shoulderFray_top_03_b(ComponentPose=@n117, bAlphaBoolEnabled=@n51, ExternalForce="0,0,0") {
    alphaInputType=Bool
    angularDampingOverride=0.9999
    angularSpringConstant=75
    bAngularSpring=true
    bGravityOverrideInSimSpace=true
    bOverrideAngularDamping=true
    bUseGravityOverride=true
    simulationSpace=RootRelative
    alphaBoolBlend {blendInTime=0.3, blendOption=Linear, blendOutTime=0.3}
    boundBone {boneName=dyn_l_shoulderFray_top_03_b}
    componentAppliedLinearAccClamp=<3000 3000 3000>
    componentLinearAccScale=<0.5 0.5 0.5>
    componentLinearVelScale=<0.6 0.6 0.6>
    gravityOverride {z=-900}
    physicsBodyDefinitions [{
      sphereCollisionRadius=0
      boundBone {boneName=dyn_l_shoulderFray_top_03_b}
      boxExtents=<20 10 10>
      constraintSetup {angularLimitsMax=<25 10 25>, angularLimitsMin=<-25 -25 -25>, angularTarget=<0 0 0>}
      localJointOffset=<2 0 0>
     }]
   }
   n119 = Anim Dynamics - Bone: dyn_l_shoulderFray_btm_03_a(ComponentPose=@n118, bAlphaBoolEnabled=@n51, ExternalForce="0,0,0") {
    alphaInputType=Bool
    angularDampingOverride=0.9999
    angularSpringConstant=75
    bAngularSpring=true
    bGravityOverrideInSimSpace=true
    bOverrideAngularDamping=true
    bUseGravityOverride=true
    simulationSpace=RootRelative
    alphaBoolBlend {blendInTime=0.3, blendOption=Linear, blendOutTime=0.3}
    boundBone {boneName=dyn_l_shoulderFray_btm_03_a}
    componentAppliedLinearAccClamp=<3000 3000 3000>
    componentLinearAccScale=<0.5 0.5 0.5>
    componentLinearVelScale=<0.6 0.6 0.6>
    gravityOverride {z=-900}
    physicsBodyDefinitions [{
      sphereCollisionRadius=0
      boundBone {boneName=dyn_l_shoulderFray_btm_03_a}
      boxExtents=<20 10 10>
      constraintSetup {angularLimitsMax=<25 10 25>, angularLimitsMin=<-25 -25 -25>, angularTarget=<0 0 0>}
      localJointOffset=<2 0 0>
     }]
   }
   n120 = Anim Dynamics - Bone: dyn_l_shoulderFray_btm_03_b(ComponentPose=@n119, bAlphaBoolEnabled=@n51, ExternalForce="0,0,0") {
    alphaInputType=Bool
    angularDampingOverride=0.9999
    angularSpringConstant=75
    bAngularSpring=true
    bGravityOverrideInSimSpace=true
    bOverrideAngularDamping=true
    bUseGravityOverride=true
    simulationSpace=RootRelative
    alphaBoolBlend {blendInTime=0.3, blendOption=Linear, blendOutTime=0.3}
    boundBone {boneName=dyn_l_shoulderFray_btm_03_b}
    componentAppliedLinearAccClamp=<3000 3000 3000>
    componentLinearAccScale=<0.5 0.5 0.5>
    componentLinearVelScale=<0.6 0.6 0.6>
    gravityOverride {z=-900}
    physicsBodyDefinitions [{
      sphereCollisionRadius=0
      boundBone {boneName=dyn_l_shoulderFray_btm_03_b}
      boxExtents=<20 10 10>
      constraintSetup {angularLimitsMax=<25 10 25>, angularLimitsMin=<-25 -25 -25>, angularTarget=<0 0 0>}
      localJointOffset=<2 0 0>
     }]
   }
   n121 = Anim Dynamics - Bone: dyn_l_shoulderFray_top_02_a(ComponentPose=@n120, bAlphaBoolEnabled=@n51, ExternalForce="0,0,0") {
    alphaInputType=Bool
    angularDampingOverride=0.9999
    angularSpringConstant=75
    bAngularSpring=true
    bGravityOverrideInSimSpace=true
    bOverrideAngularDamping=true
    bUseGravityOverride=true
    simulationSpace=RootRelative
    alphaBoolBlend {blendInTime=0.3, blendOption=Linear, blendOutTime=0.3}
    boundBone {boneName=dyn_l_shoulderFray_top_02_a}
    componentAppliedLinearAccClamp=<3000 3000 3000>
    componentLinearAccScale=<0.5 0.5 0.5>
    componentLinearVelScale=<0.6 0.6 0.6>
    gravityOverride {z=-900}
    physicsBodyDefinitions [{
      sphereCollisionRadius=0
      boundBone {boneName=dyn_l_shoulderFray_top_02_a}
      boxExtents=<20 10 10>
      constraintSetup {angularLimitsMax=<25 10 25>, angularLimitsMin=<-25 -25 -25>, angularTarget=<0 0 0>}
      localJointOffset=<2 0 0>
     }]
   }
   n122 = Anim Dynamics - Bone: dyn_l_shoulderFray_top_02_b(ComponentPose=@n121, bAlphaBoolEnabled=@n51, ExternalForce="0,0,0") {
    alphaInputType=Bool
    angularDampingOverride=0.9999
    angularSpringConstant=75
    bAngularSpring=true
    bGravityOverrideInSimSpace=true
    bOverrideAngularDamping=true
    bUseGravityOverride=true
    simulationSpace=RootRelative
    alphaBoolBlend {blendInTime=0.3, blendOption=Linear, blendOutTime=0.3}
    boundBone {boneName=dyn_l_shoulderFray_top_02_b}
    componentAppliedLinearAccClamp=<3000 3000 3000>
    componentLinearAccScale=<0.5 0.5 0.5>
    componentLinearVelScale=<0.6 0.6 0.6>
    gravityOverride {z=-900}
    physicsBodyDefinitions [{
      sphereCollisionRadius=0
      boundBone {boneName=dyn_l_shoulderFray_top_02_b}
      boxExtents=<20 10 10>
      constraintSetup {angularLimitsMax=<25 10 25>, angularLimitsMin=<-25 -25 -25>, angularTarget=<0 0 0>}
      localJointOffset=<2 0 0>
     }]
   }
   n123 = Anim Dynamics - Bone: dyn_l_shoulderFray_btm_02_a(ComponentPose=@n122, bAlphaBoolEnabled=@n51, ExternalForce="0,0,0") {
    alphaInputType=Bool
    angularDampingOverride=0.9999
    angularSpringConstant=75
    bAngularSpring=true
    bGravityOverrideInSimSpace=true
    bOverrideAngularDamping=true
    bUseGravityOverride=true
    simulationSpace=RootRelative
    alphaBoolBlend {blendInTime=0.3, blendOption=Linear, blendOutTime=0.3}
    boundBone {boneName=dyn_l_shoulderFray_btm_02_a}
    componentAppliedLinearAccClamp=<3000 3000 3000>
    componentLinearAccScale=<0.5 0.5 0.5>
    componentLinearVelScale=<0.6 0.6 0.6>
    gravityOverride {z=-900}
    physicsBodyDefinitions [{
      sphereCollisionRadius=0
      boundBone {boneName=dyn_l_shoulderFray_btm_02_a}
      boxExtents=<20 10 10>
      constraintSetup {angularLimitsMax=<25 10 25>, angularLimitsMin=<-25 -25 -25>, angularTarget=<0 0 0>}
      localJointOffset=<2 0 0>
     }]
   }
   n124 = Anim Dynamics - Bone: dyn_l_shoulderFray_btm_02_b(ComponentPose=@n123, bAlphaBoolEnabled=@n51, ExternalForce="0,0,0") {
    alphaInputType=Bool
    angularDampingOverride=0.9999
    angularSpringConstant=75
    bAngularSpring=true
    bGravityOverrideInSimSpace=true
    bOverrideAngularDamping=true
    bUseGravityOverride=true
    simulationSpace=RootRelative
    alphaBoolBlend {blendInTime=0.3, blendOption=Linear, blendOutTime=0.3}
    boundBone {boneName=dyn_l_shoulderFray_btm_02_b}
    componentAppliedLinearAccClamp=<3000 3000 3000>
    componentLinearAccScale=<0.5 0.5 0.5>
    componentLinearVelScale=<0.6 0.6 0.6>
    gravityOverride {z=-900}
    physicsBodyDefinitions [{
      sphereCollisionRadius=0
      boundBone {boneName=dyn_l_shoulderFray_btm_02_b}
      boxExtents=<20 10 10>
      constraintSetup {angularLimitsMax=<25 10 25>, angularLimitsMin=<-25 -25 -25>, angularTarget=<0 0 0>}
      localJointOffset=<2 0 0>
     }]
   }
   n125 = Anim Dynamics - Bone: dyn_l_shoulderFray_top_01_a(ComponentPose=@n124, bAlphaBoolEnabled=@n51, ExternalForce="0,0,0") {
    alphaInputType=Bool
    angularDampingOverride=0.9999
    angularSpringConstant=75
    bAngularSpring=true
    bGravityOverrideInSimSpace=true
    bOverrideAngularDamping=true
    bUseGravityOverride=true
    simulationSpace=RootRelative
    alphaBoolBlend {blendInTime=0.3, blendOption=Linear, blendOutTime=0.3}
    boundBone {boneName=dyn_l_shoulderFray_top_01_a}
    componentAppliedLinearAccClamp=<3000 3000 3000>
    componentLinearAccScale=<0.5 0.5 0.5>
    componentLinearVelScale=<0.6 0.6 0.6>
    gravityOverride {z=-900}
    physicsBodyDefinitions [{
      sphereCollisionRadius=0
      boundBone {boneName=dyn_l_shoulderFray_top_01_a}
      boxExtents=<20 10 10>
      constraintSetup {angularLimitsMax=<25 10 25>, angularLimitsMin=<-25 -25 -25>, angularTarget=<0 0 0>}
      localJointOffset=<2 0 0>
     }]
   }
   n126 = Anim Dynamics - Bone: dyn_l_shoulderFray_top_01_b(ComponentPose=@n125, bAlphaBoolEnabled=@n51, ExternalForce="0,0,0") {
    alphaInputType=Bool
    angularDampingOverride=0.9999
    angularSpringConstant=75
    bAngularSpring=true
    bGravityOverrideInSimSpace=true
    bOverrideAngularDamping=true
    bUseGravityOverride=true
    simulationSpace=RootRelative
    alphaBoolBlend {blendInTime=0.3, blendOption=Linear, blendOutTime=0.3}
    boundBone {boneName=dyn_l_shoulderFray_top_01_b}
    componentAppliedLinearAccClamp=<3000 3000 3000>
    componentLinearAccScale=<0.5 0.5 0.5>
    componentLinearVelScale=<0.6 0.6 0.6>
    gravityOverride {z=-900}
    physicsBodyDefinitions [{
      sphereCollisionRadius=0
      boundBone {boneName=dyn_l_shoulderFray_top_01_b}
      boxExtents=<20 10 10>
      constraintSetup {angularLimitsMax=<25 10 25>, angularLimitsMin=<-25 -25 -25>, angularTarget=<0 0 0>}
      localJointOffset=<2 0 0>
     }]
   }
   n127 = Anim Dynamics - Bone: dyn_l_shoulderFray_btm_01_a(ComponentPose=@n126, bAlphaBoolEnabled=@n51, ExternalForce="0,0,0") {
    alphaInputType=Bool
    angularDampingOverride=0.9999
    angularSpringConstant=75
    bAngularSpring=true
    bGravityOverrideInSimSpace=true
    bOverrideAngularDamping=true
    bUseGravityOverride=true
    simulationSpace=RootRelative
    alphaBoolBlend {blendInTime=0.3, blendOption=Linear, blendOutTime=0.3}
    boundBone {boneName=dyn_l_shoulderFray_btm_01_a}
    componentAppliedLinearAccClamp=<3000 3000 3000>
    componentLinearAccScale=<0.5 0.5 0.5>
    componentLinearVelScale=<0.6 0.6 0.6>
    gravityOverride {z=-900}
    physicsBodyDefinitions [{
      sphereCollisionRadius=0
      boundBone {boneName=dyn_l_shoulderFray_btm_01_a}
      boxExtents=<20 10 10>
      constraintSetup {angularLimitsMax=<25 10 25>, angularLimitsMin=<-25 -25 -25>, angularTarget=<0 0 0>}
      localJointOffset=<2 0 0>
     }]
   }
   n128 = Anim Dynamics - Bone: dyn_l_shoulderFray_btm_01_b(ComponentPose=@n127, bAlphaBoolEnabled=@n51, ExternalForce="0,0,0") {
    alphaInputType=Bool
    angularDampingOverride=0.9999
    angularSpringConstant=75
    bAngularSpring=true
    bGravityOverrideInSimSpace=true
    bOverrideAngularDamping=true
    bUseGravityOverride=true
    simulationSpace=RootRelative
    alphaBoolBlend {blendInTime=0.3, blendOption=Linear, blendOutTime=0.3}
    boundBone {boneName=dyn_l_shoulderFray_btm_01_b}
    componentAppliedLinearAccClamp=<3000 3000 3000>
    componentLinearAccScale=<0.5 0.5 0.5>
    componentLinearVelScale=<0.6 0.6 0.6>
    gravityOverride {z=-900}
    physicsBodyDefinitions [{
      sphereCollisionRadius=0
      boundBone {boneName=dyn_l_shoulderFray_btm_01_b}
      boxExtents=<20 10 10>
      constraintSetup {angularLimitsMax=<25 10 25>, angularLimitsMin=<-25 -25 -25>, angularTarget=<0 0 0>}
      localJointOffset=<2 0 0>
     }]
   }
   n129 = Anim Dynamics - Bone: dyn_r_shoulderFray_btm_06_a(ComponentPose=@n128, bAlphaBoolEnabled=@n51, ExternalForce="0,0,0") {
    alphaInputType=Bool
    angularDampingOverride=0.9999
    angularSpringConstant=75
    bAngularSpring=true
    bGravityOverrideInSimSpace=true
    bOverrideAngularDamping=true
    bUseGravityOverride=true
    simulationSpace=RootRelative
    alphaBoolBlend {blendInTime=0.3, blendOption=Linear, blendOutTime=0.3}
    boundBone {boneName=dyn_r_shoulderFray_btm_06_a}
    componentAppliedLinearAccClamp=<3000 3000 3000>
    componentLinearAccScale=<0.5 0.5 0.5>
    componentLinearVelScale=<0.6 0.6 0.6>
    gravityOverride {z=-900}
    physicsBodyDefinitions [{
      sphereCollisionRadius=0
      boundBone {boneName=dyn_r_shoulderFray_btm_06_a}
      boxExtents=<20 10 10>
      constraintSetup {angularLimitsMax=<25 10 25>, angularLimitsMin=<-25 -25 -25>, angularTarget=<0 0 0>}
      localJointOffset=<-2 0 0>
     }]
   }
   n130 = Anim Dynamics - Bone: dyn_r_shoulderFray_btm_06_b(ComponentPose=@n129, bAlphaBoolEnabled=@n51, ExternalForce="0,0,0") {
    alphaInputType=Bool
    angularDampingOverride=0.9999
    angularSpringConstant=75
    bAngularSpring=true
    bGravityOverrideInSimSpace=true
    bOverrideAngularDamping=true
    bUseGravityOverride=true
    simulationSpace=RootRelative
    alphaBoolBlend {blendInTime=0.3, blendOption=Linear, blendOutTime=0.3}
    boundBone {boneName=dyn_r_shoulderFray_btm_06_b}
    componentAppliedLinearAccClamp=<3000 3000 3000>
    componentLinearAccScale=<0.5 0.5 0.5>
    componentLinearVelScale=<0.6 0.6 0.6>
    gravityOverride {z=-900}
    physicsBodyDefinitions [{
      sphereCollisionRadius=0
      boundBone {boneName=dyn_r_shoulderFray_btm_06_b}
      boxExtents=<20 10 10>
      constraintSetup {angularLimitsMax=<25 10 25>, angularLimitsMin=<-25 -25 -25>, angularTarget=<0 0 0>}
      localJointOffset=<-2 0 0>
     }]
   }
   n131 = Anim Dynamics - Bone: dyn_r_shoulderFray_top_05_a(ComponentPose=@n130, bAlphaBoolEnabled=@n51, ExternalForce="0,0,0") {
    alphaInputType=Bool
    angularDampingOverride=0.9999
    angularSpringConstant=75
    bAngularSpring=true
    bGravityOverrideInSimSpace=true
    bOverrideAngularDamping=true
    bUseGravityOverride=true
    simulationSpace=RootRelative
    alphaBoolBlend {blendInTime=0.3, blendOption=Linear, blendOutTime=0.3}
    boundBone {boneName=dyn_r_shoulderFray_top_05_a}
    componentAppliedLinearAccClamp=<3000 3000 3000>
    componentLinearAccScale=<0.5 0.5 0.5>
    componentLinearVelScale=<0.6 0.6 0.6>
    gravityOverride {z=-900}
    physicsBodyDefinitions [{
      sphereCollisionRadius=0
      boundBone {boneName=dyn_r_shoulderFray_top_05_a}
      boxExtents=<20 10 10>
      constraintSetup {angularLimitsMax=<10 10 10>, angularLimitsMin=<-10 -25 -10>, angularTarget=<0 0 0>}
      localJointOffset=<-2 0 0>
     }]
   }
   n132 = Anim Dynamics - Bone: dyn_r_shoulderFray_top_05_b(ComponentPose=@n131, bAlphaBoolEnabled=@n51, ExternalForce="0,0,0") {
    alphaInputType=Bool
    angularDampingOverride=0.9999
    angularSpringConstant=75
    bAngularSpring=true
    bGravityOverrideInSimSpace=true
    bOverrideAngularDamping=true
    bUseGravityOverride=true
    simulationSpace=RootRelative
    alphaBoolBlend {blendInTime=0.3, blendOption=Linear, blendOutTime=0.3}
    boundBone {boneName=dyn_r_shoulderFray_top_05_b}
    componentAppliedLinearAccClamp=<3000 3000 3000>
    componentLinearAccScale=<0.5 0.5 0.5>
    componentLinearVelScale=<0.6 0.6 0.6>
    gravityOverride {z=-900}
    physicsBodyDefinitions [{
      sphereCollisionRadius=0
      boundBone {boneName=dyn_r_shoulderFray_top_05_b}
      boxExtents=<20 10 10>
      constraintSetup {angularLimitsMax=<10 10 10>, angularLimitsMin=<-10 -25 -10>, angularTarget=<0 0 0>}
      localJointOffset=<-2 0 0>
     }]
   }
   n133 = Anim Dynamics - Bone: dyn_r_shoulderFray_btm_05_a(ComponentPose=@n132, bAlphaBoolEnabled=@n51, ExternalForce="0,0,0") {
    alphaInputType=Bool
    angularDampingOverride=0.9999
    angularSpringConstant=75
    bAngularSpring=true
    bGravityOverrideInSimSpace=true
    bOverrideAngularDamping=true
    bUseGravityOverride=true
    simulationSpace=RootRelative
    alphaBoolBlend {blendInTime=0.3, blendOption=Linear, blendOutTime=0.3}
    boundBone {boneName=dyn_r_shoulderFray_btm_05_a}
    componentAppliedLinearAccClamp=<3000 3000 3000>
    componentLinearAccScale=<0.5 0.5 0.5>
    componentLinearVelScale=<0.6 0.6 0.6>
    gravityOverride {z=-900}
    physicsBodyDefinitions [{
      sphereCollisionRadius=0
      boundBone {boneName=dyn_r_shoulderFray_btm_05_a}
      boxExtents=<20 10 10>
      constraintSetup {angularLimitsMax=<10 10 10>, angularLimitsMin=<-10 -25 -10>, angularTarget=<0 0 0>}
      localJointOffset=<-2 0 0>
     }]
   }
   n134 = Anim Dynamics - Bone: dyn_r_shoulderFray_btm_05_b(ComponentPose=@n133, bAlphaBoolEnabled=@n51, ExternalForce="0,0,0") {
    alphaInputType=Bool
    angularDampingOverride=0.9999
    angularSpringConstant=75
    bAngularSpring=true
    bGravityOverrideInSimSpace=true
    bOverrideAngularDamping=true
    bUseGravityOverride=true
    simulationSpace=RootRelative
    alphaBoolBlend {blendInTime=0.3, blendOption=Linear, blendOutTime=0.3}
    boundBone {boneName=dyn_r_shoulderFray_btm_05_b}
    componentAppliedLinearAccClamp=<3000 3000 3000>
    componentLinearAccScale=<0.5 0.5 0.5>
    componentLinearVelScale=<0.6 0.6 0.6>
    gravityOverride {z=-900}
    physicsBodyDefinitions [{
      sphereCollisionRadius=0
      boundBone {boneName=dyn_r_shoulderFray_btm_05_b}
      boxExtents=<20 10 10>
      constraintSetup {angularLimitsMax=<10 10 10>, angularLimitsMin=<-10 -25 -10>, angularTarget=<0 0 0>}
      localJointOffset=<-2 0 0>
     }]
   }
   n135 = Anim Dynamics - Bone: dyn_r_shoulderFray_top_06_a(ComponentPose=@n134, bAlphaBoolEnabled=@n51, ExternalForce="0,0,0") {
    alphaInputType=Bool
    angularDampingOverride=0.9999
    angularSpringConstant=75
    bAngularSpring=true
    bGravityOverrideInSimSpace=true
    bOverrideAngularDamping=true
    bUseGravityOverride=true
    simulationSpace=RootRelative
    alphaBoolBlend {blendInTime=0.3, blendOption=Linear, blendOutTime=0.3}
    boundBone {boneName=dyn_r_shoulderFray_top_06_a}
    componentAppliedLinearAccClamp=<3000 3000 3000>
    componentLinearAccScale=<0.5 0.5 0.5>
    componentLinearVelScale=<0.6 0.6 0.6>
    gravityOverride {z=-900}
    physicsBodyDefinitions [{
      sphereCollisionRadius=0
      boundBone {boneName=dyn_r_shoulderFray_top_06_a}
      boxExtents=<20 10 10>
      constraintSetup {angularLimitsMax=<25 10 25>, angularLimitsMin=<-25 -25 -25>, angularTarget=<0 0 0>}
      localJointOffset=<-2 0 0>
     }]
   }
   n136 = Anim Dynamics - Bone: dyn_r_shoulderFray_top_06_b(ComponentPose=@n135, bAlphaBoolEnabled=@n51, ExternalForce="0,0,0") {
    alphaInputType=Bool
    angularDampingOverride=0.9999
    angularSpringConstant=75
    bAngularSpring=true
    bGravityOverrideInSimSpace=true
    bOverrideAngularDamping=true
    bUseGravityOverride=true
    simulationSpace=RootRelative
    alphaBoolBlend {blendInTime=0.3, blendOption=Linear, blendOutTime=0.3}
    boundBone {boneName=dyn_r_shoulderFray_top_06_b}
    componentAppliedLinearAccClamp=<3000 3000 3000>
    componentLinearAccScale=<0.5 0.5 0.5>
    componentLinearVelScale=<0.6 0.6 0.6>
    gravityOverride {z=-900}
    physicsBodyDefinitions [{
      sphereCollisionRadius=0
      boundBone {boneName=dyn_r_shoulderFray_top_06_b}
      boxExtents=<20 10 10>
      constraintSetup {angularLimitsMax=<25 10 25>, angularLimitsMin=<-25 -25 -25>, angularTarget=<0 0 0>}
      localJointOffset=<-2 0 0>
     }]
   }
   n137 = Anim Dynamics - Bone: dyn_r_shoulderFray_top_04_a(ComponentPose=@n136, bAlphaBoolEnabled=@n51, ExternalForce="0,0,0") {
    alphaInputType=Bool
    angularDampingOverride=0.9999
    angularSpringConstant=75
    bAngularSpring=true
    bGravityOverrideInSimSpace=true
    bOverrideAngularDamping=true
    bUseGravityOverride=true
    simulationSpace=RootRelative
    alphaBoolBlend {blendInTime=0.3, blendOption=Linear, blendOutTime=0.3}
    boundBone {boneName=dyn_r_shoulderFray_top_04_a}
    componentAppliedLinearAccClamp=<3000 3000 3000>
    componentLinearAccScale=<0.5 0.5 0.5>
    componentLinearVelScale=<0.6 0.6 0.6>
    gravityOverride {z=-900}
    physicsBodyDefinitions [{
      sphereCollisionRadius=0
      boundBone {boneName=dyn_r_shoulderFray_top_04_a}
      boxExtents=<20 10 10>
      constraintSetup {angularLimitsMax=<25 10 25>, angularLimitsMin=<-25 -25 -25>, angularTarget=<0 0 0>}
      localJointOffset=<-2 0 0>
     }]
   }
   n138 = Anim Dynamics - Bone: dyn_r_shoulderFray_top_04_b(ComponentPose=@n137, bAlphaBoolEnabled=@n51, ExternalForce="0,0,0") {
    alphaInputType=Bool
    angularDampingOverride=0.9999
    angularSpringConstant=75
    bAngularSpring=true
    bGravityOverrideInSimSpace=true
    bOverrideAngularDamping=true
    bUseGravityOverride=true
    simulationSpace=RootRelative
    alphaBoolBlend {blendInTime=0.3, blendOption=Linear, blendOutTime=0.3}
    boundBone {boneName=dyn_r_shoulderFray_top_04_b}
    componentAppliedLinearAccClamp=<3000 3000 3000>
    componentLinearAccScale=<0.5 0.5 0.5>
    componentLinearVelScale=<0.6 0.6 0.6>
    gravityOverride {z=-900}
    physicsBodyDefinitions [{
      sphereCollisionRadius=0
      boundBone {boneName=dyn_r_shoulderFray_top_04_b}
      boxExtents=<20 10 10>
      constraintSetup {angularLimitsMax=<25 10 25>, angularLimitsMin=<-25 -25 -25>, angularTarget=<0 0 0>}
      localJointOffset=<-2 0 0>
     }]
   }
   n139 = Anim Dynamics - Bone: dyn_r_shoulderFray_btm_04_a(ComponentPose=@n138, bAlphaBoolEnabled=@n51, ExternalForce="0,0,0") {
    alphaInputType=Bool
    angularDampingOverride=0.9999
    angularSpringConstant=75
    bAngularSpring=true
    bGravityOverrideInSimSpace=true
    bOverrideAngularDamping=true
    bUseGravityOverride=true
    simulationSpace=RootRelative
    alphaBoolBlend {blendInTime=0.3, blendOption=Linear, blendOutTime=0.3}
    boundBone {boneName=dyn_r_shoulderFray_btm_04_a}
    componentAppliedLinearAccClamp=<3000 3000 3000>
    componentLinearAccScale=<0.5 0.5 0.5>
    componentLinearVelScale=<0.6 0.6 0.6>
    gravityOverride {z=-900}
    physicsBodyDefinitions [{
      sphereCollisionRadius=0
      boundBone {boneName=dyn_r_shoulderFray_btm_04_a}
      boxExtents=<20 10 10>
      constraintSetup {angularLimitsMax=<25 10 25>, angularLimitsMin=<-25 -25 -25>, angularTarget=<0 0 0>}
      localJointOffset=<-2 0 0>
     }]
   }
   n140 = Anim Dynamics - Bone: dyn_r_shoulderFray_btm_04_b(ComponentPose=@n139, bAlphaBoolEnabled=@n51, ExternalForce="0,0,0") {
    alphaInputType=Bool
    angularDampingOverride=0.9999
    angularSpringConstant=75
    bAngularSpring=true
    bGravityOverrideInSimSpace=true
    bOverrideAngularDamping=true
    bUseGravityOverride=true
    simulationSpace=RootRelative
    alphaBoolBlend {blendInTime=0.3, blendOption=Linear, blendOutTime=0.3}
    boundBone {boneName=dyn_r_shoulderFray_btm_04_b}
    componentAppliedLinearAccClamp=<3000 3000 3000>
    componentLinearAccScale=<0.5 0.5 0.5>
    componentLinearVelScale=<0.6 0.6 0.6>
    gravityOverride {z=-900}
    physicsBodyDefinitions [{
      sphereCollisionRadius=0
      boundBone {boneName=dyn_r_shoulderFray_btm_04_b}
      boxExtents=<20 10 10>
      constraintSetup {angularLimitsMax=<25 10 25>, angularLimitsMin=<-25 -25 -25>, angularTarget=<0 0 0>}
      localJointOffset=<-2 0 0>
     }]
   }
   n141 = Anim Dynamics - Bone: dyn_r_shoulderFray_top_03_a(ComponentPose=@n140, bAlphaBoolEnabled=@n51, ExternalForce="0,0,0") {
    alphaInputType=Bool
    angularDampingOverride=0.9999
    angularSpringConstant=75
    bAngularSpring=true
    bGravityOverrideInSimSpace=true
    bOverrideAngularDamping=true
    bUseGravityOverride=true
    simulationSpace=RootRelative
    alphaBoolBlend {blendInTime=0.3, blendOption=Linear, blendOutTime=0.3}
    boundBone {boneName=dyn_r_shoulderFray_top_03_a}
    componentAppliedLinearAccClamp=<3000 3000 3000>
    componentLinearAccScale=<0.5 0.5 0.5>
    componentLinearVelScale=<0.6 0.6 0.6>
    gravityOverride {z=-900}
    physicsBodyDefinitions [{
      sphereCollisionRadius=0
      boundBone {boneName=dyn_r_shoulderFray_top_03_a}
      boxExtents=<20 10 10>
      constraintSetup {angularLimitsMax=<25 10 25>, angularLimitsMin=<-25 -25 -25>, angularTarget=<0 0 0>}
      localJointOffset=<-2 0 0>
     }]
   }
   n142 = Anim Dynamics - Bone: dyn_r_shoulderFray_top_03_b(ComponentPose=@n141, bAlphaBoolEnabled=@n51, ExternalForce="0,0,0") {
    alphaInputType=Bool
    angularDampingOverride=0.9999
    angularSpringConstant=75
    bAngularSpring=true
    bGravityOverrideInSimSpace=true
    bOverrideAngularDamping=true
    bUseGravityOverride=true
    simulationSpace=RootRelative
    alphaBoolBlend {blendInTime=0.3, blendOption=Linear, blendOutTime=0.3}
    boundBone {boneName=dyn_r_shoulderFray_top_03_b}
    componentAppliedLinearAccClamp=<3000 3000 3000>
    componentLinearAccScale=<0.5 0.5 0.5>
    componentLinearVelScale=<0.6 0.6 0.6>
    gravityOverride {z=-900}
    physicsBodyDefinitions [{
      sphereCollisionRadius=0
      boundBone {boneName=dyn_r_shoulderFray_top_03_b}
      boxExtents=<20 10 10>
      constraintSetup {angularLimitsMax=<25 10 25>, angularLimitsMin=<-25 -25 -25>, angularTarget=<0 0 0>}
      localJointOffset=<-2 0 0>
     }]
   }
   n143 = Anim Dynamics - Bone: dyn_r_shoulderFray_btm_03_a(ComponentPose=@n142, bAlphaBoolEnabled=@n51, ExternalForce="0,0,0") {
    alphaInputType=Bool
    angularDampingOverride=0.9999
    angularSpringConstant=75
    bAngularSpring=true
    bGravityOverrideInSimSpace=true
    bOverrideAngularDamping=true
    bUseGravityOverride=true
    simulationSpace=RootRelative
    alphaBoolBlend {blendInTime=0.3, blendOption=Linear, blendOutTime=0.3}
    boundBone {boneName=dyn_r_shoulderFray_btm_03_a}
    componentAppliedLinearAccClamp=<3000 3000 3000>
    componentLinearAccScale=<0.5 0.5 0.5>
    componentLinearVelScale=<0.6 0.6 0.6>
    gravityOverride {z=-900}
    physicsBodyDefinitions [{
      sphereCollisionRadius=0
      boundBone {boneName=dyn_r_shoulderFray_btm_03_a}
      boxExtents=<20 10 10>
      constraintSetup {angularLimitsMax=<25 10 25>, angularLimitsMin=<-25 -25 -25>, angularTarget=<0 0 0>}
      localJointOffset=<-2 0 0>
     }]
   }
   n144 = Anim Dynamics - Bone: dyn_r_shoulderFray_btm_03_b(ComponentPose=@n143, bAlphaBoolEnabled=@n51, ExternalForce="0,0,0") {
    alphaInputType=Bool
    angularDampingOverride=0.9999
    angularSpringConstant=75
    bAngularSpring=true
    bGravityOverrideInSimSpace=true
    bOverrideAngularDamping=true
    bUseGravityOverride=true
    simulationSpace=RootRelative
    alphaBoolBlend {blendInTime=0.3, blendOption=Linear, blendOutTime=0.3}
    boundBone {boneName=dyn_r_shoulderFray_btm_03_b}
    componentAppliedLinearAccClamp=<3000 3000 3000>
    componentLinearAccScale=<0.5 0.5 0.5>
    componentLinearVelScale=<0.6 0.6 0.6>
    gravityOverride {z=-900}
    physicsBodyDefinitions [{
      sphereCollisionRadius=0
      boundBone {boneName=dyn_r_shoulderFray_btm_03_b}
      boxExtents=<20 10 10>
      constraintSetup {angularLimitsMax=<25 10 25>, angularLimitsMin=<-25 -25 -25>, angularTarget=<0 0 0>}
      localJointOffset=<-2 0 0>
     }]
   }
   n145 = Anim Dynamics - Bone: dyn_r_shoulderFray_top_02_a(ComponentPose=@n144, bAlphaBoolEnabled=@n51, ExternalForce="0,0,0") {
    alphaInputType=Bool
    angularDampingOverride=0.9999
    angularSpringConstant=75
    bAngularSpring=true
    bGravityOverrideInSimSpace=true
    bOverrideAngularDamping=true
    bUseGravityOverride=true
    simulationSpace=RootRelative
    alphaBoolBlend {blendInTime=0.3, blendOption=Linear, blendOutTime=0.3}
    boundBone {boneName=dyn_r_shoulderFray_top_02_a}
    componentAppliedLinearAccClamp=<3000 3000 3000>
    componentLinearAccScale=<0.5 0.5 0.5>
    componentLinearVelScale=<0.6 0.6 0.6>
    gravityOverride {z=-900}
    physicsBodyDefinitions [{
      sphereCollisionRadius=0
      boundBone {boneName=dyn_r_shoulderFray_top_02_a}
      boxExtents=<20 10 10>
      constraintSetup {angularLimitsMax=<25 10 25>, angularLimitsMin=<-25 -25 -25>, angularTarget=<0 0 0>}
      localJointOffset=<-2 0 0>
     }]
   }
   n146 = Anim Dynamics - Bone: dyn_r_shoulderFray_top_02_b(ComponentPose=@n145, bAlphaBoolEnabled=@n51, ExternalForce="0,0,0") {
    alphaInputType=Bool
    angularDampingOverride=0.9999
    angularSpringConstant=75
    bAngularSpring=true
    bGravityOverrideInSimSpace=true
    bOverrideAngularDamping=true
    bUseGravityOverride=true
    simulationSpace=RootRelative
    alphaBoolBlend {blendInTime=0.3, blendOption=Linear, blendOutTime=0.3}
    boundBone {boneName=dyn_r_shoulderFray_top_02_b}
    componentAppliedLinearAccClamp=<3000 3000 3000>
    componentLinearAccScale=<0.5 0.5 0.5>
    componentLinearVelScale=<0.6 0.6 0.6>
    gravityOverride {z=-900}
    physicsBodyDefinitions [{
      sphereCollisionRadius=0
      boundBone {boneName=dyn_r_shoulderFray_top_02_b}
      boxExtents=<20 10 10>
      constraintSetup {angularLimitsMax=<25 10 25>, angularLimitsMin=<-25 -25 -25>, angularTarget=<0 0 0>}
      localJointOffset=<-2 0 0>
     }]
   }
   n147 = Anim Dynamics - Bone: dyn_r_shoulderFray_btm_02_a(ComponentPose=@n146, bAlphaBoolEnabled=@n51, ExternalForce="0,0,0") {
    alphaInputType=Bool
    angularDampingOverride=0.9999
    angularSpringConstant=75
    bAngularSpring=true
    bGravityOverrideInSimSpace=true
    bOverrideAngularDamping=true
    bUseGravityOverride=true
    simulationSpace=RootRelative
    alphaBoolBlend {blendInTime=0.3, blendOption=Linear, blendOutTime=0.3}
    boundBone {boneName=dyn_r_shoulderFray_btm_02_a}
    componentAppliedLinearAccClamp=<3000 3000 3000>
    componentLinearAccScale=<0.5 0.5 0.5>
    componentLinearVelScale=<0.6 0.6 0.6>
    gravityOverride {z=-900}
    physicsBodyDefinitions [{
      sphereCollisionRadius=0
      boundBone {boneName=dyn_r_shoulderFray_btm_02_a}
      boxExtents=<20 10 10>
      constraintSetup {angularLimitsMax=<25 10 25>, angularLimitsMin=<-25 -25 -25>, angularTarget=<0 0 0>}
      localJointOffset=<-2 0 0>
     }]
   }
   n148 = Anim Dynamics - Bone: dyn_r_shoulderFray_btm_02_b(ComponentPose=@n147, bAlphaBoolEnabled=@n51, ExternalForce="0,0,0") {
    alphaInputType=Bool
    angularDampingOverride=0.9999
    angularSpringConstant=75
    bAngularSpring=true
    bGravityOverrideInSimSpace=true
    bOverrideAngularDamping=true
    bUseGravityOverride=true
    simulationSpace=RootRelative
    alphaBoolBlend {blendInTime=0.3, blendOption=Linear, blendOutTime=0.3}
    boundBone {boneName=dyn_r_shoulderFray_btm_02_b}
    componentAppliedLinearAccClamp=<3000 3000 3000>
    componentLinearAccScale=<0.5 0.5 0.5>
    componentLinearVelScale=<0.6 0.6 0.6>
    gravityOverride {z=-900}
    physicsBodyDefinitions [{
      sphereCollisionRadius=0
      boundBone {boneName=dyn_r_shoulderFray_btm_02_b}
      boxExtents=<20 10 10>
      constraintSetup {angularLimitsMax=<25 10 25>, angularLimitsMin=<-25 -25 -25>, angularTarget=<0 0 0>}
      localJointOffset=<-2 0 0>
     }]
   }
   n149 = Anim Dynamics - Bone: dyn_r_shoulderFray_top_01_a(ComponentPose=@n148, bAlphaBoolEnabled=@n51, ExternalForce="0,0,0") {
    alphaInputType=Bool
    angularDampingOverride=0.9999
    angularSpringConstant=75
    bAngularSpring=true
    bGravityOverrideInSimSpace=true
    bOverrideAngularDamping=true
    bUseGravityOverride=true
    simulationSpace=RootRelative
    alphaBoolBlend {blendInTime=0.3, blendOption=Linear, blendOutTime=0.3}
    boundBone {boneName=dyn_r_shoulderFray_top_01_a}
    componentAppliedLinearAccClamp=<3000 3000 3000>
    componentLinearAccScale=<0.5 0.5 0.5>
    componentLinearVelScale=<0.6 0.6 0.6>
    gravityOverride {z=-900}
    physicsBodyDefinitions [{
      sphereCollisionRadius=0
      boundBone {boneName=dyn_r_shoulderFray_top_01_a}
      boxExtents=<20 10 10>
      constraintSetup {angularLimitsMax=<25 10 25>, angularLimitsMin=<-25 -25 -25>, angularTarget=<0 0 0>}
      localJointOffset=<-2 0 0>
     }]
   }
   n150 = Anim Dynamics - Bone: dyn_r_shoulderFray_top_01_b(ComponentPose=@n149, bAlphaBoolEnabled=@n51, ExternalForce="0,0,0") {
    alphaInputType=Bool
    angularDampingOverride=0.9999
    angularSpringConstant=75
    bAngularSpring=true
    bGravityOverrideInSimSpace=true
    bOverrideAngularDamping=true
    bUseGravityOverride=true
    simulationSpace=RootRelative
    alphaBoolBlend {blendInTime=0.3, blendOption=Linear, blendOutTime=0.3}
    boundBone {boneName=dyn_r_shoulderFray_top_01_b}
    componentAppliedLinearAccClamp=<3000 3000 3000>
    componentLinearAccScale=<0.5 0.5 0.5>
    componentLinearVelScale=<0.6 0.6 0.6>
    gravityOverride {z=-900}
    physicsBodyDefinitions [{
      sphereCollisionRadius=0
      boundBone {boneName=dyn_r_shoulderFray_top_01_b}
      boxExtents=<20 10 10>
      constraintSetup {angularLimitsMax=<25 10 25>, angularLimitsMin=<-25 -25 -25>, angularTarget=<0 0 0>}
      localJointOffset=<-2 0 0>
     }]
   }
   n151 = Anim Dynamics - Bone: dyn_r_shoulderFray_btm_01_a(ComponentPose=@n150, bAlphaBoolEnabled=@n51, ExternalForce="0,0,0") {
    alphaInputType=Bool
    angularDampingOverride=0.9999
    angularSpringConstant=75
    bAngularSpring=true
    bGravityOverrideInSimSpace=true
    bOverrideAngularDamping=true
    bUseGravityOverride=true
    simulationSpace=RootRelative
    alphaBoolBlend {blendInTime=0.3, blendOption=Linear, blendOutTime=0.3}
    boundBone {boneName=dyn_r_shoulderFray_btm_01_a}
    componentAppliedLinearAccClamp=<3000 3000 3000>
    componentLinearAccScale=<0.5 0.5 0.5>
    componentLinearVelScale=<0.6 0.6 0.6>
    gravityOverride {z=-900}
    physicsBodyDefinitions [{
      sphereCollisionRadius=0
      boundBone {boneName=dyn_r_shoulderFray_btm_01_a}
      boxExtents=<20 10 10>
      constraintSetup {angularLimitsMax=<25 10 25>, angularLimitsMin=<-25 -25 -25>, angularTarget=<0 0 0>}
      localJointOffset=<-2 0 0>
     }]
   }
   n152 = Anim Dynamics - Bone: dyn_r_shoulderFray_btm_01_b(ComponentPose=@n151, bAlphaBoolEnabled=@n51, ExternalForce="0,0,0") {
    alphaInputType=Bool
    angularDampingOverride=0.9999
    angularSpringConstant=75
    bAngularSpring=true
    bGravityOverrideInSimSpace=true
    bOverrideAngularDamping=true
    bUseGravityOverride=true
    simulationSpace=RootRelative
    alphaBoolBlend {blendInTime=0.3, blendOption=Linear, blendOutTime=0.3}
    boundBone {boneName=dyn_r_shoulderFray_btm_01_b}
    componentAppliedLinearAccClamp=<3000 3000 3000>
    componentLinearAccScale=<0.5 0.5 0.5>
    componentLinearVelScale=<0.6 0.6 0.6>
    gravityOverride {z=-900}
    physicsBodyDefinitions [{
      sphereCollisionRadius=0
      boundBone {boneName=dyn_r_shoulderFray_btm_01_b}
      boxExtents=<20 10 10>
      constraintSetup {angularLimitsMax=<25 10 25>, angularLimitsMin=<-25 -25 -25>, angularTarget=<0 0 0>}
      localJointOffset=<-2 0 0>
     }]
   }
   n153 = Anim Dynamics - Bone: dyn_jacket_l_01(ComponentPose=@n152, Alpha=1, ExternalForce="0,0,0") {
    angularDampingOverride=0.9999
    angularSpringConstant=25
    bAngularSpring=true
    bGravityOverrideInSimSpace=true
    bOverrideAngularDamping=true
    simulationSpace=BoneRelative
    alphaBoolBlend {blendOption=Linear}
    boundBone {boneName=dyn_jacket_l_01}
    componentAppliedLinearAccClamp=<3500 3500 3500>
    componentLinearAccScale=<1.3 1.3 1.3>
    componentLinearVelScale=<1.3 1.3 1.3>
    physicsBodyDefinitions [{
      sphereCollisionRadius=0
      boundBone {boneName=dyn_jacket_l_01}
      boxExtents=<10 20 20>
      constraintSetup {angularLimitsMax=<3 0 10>, angularLimitsMin=<-3 -25 -10>, angularTarget=<0 0 0>}
      localJointOffset=<2 0 0>
     }]
    relativeSpaceBone {boneName=root}
   }
   n154 = Anim Dynamics - Bone: dyn_jacket_l_02(ComponentPose=@n153, Alpha=1, ExternalForce="0,0,0") {
    angularDampingOverride=0.9999
    angularSpringConstant=25
    bAngularSpring=true
    bGravityOverrideInSimSpace=true
    bOverrideAngularDamping=true
    bUsePlanarLimit=false
    simulationSpace=BoneRelative
    alphaBoolBlend {blendOption=Linear}
    boundBone {boneName=dyn_jacket_l_02}
    componentAppliedLinearAccClamp=<3500 3500 3500>
    componentLinearAccScale=<1.3 1.3 1.3>
    componentLinearVelScale=<1.3 1.3 1.3>
    physicsBodyDefinitions [{
      sphereCollisionRadius=0
      boundBone {boneName=dyn_jacket_l_02}
      boxExtents=<10 20 20>
      constraintSetup {angularLimitsMax=<3 0 10>, angularLimitsMin=<-3 -25 -10>, angularTarget=<0 0 0>}
      localJointOffset=<2 0 0>
     }]
    planarLimits [{drivingBone {boneName=thigh_l}, planeTransform {location=<-1.8082 17.8291 0.0003>, rotation=<rot -0.0089 0.0004 90>}}]
    relativeSpaceBone {boneName=root}
   }
   n155 = Anim Dynamics - Bone: dyn_jacket_fr_l_01(ComponentPose=@n154, Alpha=1, ExternalForce="0,0,0") {
    angularDampingOverride=0.9999
    angularSpringConstant=25
    bAngularSpring=true
    bGravityOverrideInSimSpace=true
    bOverrideAngularDamping=true
    simulationSpace=BoneRelative
    alphaBoolBlend {blendOption=Linear}
    boundBone {boneName=dyn_jacket_fr_l_01}
    componentAppliedLinearAccClamp=<3500 3500 3500>
    componentLinearAccScale=<1.3 1.3 1.3>
    componentLinearVelScale=<1.3 1.3 1.3>
    physicsBodyDefinitions [{
      sphereCollisionRadius=0
      boundBone {boneName=dyn_jacket_fr_l_01}
      boxExtents=<10 20 20>
      constraintSetup {angularLimitsMax=<3 0 10>, angularLimitsMin=<-3 -25 -10>, angularTarget=<0 0 0>}
      localJointOffset=<2 0 0>
     }]
    relativeSpaceBone {boneName=root}
   }
   n156 = Anim Dynamics - Bone: dyn_jacket_fr_l_02(ComponentPose=@n155, Alpha=1, ExternalForce="0,0,0") {
    angularDampingOverride=0.9999
    angularSpringConstant=25
    bAngularSpring=true
    bGravityOverrideInSimSpace=true
    bOverrideAngularDamping=true
    bUsePlanarLimit=false
    simulationSpace=BoneRelative
    alphaBoolBlend {blendOption=Linear}
    boundBone {boneName=dyn_jacket_fr_l_02}
    componentAppliedLinearAccClamp=<3500 3500 3500>
    componentLinearAccScale=<1.3 1.3 1.3>
    componentLinearVelScale=<1.3 1.3 1.3>
    physicsBodyDefinitions [{
      sphereCollisionRadius=0
      boundBone {boneName=dyn_jacket_fr_l_02}
      boxExtents=<10 20 20>
      constraintSetup {angularLimitsMax=<3 0 10>, angularLimitsMin=<-3 -25 -10>, angularTarget=<0 0 0>}
      localJointOffset=<2 0 0>
     }]
    relativeSpaceBone {boneName=root}
   }
   n157 = Anim Dynamics - Bone: dyn_jacket_bk_l_01(ComponentPose=@n156, Alpha=1, ExternalForce="0,0,0") {
    angularDampingOverride=0.9999
    angularSpringConstant=25
    bAngularSpring=true
    bGravityOverrideInSimSpace=true
    bOverrideAngularDamping=true
    simulationSpace=BoneRelative
    alphaBoolBlend {blendOption=Linear}
    boundBone {boneName=dyn_jacket_bk_l_01}
    componentAppliedLinearAccClamp=<3500 3500 3500>
    componentLinearAccScale=<1.3 1.3 1.3>
    componentLinearVelScale=<1.3 1.3 1.3>
    physicsBodyDefinitions [{
      sphereCollisionRadius=0
      boundBone {boneName=dyn_jacket_bk_l_01}
      boxExtents=<10 20 20>
      constraintSetup {angularLimitsMax=<3 0 10>, angularLimitsMin=<-3 -25 -10>, angularTarget=<0 0 0>}
      localJointOffset=<2 0 0>
     }]
    relativeSpaceBone {boneName=root}
   }
   n158 = Anim Dynamics - Bone: dyn_jacket_bk_l_02(ComponentPose=@n157, Alpha=1, ExternalForce="0,0,0") {
    angularDampingOverride=0.9999
    angularSpringConstant=25
    bAngularSpring=true
    bGravityOverrideInSimSpace=true
    bOverrideAngularDamping=true
    bUsePlanarLimit=false
    simulationSpace=BoneRelative
    alphaBoolBlend {blendOption=Linear}
    boundBone {boneName=dyn_jacket_bk_l_02}
    componentAppliedLinearAccClamp=<3500 3500 3500>
    componentLinearAccScale=<1.3 1.3 1.3>
    componentLinearVelScale=<1.3 1.3 1.3>
    physicsBodyDefinitions [{
      sphereCollisionRadius=0
      boundBone {boneName=dyn_jacket_bk_l_02}
      boxExtents=<10 20 20>
      constraintSetup {angularLimitsMax=<3 0 10>, angularLimitsMin=<-3 -25 -10>, angularTarget=<0 0 0>}
      localJointOffset=<2 0 0>
     }]
    relativeSpaceBone {boneName=root}
   }
   n159 = Anim Dynamics - Bone: dyn_jacket_bk_01(ComponentPose=@n158, Alpha=1, ExternalForce="0,0,0") {
    angularDampingOverride=0.9999
    angularSpringConstant=25
    bAngularSpring=true
    bGravityOverrideInSimSpace=true
    bOverrideAngularDamping=true
    simulationSpace=BoneRelative
    alphaBoolBlend {blendOption=Linear}
    boundBone {boneName=dyn_jacket_bk_01}
    componentAppliedLinearAccClamp=<3500 3500 3500>
    componentLinearAccScale=<1.3 1.3 1.3>
    componentLinearVelScale=<1.3 1.3 1.3>
    physicsBodyDefinitions [{
      sphereCollisionRadius=0
      boundBone {boneName=dyn_jacket_bk_01}
      boxExtents=<10 20 20>
      constraintSetup {angularLimitsMax=<3 0 10>, angularLimitsMin=<-3 -25 -10>, angularTarget=<0 0 0>}
      localJointOffset=<2 0 0>
     }]
    relativeSpaceBone {boneName=root}
   }
   n160 = Anim Dynamics - Bone: dyn_jacket_bk_02(ComponentPose=@n159, Alpha=1, ExternalForce="0,0,0") {
    angularDampingOverride=0.9999
    angularSpringConstant=25
    bAngularSpring=true
    bGravityOverrideInSimSpace=true
    bOverrideAngularDamping=true
    simulationSpace=BoneRelative
    alphaBoolBlend {blendOption=Linear}
    boundBone {boneName=dyn_jacket_bk_02}
    componentAppliedLinearAccClamp=<3500 3500 3500>
    componentLinearAccScale=<1.3 1.3 1.3>
    componentLinearVelScale=<1.3 1.3 1.3>
    physicsBodyDefinitions [{
      sphereCollisionRadius=0
      boundBone {boneName=dyn_jacket_bk_02}
      boxExtents=<10 20 20>
      constraintSetup {angularLimitsMax=<3 0 10>, angularLimitsMin=<-3 -25 -10>, angularTarget=<0 0 0>}
      localJointOffset=<2 0 0>
     }]
    relativeSpaceBone {boneName=root}
   }
   n161 = Anim Dynamics - Bone: dyn_jacket_fr_r_01(ComponentPose=@n160, Alpha=1, ExternalForce="0,0,0") {
    angularDampingOverride=0.9999
    angularSpringConstant=25
    bAngularSpring=true
    bGravityOverrideInSimSpace=true
    bOverrideAngularDamping=true
    simulationSpace=BoneRelative
    alphaBoolBlend {blendOption=Linear}
    boundBone {boneName=dyn_jacket_fr_r_01}
    componentAppliedLinearAccClamp=<3500 3500 3500>
    componentLinearAccScale=<1.3 1.3 1.3>
    componentLinearVelScale=<1.3 1.3 1.3>
    physicsBodyDefinitions [{
      sphereCollisionRadius=0
      boundBone {boneName=dyn_jacket_fr_r_01}
      boxExtents=<10 20 20>
      constraintSetup {angularLimitsMax=<3 0 10>, angularLimitsMin=<-3 -25 -10>, angularTarget=<0 0 0>}
      localJointOffset=<2 0 0>
     }]
    relativeSpaceBone {boneName=root}
   }
   n162 = Anim Dynamics - Bone: dyn_jacket_fr_r_02(ComponentPose=@n161, Alpha=1, ExternalForce="0,0,0") {
    angularDampingOverride=0.9999
    angularSpringConstant=25
    bAngularSpring=true
    bGravityOverrideInSimSpace=true
    bOverrideAngularDamping=true
    simulationSpace=BoneRelative
    alphaBoolBlend {blendOption=Linear}
    boundBone {boneName=dyn_jacket_fr_r_02}
    componentAppliedLinearAccClamp=<3500 3500 3500>
    componentLinearAccScale=<1.3 1.3 1.3>
    componentLinearVelScale=<1.3 1.3 1.3>
    physicsBodyDefinitions [{
      sphereCollisionRadius=0
      boundBone {boneName=dyn_jacket_fr_r_02}
      boxExtents=<10 20 20>
      constraintSetup {angularLimitsMax=<3 0 10>, angularLimitsMin=<-3 -25 -10>, angularTarget=<0 0 0>}
      localJointOffset=<2 0 0>
     }]
    relativeSpaceBone {boneName=root}
   }
   n163 = Anim Dynamics - Bone: dyn_jacket_bk_r_01(ComponentPose=@n162, Alpha=1, ExternalForce="0,0,0") {
    angularDampingOverride=0.9999
    angularSpringConstant=25
    bAngularSpring=true
    bGravityOverrideInSimSpace=true
    bOverrideAngularDamping=true
    bUseSphericalLimits=true
    simulationSpace=BoneRelative
    alphaBoolBlend {blendOption=Linear}
    boundBone {boneName=dyn_jacket_bk_r_01}
    componentAppliedLinearAccClamp=<3500 3500 3500>
    componentLinearAccScale=<1.3 1.3 1.3>
    componentLinearVelScale=<1.3 1.3 1.3>
    physicsBodyDefinitions [{
      sphereCollisionRadius=0
      boundBone {boneName=dyn_jacket_bk_r_01}
      boxExtents=<10 20 20>
      constraintSetup {angularLimitsMax=<3 0 10>, angularLimitsMin=<-3 -25 -10>, angularTarget=<0 0 0>}
      localJointOffset=<2 0 0>
     }]
    relativeSpaceBone {boneName=root}
    sphericalLimits [{limitRadius=15, drivingBone {boneName=foot_r}, sphereLocalOffset=<2.0801 10.0238 2>}]
   }
   n164 = Anim Dynamics - Bone: dyn_jacket_bk_r_02(ComponentPose=@n163, Alpha=1, ExternalForce="0,0,0") {
    angularDampingOverride=0.9999
    angularSpringConstant=25
    bAngularSpring=true
    bGravityOverrideInSimSpace=true
    bOverrideAngularDamping=true
    simulationSpace=BoneRelative
    alphaBoolBlend {blendOption=Linear}
    boundBone {boneName=dyn_jacket_bk_r_02}
    componentAppliedLinearAccClamp=<3500 3500 3500>
    componentLinearAccScale=<1.3 1.3 1.3>
    componentLinearVelScale=<1.3 1.3 1.3>
    physicsBodyDefinitions [{
      sphereCollisionRadius=0
      boundBone {boneName=dyn_jacket_bk_r_02}
      boxExtents=<10 20 20>
      constraintSetup {angularLimitsMax=<3 0 10>, angularLimitsMin=<-3 -25 -10>, angularTarget=<0 0 0>}
      localJointOffset=<2 0 0>
     }]
    relativeSpaceBone {boneName=root}
   }
   n165 = Anim Dynamics - Bone: dyn_jacket_r_01(ComponentPose=@n164, Alpha=1, ExternalForce="0,0,0") {
    angularDampingOverride=0.9999
    angularSpringConstant=25
    bAngularSpring=true
    bGravityOverrideInSimSpace=true
    bOverrideAngularDamping=true
    simulationSpace=BoneRelative
    alphaBoolBlend {blendOption=Linear}
    boundBone {boneName=dyn_jacket_r_01}
    componentAppliedLinearAccClamp=<3500 3500 3500>
    componentLinearAccScale=<1.3 1.3 1.3>
    componentLinearVelScale=<1.3 1.3 1.3>
    physicsBodyDefinitions [{
      sphereCollisionRadius=0
      boundBone {boneName=dyn_jacket_r_01}
      boxExtents=<10 20 20>
      constraintSetup {angularLimitsMax=<3 0 10>, angularLimitsMin=<-3 -25 -10>, angularTarget=<0 0 0>}
      localJointOffset=<2 0 0>
     }]
    relativeSpaceBone {boneName=root}
   }
   n166 = Anim Dynamics - Bone: dyn_jacket_r_02(ComponentPose=@n165, Alpha=1, ExternalForce="0,0,0") {
    angularDampingOverride=0.9999
    angularSpringConstant=25
    bAngularSpring=true
    bGravityOverrideInSimSpace=true
    bOverrideAngularDamping=true
    simulationSpace=BoneRelative
    alphaBoolBlend {blendOption=Linear}
    boundBone {boneName=dyn_jacket_r_02}
    componentAppliedLinearAccClamp=<3500 3500 3500>
    componentLinearAccScale=<1.3 1.3 1.3>
    componentLinearVelScale=<1.3 1.3 1.3>
    physicsBodyDefinitions [{
      sphereCollisionRadius=0
      boundBone {boneName=dyn_jacket_r_02}
      boxExtents=<10 20 20>
      constraintSetup {angularLimitsMax=<3 0 10>, angularLimitsMin=<-3 -25 -10>, angularTarget=<0 0 0>}
      localJointOffset=<2 0 0>
     }]
    relativeSpaceBone {boneName=root}
   }
   n167 = "Transform (Modify) Bone - Bone: dyn_jacket_l_MSTR"(ComponentPose=@n166, Alpha=1, Translation="0,0,2", Rotation="0,0,0", Scale="1,1,1") {
    rotationMode=BMM_Additive
    rotationSpace=BCS_BoneSpace
    scaleMode=BMM_Additive
    scaleSpace=BCS_BoneSpace
    translationMode=BMM_Additive
    translationSpace=BCS_BoneSpace
    alphaBoolBlend {blendOption=Linear}
    boneToModify {boneName=dyn_jacket_l_MSTR}
    translation {z=2}
   }
   n168 = "Transform (Modify) Bone - Bone: dyn_jacket_fr_l_MSTR"(ComponentPose=@n167, Alpha=1, Translation="0,0,2", Rotation="0,0,0", Scale="1,1,1") {
    rotationMode=BMM_Additive
    rotationSpace=BCS_BoneSpace
    scaleMode=BMM_Additive
    scaleSpace=BCS_BoneSpace
    translationMode=BMM_Additive
    translationSpace=BCS_BoneSpace
    alphaBoolBlend {blendOption=Linear}
    boneToModify {boneName=dyn_jacket_fr_l_MSTR}
    translation {z=2}
   }
   n169 = "Transform (Modify) Bone - Bone: dyn_jacket_bk_l_MSTR"(ComponentPose=@n168, Alpha=1, Translation="0,0,1", Rotation="0,0,0", Scale="1,1,1") {
    rotationMode=BMM_Additive
    rotationSpace=BCS_BoneSpace
    scaleMode=BMM_Additive
    scaleSpace=BCS_BoneSpace
    translationMode=BMM_Additive
    translationSpace=BCS_BoneSpace
    alphaBoolBlend {blendOption=Linear}
    boneToModify {boneName=dyn_jacket_bk_l_MSTR}
    translation {z=1}
   }
   n170 = "Transform (Modify) Bone - Bone: dyn_jacket_bk_MSTR"(ComponentPose=@n169, Alpha=1, Translation="0,0,1", Rotation="0,0,0", Scale="1,1,1") {
    rotationMode=BMM_Additive
    rotationSpace=BCS_BoneSpace
    scaleMode=BMM_Additive
    scaleSpace=BCS_BoneSpace
    translationMode=BMM_Additive
    translationSpace=BCS_BoneSpace
    alphaBoolBlend {blendOption=Linear}
    boneToModify {boneName=dyn_jacket_bk_MSTR}
    translation {z=1}
   }
   n171 = "Transform (Modify) Bone - Bone: dyn_jacket_fr_r_MSTR"(ComponentPose=@n170, Alpha=1, Translation="0,0,2", Rotation="0,0,0", Scale="1,1,1") {
    rotationMode=BMM_Additive
    rotationSpace=BCS_BoneSpace
    scaleMode=BMM_Additive
    scaleSpace=BCS_BoneSpace
    translationMode=BMM_Additive
    translationSpace=BCS_BoneSpace
    alphaBoolBlend {blendOption=Linear}
    boneToModify {boneName=dyn_jacket_fr_r_MSTR}
    translation {z=2}
   }
   n172 = "Transform (Modify) Bone - Bone: dyn_jacket_bk_r_MSTR"(ComponentPose=@n171, Alpha=1, Translation="0,0,1", Rotation="0,0,0", Scale="1,1,1") {
    rotationMode=BMM_Additive
    rotationSpace=BCS_BoneSpace
    scaleMode=BMM_Additive
    scaleSpace=BCS_BoneSpace
    translationMode=BMM_Additive
    translationSpace=BCS_BoneSpace
    alphaBoolBlend {blendOption=Linear}
    boneToModify {boneName=dyn_jacket_bk_r_MSTR}
    translation {z=1}
   }
   n173 = "Transform (Modify) Bone - Bone: dyn_jacket_r_MSTR"(ComponentPose=@n172, Alpha=1, Translation="0,0,1", Rotation="0,0,0", Scale="1,1,1") {
    rotationMode=BMM_Additive
    rotationSpace=BCS_BoneSpace
    scaleMode=BMM_Additive
    scaleSpace=BCS_BoneSpace
    translationMode=BMM_Additive
    translationSpace=BCS_BoneSpace
    alphaBoolBlend {blendOption=Linear}
    boneToModify {boneName=dyn_jacket_r_MSTR}
    translation {z=1}
   }
   n174 = Anim Dynamics - Bone: dyn_l_zipper_chest_01!(ComponentPose=@n173, bAlphaBoolEnabled:bool=True, ExternalForce="0,0,0") -> (Pose) {
    alphaInputType=Bool
    angularDampingOverride=0.99
    bOverrideAngularDamping=true
    bUseGravityOverride=true
    bUsePlanarLimit=false
    linearDampingOverride=0.9999
    simulationSpace=RootRelative
    alphaBoolBlend {blendInTime=0.3, blendOption=Linear, blendOutTime=0.4}
    boundBone {boneName=dyn_l_zipper_chest_01}
    componentAppliedLinearAccClamp=<3000 3000 1000>
    componentLinearAccScale=<0.2 0.2 0.6>
    componentLinearVelScale=<0.3 0.3 0.7>
    gravityOverride {z=-800}
    physicsBodyDefinitions [{
      boundBone {boneName=dyn_l_zipper_chest_01}
      boxExtents=<12 8 8>
      constraintSetup {angularLimitsMax=<5 0 2>, angularLimitsMin=<-5 -75 -2>, angularTarget=<0 0 0>}
      localJointOffset=<1 0 0>
     }]
   }
   n175 = Anim Dynamics - Bone: dyn_l_zipper_chest_02!(ComponentPose=@n174, bAlphaBoolEnabled:bool=True, ExternalForce="0,0,0") -> (Pose) {
    alphaInputType=Bool
    angularDampingOverride=0.99
    bOverrideAngularDamping=true
    bUseGravityOverride=true
    bUsePlanarLimit=false
    linearDampingOverride=0.9999
    simulationSpace=RootRelative
    alphaBoolBlend {blendInTime=0.3, blendOption=Linear, blendOutTime=0.4}
    boundBone {boneName=dyn_l_zipper_chest_02}
    componentAppliedLinearAccClamp=<3000 3000 1000>
    componentLinearAccScale=<0.2 0.2 0.6>
    componentLinearVelScale=<0.3 0.3 0.7>
    gravityOverride {z=-600}
    physicsBodyDefinitions [{
      boundBone {boneName=dyn_l_zipper_chest_02}
      boxExtents=<14 9 9>
      constraintSetup {angularLimitsMax=<5 0 2>, angularLimitsMin=<-5 -50 -2>, angularTarget=<0 0 0>}
      localJointOffset=<1 0 0>
     }]
   }
   n176 = Anim Dynamics - Bone: dyn_r_zipper_chest_01!(ComponentPose=@n175, bAlphaBoolEnabled:bool=True, ExternalForce="0,0,0") -> (Pose) {
    alphaInputType=Bool
    angularDampingOverride=0.99
    bOverrideAngularDamping=true
    bUseGravityOverride=true
    bUsePlanarLimit=false
    linearDampingOverride=0.9999
    simulationSpace=RootRelative
    alphaBoolBlend {blendInTime=0.3, blendOption=Linear, blendOutTime=0.4}
    boundBone {boneName=dyn_r_zipper_chest_01}
    componentAppliedLinearAccClamp=<3000 3000 1000>
    componentLinearAccScale=<0.2 0.2 0.6>
    componentLinearVelScale=<0.3 0.3 0.7>
    gravityOverride {z=-800}
    physicsBodyDefinitions [{
      boundBone {boneName=dyn_r_zipper_chest_01}
      boxExtents=<12 8 8>
      constraintSetup {angularLimitsMax=<5 0 2>, angularLimitsMin=<-5 -75 -2>, angularTarget=<0 0 0>}
      localJointOffset=<1 0 0>
     }]
   }
   n177 = Anim Dynamics - Bone: dyn_r_zipper_chest_02!(ComponentPose=@n176, bAlphaBoolEnabled:bool=True, ExternalForce="0,0,0") -> (Pose) {
    alphaInputType=Bool
    angularDampingOverride=0.99
    bOverrideAngularDamping=true
    bUseGravityOverride=true
    bUsePlanarLimit=false
    linearDampingOverride=0.9999
    simulationSpace=RootRelative
    alphaBoolBlend {blendInTime=0.3, blendOption=Linear, blendOutTime=0.4}
    boundBone {boneName=dyn_r_zipper_chest_02}
    componentAppliedLinearAccClamp=<3000 3000 1000>
    componentLinearAccScale=<0.2 0.2 0.6>
    componentLinearVelScale=<0.3 0.3 0.7>
    gravityOverride {z=-600}
    physicsBodyDefinitions [{
      boundBone {boneName=dyn_r_zipper_chest_02}
      boxExtents=<14 9 9>
      constraintSetup {angularLimitsMax=<5 0 2>, angularLimitsMin=<-5 -40 -2>, angularTarget=<0 0 0>}
      localJointOffset=<1 0 0>
     }]
   }
   n178 = Anim Dynamics - Bone: dyn_l_zipper_leg!(ComponentPose=@n177, bAlphaBoolEnabled:bool=True, ExternalForce="0,0,0") -> (Pose) {
    alphaInputType=Bool
    angularDampingOverride=0.99
    bOverrideAngularDamping=true
    bUseGravityOverride=true
    bUsePlanarLimit=false
    linearDampingOverride=0.9999
    simulationSpace=RootRelative
    alphaBoolBlend {blendInTime=0.3, blendOption=Linear, blendOutTime=0.4}
    boundBone {boneName=dyn_l_zipper_leg}
    componentAppliedLinearAccClamp=<3000 3000 1500>
    componentLinearAccScale=<0.2 0.2 0.4>
    componentLinearVelScale=<0.3 0.3 0.5>
    gravityOverride {z=-400}
    physicsBodyDefinitions [{
      boundBone {boneName=dyn_l_zipper_leg}
      boxExtents=<12 8 8>
      constraintSetup {angularLimitsMax=<5 0 2>, angularLimitsMin=<-5 -50 -2>, angularTarget=<0 0 0>}
      localJointOffset=<1 0 0>
     }]
   }
   n179 = Anim Dynamics - Bone: dyn_r_zipper_leg!(ComponentPose=@n178, bAlphaBoolEnabled:bool=True, ExternalForce="0,0,0") -> (Pose) {
    alphaInputType=Bool
    angularDampingOverride=0.99
    bOverrideAngularDamping=true
    bUseGravityOverride=true
    bUsePlanarLimit=false
    linearDampingOverride=0.9999
    simulationSpace=RootRelative
    alphaBoolBlend {blendInTime=0.3, blendOption=Linear, blendOutTime=0.4}
    boundBone {boneName=dyn_r_zipper_leg}
    componentAppliedLinearAccClamp=<3000 3000 1500>
    componentLinearAccScale=<0.2 0.2 0.4>
    componentLinearVelScale=<0.3 0.3 0.5>
    gravityOverride {z=-400}
    physicsBodyDefinitions [{
      boundBone {boneName=dyn_r_zipper_leg}
      boxExtents=<12 8 8>
      constraintSetup {angularLimitsMax=<5 0 2>, angularLimitsMin=<-5 -50 -2>, angularTarget=<0 0 0>}
      localJointOffset=<1 0 0>
     }]
   }
   Save cached pose 'Anim_branch'(Pose=Component To Local(ComponentPose=@n179)) {cacheName=Anim_branch}
   Get bIsSkyDiving()
  ]}
 EventGraph {nodes [
   n1 = Event BlueprintUpdateAnimation()
   > Set Delta Time X(Delta Time X=@n1.DeltaTimeX)
   > n2 = Sequence!(execute) -> (then_0, then_1, then_2:Exec, then_3:Exec)
   > n3 = Branch(Condition=Get bIsJumping())
   Set jump_alpha(execute=[@n3.then, @n4.then], jump_alpha:double=1) -> (Output_Get:double)
   n4 = Branch(execute=@n3.else, Condition=Get bIsFalling())
   Set jump_alpha(execute=@n4.else, jump_alpha=0)
   n5 = GetOwnerSkeletalMeshComponent()
   n6 = BreakTransform(InTransform=GetDeltaTransformFromRefPose(self=@n5, BoneName=upperarm_r, BaseName=clavicle_r))
   Set Tassels_LookDown_Right(execute=@n2.then_1, Tassels_LookDown_Right:double=MapRangeClamped(Value=BreakRotator(InRot=@n6.Rotation).Yaw, InRangeA=100, InRangeB=35)) -> (Output_Get:double)
   n7 = BreakTransform(InTransform=GetDeltaTransformFromRefPose(self=@n5, BoneName=upperarm_r, BaseName=clavicle_r))
   > Set Tassels_LookUp_Right(Tassels_LookUp_Right:double=MapRangeClamped(Value=BreakRotator(InRot=@n7.Rotation).Yaw, InRangeA=90, InRangeB=120)) -> (Output_Get:double)
   n8 = BreakTransform(InTransform=GetDeltaTransformFromRefPose(self=@n5, BoneName=upperarm_l, BaseName=clavicle_l))
   > Set Tassels_LookUp_Left(Tassels_LookUp_Left:double=MapRangeClamped(Value=BreakRotator(InRot=@n8.Rotation).Yaw, InRangeA=40, InRangeB=83)) -> (Output_Get:double)
   n9 = BreakTransform(InTransform=GetDeltaTransformFromRefPose(self=@n5, BoneName=upperarm_r, BaseName=clavicle_r))
   > Set Tassels_LookUp_Right_TwoCorrect(Tassels_LookUp_Right_TwoCorrect:double=MapRangeClamped(Value=BreakRotator(InRot=@n9.Rotation).Yaw, InRangeA=100, InRangeB=120)) -> (Output_Get:double)
   n10 = BreakTransform(InTransform=GetDeltaTransformFromRefPose(self=@n5, BoneName=upperarm_r, BaseName=clavicle_r))
   > Set Tassels_LookLevel(Tassels_LookLevel:double=MapRangeClamped(Value=BreakRotator(InRot=@n10.Rotation).Yaw, InRangeA=105, InRangeB=92)) -> (Output_Get:double)
   n11 = BreakTransform(InTransform=GetDeltaTransformFromRefPose(self=@n5, BoneName=lowerarm_l, BaseName=upperarm_l))
   > Set Tassels_Rifle_LookUp_Left(Tassels_Rifle_LookUp_Left:double=MapRangeClamped(Value=BreakRotator(InRot=@n11.Rotation).Yaw, InRangeA=30, InRangeB=74)) -> (Output_Get:double)
   > Set CrouchingAndMoving(CrouchingAndMoving:bool=AND Boolean(A=Get bIsCrouching(), B=Get bIsMoving2D())) -> (Output_Get:bool)
   > Set CrouchingAiming_Moving(CrouchingAiming_Moving:bool=AND Boolean!(A=Get bIsCrouching(), B=Get bIsMoving2D(), C:bool=Get bIsTargeting()) -> (ReturnValue)) -> (Output_Get:bool)
   n12 = Get bIsPlayingEmote()
   n13 = Get bIsMoving2D()
   > Set Idle(Idle:bool=AND Boolean!(A=NOT Boolean(A=@n13), B=NOT Boolean(A=@n12)) -> (ReturnValue)) -> (Output_Get:bool)
   > Set TargetingCrouchMoving(TargetingCrouchMoving:bool=AND Boolean!(A=Get bIsCrouching(), B=Get bIsMoving2D(), C:bool=Get bIsTargeting()) -> (ReturnValue)) -> (Output_Get:bool)
   n14 = Get bIsParachuteOpened()
   > Set Parachute Offset(Parachute Offset:double=FInterpTo(Current=Get Parachute Offset(), Target="To Float (Boolean)"(InBool=@n14), DeltaTime=Get Delta Time X(), InterpSpeed=10)) -> (Output_Get:double)
   n15 = Get bIsCrouching()
   > Set Crouching Offset(Crouching Offset:double=FInterpTo(Current=Get Crouching Offset(), Target="To Float (Boolean)"(InBool=@n15), DeltaTime=Get Delta Time X(), InterpSpeed=10)) -> (Output_Get:double)
   > Set NotTargeting(NotTargeting:bool=NOT Boolean(A=Get bIsTargeting())) -> (Output_Get:bool)
   n16 = Cast To FortPlayerAnimInstance(execute=@n2.then_2, Object=GetAnimInstance(self=GetOwnerSkeletalMeshComponent()))
   > Set isAimingPistol(isAimingPistol:double=1) -> (Output_Get:double)
   > Set isAimingRifle(isAimingRifle:bool=false) -> (Output_Get:bool)
   > n17 = Switch on EFortWeaponCoreAnimation!(execute, Selection:EFortWeaponCoreAnimation Enum=Get WeaponCoreAnim(self=@n16)) -> (
    Melee:Exec
    Pistol:Exec
    Shotgun:Exec
    PaperBlueprint:Exec
    Rifle:Exec
    MeleeOneHand:Exec
    MachinePistol:Exec
    RocketLauncher:Exec
    GrenadeLauncher:Exec
    GoingCommando:Exec
    AssaultRifle:Exec
    TacticalShotgun:Exec
    SniperRifle:Exec
    TrapPlacement:Exec
    ShoulderLauncher:Exec
    AbilityDecoTool:Exec
    Crossbow:Exec
    C4:Exec
    RemoteControl:Exec
    DualWield:Exec
    AR_BullPup:Exec
    AR_ForwardGrip:Exec
    MedPackPaddles:Exec
    SMG_P90:Exec
    AR_DrumGun:Exec
    Consumable_Small:Exec
    Consumable_Large:Exec
    Balloon:Exec
    MountedTurret:Exec
    CreativeTool:Exec
    ExplosiveBow:Exec
    AshtonIndigo:Exec
    AshtonChicago:Exec
    MeleeDualWield:Exec
    Unarmed:Exec
   )
   Branch(execute=[
     @n17.AssaultRifle
     @n17.Crossbow
     @n17.GrenadeLauncher
     @n17.Rifle
     @n17.RocketLauncher
     @n17.Shotgun
     @n17.SniperRifle
     @n17.TacticalShotgun
    ])
   > Set isAimingRifle(isAimingRifle=true)
   Branch(execute=[@n17.MachinePistol, @n17.Pistol], Condition=false)
   > Set isAimingPistol(isAimingPistol=1)
   n18 = Branch(execute=@n2.then_3, Condition=Get bIsSkyDiving())
   > n19 = Branch(Condition=Get bIsParachuteOpened())
   Set gravity_override(execute=[@n19.then, @n18.else], gravity_override="0, 0,-980")
   > Set rigids alpha(rigids alpha:double=1) -> (Output_Get:double)
   n20 = Branch(execute=@n19.else, Condition="float > float"(A=Get PawnVelocityZ(), B=50) {is=title K2Node_CallFunction})
   > n21 = Set gravity_override(gravity_override="0, 0,200")
   Set rigids alpha(execute=[@n21.then, @n22.then], rigids alpha=0.5)
   n22 = Set gravity_override(execute=@n20.else, gravity_override="0, 0,-200")
   Event OnBeginCrouching()
   > Set crouching alpha(crouching alpha:double=1) -> (Output_Get:double)
   Event OnEndCrouching()
   > Set crouching alpha(crouching alpha=0)
   TryGetPawnOwner()
   PrintString!(execute, InString="To String (Boolean)"(InBool=false), bPrintToScreen, bPrintToLog, TextColor, Duration=0, Key) -> (then)
   PrintString(InString="To String (Boolean)"(InBool=Get bIsSkydiveDiveMode()), Duration=0)
   PrintString(InString="To String (Boolean)"(InBool=false), Duration=0)
   Get bIsSkydiveFloating()
  ]}
]
variables [
 crouching alpha {type=number}
 Crouching Offset {type=number}
 CrouchingAiming_Moving {type=boolean}
 CrouchingAndMoving {type=boolean}
 Delta Time X {type=number}
 gravity_override {type=Vector}
 Idle {type=boolean}
 isAimingPistol {type=number}
 isAimingRifle {type=boolean}
 jump_alpha {type=number}
 NotTargeting {type=boolean}
 Parachute Offset {type=number}
 rigids alpha {type=number}
 TargetingCrouchMoving {type=boolean}
 Tassels_LookDown_Right {type=number}
 Tassels_LookLevel {type=number}
 Tassels_LookUp_Left {type=number}
 Tassels_LookUp_Right {type=number}
 Tassels_LookUp_Right_TwoCorrect {type=number}
 Tassels_Rifle_LookUp_Left {type=number}
]
