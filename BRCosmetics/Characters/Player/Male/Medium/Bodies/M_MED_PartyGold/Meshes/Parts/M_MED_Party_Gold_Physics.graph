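# Physics asset for the male medium Party Gold cosmetic body mesh part, setting up dynamic hair on the head.
# Six hair chains (A–E and G) each use three small capsule bodies, with one large capsule covering the head.
# Each chain's root is constrained to the head and each later segment to the one before it, with uniformly tight swing and twist limits that keep motion subtle.
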PhysicsAsset /BRCosmetics/Characters/Player/Male/Medium/Bodies/M_MED_PartyGold/Meshes/Parts/M_MED_Party_Gold_Physics
bodies {
 dyn_hair_A_01 [{shapeType=Capsule, length=2, radius=0.5, center=<1.3702 0 0>, rotation=<rot 90 0 0>}]
 dyn_hair_A_02 [{shapeType=Capsule, length=2, radius=0.5, center=<1.2026 0.0001 0>, rotation=<rot 90 0 0>}]
 dyn_hair_A_03 [{shapeType=Capsule, length=2, radius=0.5, center=<1.2126 0 0>, rotation=<rot 90 0 0>}]
 dyn_hair_B_01 [{shapeType=Capsule, length=2.3385, radius=0.5361, center=<1.4818 0 0>, rotation=<rot 90 0 0>}]
 dyn_hair_B_02 [{shapeType=Capsule, length=1.9256, radius=0.505, center=<1.2756 0 0>, rotation=<rot 90 0 0>}]
 dyn_hair_B_03 [{shapeType=Capsule, length=0.8202, radius=0.5, center=<0.7471 0.0001 0>, rotation=<rot 90 0 0>}]
 dyn_hair_C_01 [{shapeType=Capsule, length=2, radius=0.5, center=<1.4471 0 0>, rotation=<rot 90 0 0>}]
 dyn_hair_C_02 [{shapeType=Capsule, length=2, radius=0.5, center=<1.4388 0 0>, rotation=<rot 90 0 0>}]
 dyn_hair_C_03 [{shapeType=Capsule, length=2, radius=0.5, center=<1.3502 0 -0.0001>, rotation=<rot 90 0 0>}]
 dyn_hair_D_01 [{shapeType=Capsule, length=2, radius=0.5, center=<1.2723 0 0>, rotation=<rot 90 0 0>}]
 dyn_hair_D_02 [{shapeType=Capsule, length=2, radius=0.5, center=<1.3145 0 0.0001>, rotation=<rot 90 0 0>}]
 dyn_hair_D_03 [{shapeType=Capsule, length=2, radius=0.5, center=<1.5745 0 0>, rotation=<rot 90 0 0>}]
 dyn_hair_E_01 [{shapeType=Capsule, length=2, radius=0.5, center=<1.467 0 -0.0001>, rotation=<rot 90 0 0>}]
 dyn_hair_E_02 [{shapeType=Capsule, length=2, radius=0.5, center=<1.3711 0 0.0001>, rotation=<rot 90 0 0>}]
 dyn_hair_E_03 [{shapeType=Capsule, length=2, radius=0.5, center=<1.1959 0 0.0002>, rotation=<rot 90 0 0>}]
 dyn_hair_G_01 [{shapeType=Capsule, length=2, radius=0.5, center=<1.4246 0 0.0001>, rotation=<rot 90 0 0>}]
 dyn_hair_G_02 [{shapeType=Capsule, length=2, radius=0.5, center=<1.3021 0 -0.0002>, rotation=<rot 90 0 0>}]
 dyn_hair_G_03 [{shapeType=Capsule, length=2, radius=0.5, center=<1.1769 0 -0.0001>, rotation=<rot 90 0 0>}]
 head [{shapeType=Capsule, length=6.6814, radius=7.8077, center=<2.3906 2.6042 0.0704>, rotation=<rot 90 0 7.9641>}]
}
constraints [
 {bone1Name=dyn_hair_C_01, bone2Name=head, swing1LimitDegrees=5, swing2LimitDegrees=5, twistLimitDegrees=5}
 {bone1Name=dyn_hair_C_02, bone2Name=dyn_hair_C_01, swing1LimitDegrees=5, swing2LimitDegrees=5, twistLimitDegrees=5}
 {bone1Name=dyn_hair_C_03, bone2Name=dyn_hair_C_02, swing1LimitDegrees=5, swing2LimitDegrees=5, twistLimitDegrees=5}
 {bone1Name=dyn_hair_G_01, bone2Name=head, swing1LimitDegrees=5, swing2LimitDegrees=5, twistLimitDegrees=5}
 {bone1Name=dyn_hair_G_02, bone2Name=dyn_hair_G_01, swing1LimitDegrees=5, swing2LimitDegrees=5, twistLimitDegrees=5}
 {bone1Name=dyn_hair_G_03, bone2Name=dyn_hair_G_02, swing1LimitDegrees=5, swing2LimitDegrees=5, twistLimitDegrees=5}
 {bone1Name=dyn_hair_E_01, bone2Name=head, swing1LimitDegrees=5, swing2LimitDegrees=5, twistLimitDegrees=5}
 {bone1Name=dyn_hair_E_02, bone2Name=dyn_hair_E_01, swing1LimitDegrees=5, swing2LimitDegrees=5, twistLimitDegrees=5}
 {bone1Name=dyn_hair_E_03, bone2Name=dyn_hair_E_02, swing1LimitDegrees=5, swing2LimitDegrees=5, twistLimitDegrees=5}
 {bone1Name=dyn_hair_A_01, bone2Name=head, swing1LimitDegrees=5, swing2LimitDegrees=5, twistLimitDegrees=5}
 {bone1Name=dyn_hair_A_02, bone2Name=dyn_hair_A_01, swing1LimitDegrees=5, swing2LimitDegrees=5, twistLimitDegrees=5}
 {bone1Name=dyn_hair_A_03, bone2Name=dyn_hair_A_02, swing1LimitDegrees=5, swing2LimitDegrees=5, twistLimitDegrees=5}
 {bone1Name=dyn_hair_D_01, bone2Name=head, swing1LimitDegrees=5, swing2LimitDegrees=5, twistLimitDegrees=5}
 {bone1Name=dyn_hair_D_02, bone2Name=dyn_hair_D_01, swing1LimitDegrees=5, swing2LimitDegrees=5, twistLimitDegrees=5}
 {bone1Name=dyn_hair_D_03, bone2Name=dyn_hair_D_02, swing1LimitDegrees=5, swing2LimitDegrees=5, twistLimitDegrees=5}
 {bone1Name=dyn_hair_B_01, bone2Name=head, swing1LimitDegrees=5, swing2LimitDegrees=5, twistLimitDegrees=5}
 {bone1Name=dyn_hair_B_02, bone2Name=dyn_hair_B_01, swing1LimitDegrees=5, swing2LimitDegrees=5, twistLimitDegrees=5}
 {bone1Name=dyn_hair_B_03, bone2Name=dyn_hair_B_02, swing1LimitDegrees=5, swing2LimitDegrees=5, twistLimitDegrees=5}
]
