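# Animation Blueprint for the face accessory on the medium male PartyGold body, adding secondary physics motion on top of the pose copied from the attached parent mesh.
# A half-weight RigidBody pass is blended in while jumping or falling, otherwise the plain copied pose is used.
# A second, lighter RigidBody pass with downward gravity always runs on that result.
# The event graph sets the blend flag each update from the jumping and falling booleans.
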AnimBlueprint /BRCosmetics/Characters/Player/Male/Medium/Bodies/M_MED_PartyGold/Meshes/Parts/M_MED_Party_Gold_FaceAcc_AnimBP
target_skeleton=&/M_MED_Party_Gold_FaceAcc_Skeleton
graphs [
 AnimGraph {nodes [
   n1 = RigidBody(ComponentPose=Local To Component(LocalPose=Copy Pose From Mesh() {bUseAttachedParent=true}), Alpha=0.5) {
    bClampLinearTranslationLimitToRefPose=true
    bOverrideWorldGravity=true
    overridePhysicsAsset=&/M_MED_Party_Gold_Physics
    simulationSpace=BaseBoneSpace
    alphaBoolBlend {blendOption=Linear}
    baseBoneRef {boneName=root}
    componentLinearAccScale=<0.15 0.15 0.15>
   }
   n2 = Component To Local(ComponentPose=@n1)
   n3 = Copy Pose From Mesh() {bUseAttachedParent=true}
   n4 = Get Jumping_or_falling() -> (Jumping_or_falling:bool)
   n5 = RigidBody(
    ComponentPose=Local To Component(LocalPose=Blend Poses by bool(BlendPose_0=@n2, BlendPose_1=@n3, BlendTime_0=0.3, BlendTime_1=0.3, bActiveValue=@n4) {blendType=Linear})
    Alpha=0.2
   ) {
    bClampLinearTranslationLimitToRefPose=true
    bOverrideWorldGravity=true
    overridePhysicsAsset=&M_MED_Party_Gold_Physics
    simulationSpace=BaseBoneSpace
    alphaBoolBlend {blendOption=Linear}
    baseBoneRef {boneName=root}
    componentLinearAccScale=<0.25 0.25 0.6>
    overrideWorldGravity {z=-980}
   }
   Output Pose(Result=Component To Local(ComponentPose=@n5)) {layerGroup=None}
  ]}
 EventGraph {nodes [
   Event BlueprintUpdateAnimation()
   > Set Jumping_or_falling(Jumping_or_falling:bool=OR Boolean(A=Get bIsJumping(), B=Get bIsFalling())) -> (Output_Get:bool)
  ]}
]
variables [Jumping_or_falling {type=boolean}]
